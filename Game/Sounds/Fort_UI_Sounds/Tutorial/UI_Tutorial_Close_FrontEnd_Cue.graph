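# Sound cue for closing a front-end tutorial UI element, judging by its name.
# A single wave player plays the player survey pop-up sound wave from the front-end UI sounds, with no randomization or modulation.
# It routes through the Fort_UI sound class at reduced volume, and its concurrency override rule is set to prevent new instances.
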SoundCue /Game/Sounds/Fort_UI_Sounds/Tutorial/UI_Tutorial_Close_FrontEnd_Cue
root=WavePlayer_1
nodes [{id=WavePlayer_1, properties {soundWaveAssetPtr=/Game/Sounds/Fort_UI_Sounds/FrontEnd/Fort_UI_PlayerSurvey_PopUp_01}}]
properties {
 maxDistance=2097152
 soundClassObject=&Fort_UI
 volumeMultiplier=0.7
 concurrencyOverrides {resolutionRule=PreventNew}
}
