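# Tutorial screen widget showing a "Skip Tutorial" confirmation modal over a blurred background, listing the seven skills being taught.
# OnTutorialProgressUpdated shows checkmarks on steps with an index below the current phase.
# Cancel or touch-only close plays Out and broadcasts cancellation, and the widget collapses when Out finishes.
# Skip ends the guided tutorial, sends leave-zone event and analytics, then leaves without the party; animation starts trigger open or close sounds.
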WidgetBlueprint /Game/Athena/HUD/Tutorial/Athena_Tutorial_ModalPopup
animations [
 In {
  binding_count=9
  display_label=""
  length_seconds=1.2
  object=&self:In
  sequence {bindings [
    RichText_Desc {
     id=#1
     possessed_class=RichTextBlock
     tracks [
      {
       type=MovieScene2DTransformTrack
       sections [{
         type=MovieScene2DTransformSection
         end_frame=18
         start_frame=-1
         channels [
          Translation.X {type=MovieSceneScriptingFloatChannel, keys [[8.696,100],[13.696,0]]}
          Translation.Y {type=MovieSceneScriptingFloatChannel}
         ]
        }]
      }
      {
       type=MovieSceneFloatTrack
       display_name=Render Opacity
       sections [{
         type=MovieSceneFloatSection
         end_frame=18
         start_frame=0
         channels [None {type=MovieSceneScriptingFloatChannel, keys [[8.696,0],[13.696,1]]}]
        }]
      }
     ]
    }
    Button_CloseTouch {
     id=#2
     possessed_class=CloseButton_C
     tracks [
      {
       type=MovieScene2DTransformTrack
       sections [{
         type=MovieScene2DTransformSection
         end_frame=17
         start_frame=8
         channels [
          Translation.X {type=MovieSceneScriptingFloatChannel, keys [[8.696,40],[8.82,329.5735],[12.824,189.3499],[17,40]]}
          Translation.Y {type=MovieSceneScriptingFloatChannel, const [0,1]}
          Angle {type=MovieSceneScriptingFloatChannel, const [0,1]}
          Scale.X {type=MovieSceneScriptingFloatChannel, const [1,1]}
          Scale.Y {type=MovieSceneScriptingFloatChannel, const [1,1]}
          Shear.X {type=MovieSceneScriptingFloatChannel, const [0,1]}
          Shear.Y {type=MovieSceneScriptingFloatChannel, const [0,1]}
         ]
        }]
      }
      {
       type=MovieSceneFloatTrack
       display_name=Render Opacity
       sections [{
         type=MovieSceneFloatSection
         end_frame=23
         start_frame=0
         channels [None {type=MovieSceneScriptingFloatChannel, keys [[0.0003,0],[12.824,0],[17,1]]}]
        }]
      }
     ]
    }
    UniformGridPanel_2 {
     id=#3
     possessed_class=UniformGridPanel
     tracks [
      {
       type=MovieSceneFloatTrack
       display_name=Render Opacity
       sections [{type=MovieSceneFloatSection, end_frame=24, start_frame=0, channels [None {type=MovieSceneScriptingFloatChannel, keys [[0,0],[10,0],[17,1]]}]}]
      }
      {
       type=MovieScene2DTransformTrack
       sections [{
         type=MovieScene2DTransformSection
         end_frame=24
         start_frame=0
         channels [
          Translation.X {type=MovieSceneScriptingFloatChannel, keys [[10,100],[17,0]]}
          Translation.Y {type=MovieSceneScriptingFloatChannel}
         ]
        }]
      }
     ]
    }
    Skip_Button {
     id=#4
     possessed_class=IconTextButton_C
     tracks [
      {
       type=MovieSceneFloatTrack
       display_name=Render Opacity
       sections [{
         type=MovieSceneFloatSection
         end_frame=24
         start_frame=0
         channels [None {type=MovieSceneScriptingFloatChannel, keys [[0.0003,0],[11,0],[14.602,1]]}]
        }]
      }
      {
       type=MovieScene2DTransformTrack
       sections [{
         type=MovieScene2DTransformSection
         end_frame=24
         start_frame=0
         channels [
          Translation.X {type=MovieSceneScriptingFloatChannel}
          Translation.Y {type=MovieSceneScriptingFloatChannel}
          Angle {type=MovieSceneScriptingFloatChannel}
          Scale.X {type=MovieSceneScriptingFloatChannel, keys [[10.602,1.2],[14.602,0.92],[18.602,1.01],[23.602,1]]}
          Scale.Y {type=MovieSceneScriptingFloatChannel, keys [[10.602,1.2],[14.602,0.92],[18.602,1.01],[23.602,1]]}
          Shear.X {type=MovieSceneScriptingFloatChannel}
          Shear.Y {type=MovieSceneScriptingFloatChannel}
         ]
        }]
      }
     ]
    }
    Header_1 {
     id=#5
     possessed_class=CommonTextBlock
     tracks [
      {
       type=MovieScene2DTransformTrack
       sections [{
         type=MovieScene2DTransformSection
         end_frame=18
         start_frame=0
         channels [
          Translation.X {type=MovieSceneScriptingFloatChannel, keys [[6,100],[10.503,0]]}
          Translation.Y {type=MovieSceneScriptingFloatChannel}
         ]
        }]
      }
      {
       type=MovieSceneFloatTrack
       display_name=Render Opacity
       sections [{type=MovieSceneFloatSection, end_frame=18, start_frame=0, channels [None {type=MovieSceneScriptingFloatChannel, keys [[6,0],[10.503,1]]}]}]
      }
     ]
    }
    Cancel_Button {
     id=#6
     possessed_class=IconTextButton_C
     tracks [
      {
       type=MovieSceneFloatTrack
       display_name=Render Opacity
       sections [{
         type=MovieSceneFloatSection
         end_frame=24
         start_frame=0
         channels [None {type=MovieSceneScriptingFloatChannel, keys [[0,0],[10.9997,0],[14.6017,1]]}]
        }]
      }
      {
       type=MovieScene2DTransformTrack
       sections [{
         type=MovieScene2DTransformSection
         end_frame=24
         start_frame=-1
         channels [
          Translation.X {type=MovieSceneScriptingFloatChannel}
          Translation.Y {type=MovieSceneScriptingFloatChannel}
          Angle {type=MovieSceneScriptingFloatChannel}
          Scale.X {type=MovieSceneScriptingFloatChannel, keys [[10.6017,1.2],[14.6017,0.92],[18.6017,1.01],[23.6017,1]]}
          Scale.Y {type=MovieSceneScriptingFloatChannel, keys [[10.6017,1.2],[14.6017,0.92],[18.6017,1.01],[23.6017,1]]}
          Shear.X {type=MovieSceneScriptingFloatChannel}
          Shear.Y {type=MovieSceneScriptingFloatChannel}
         ]
        }]
      }
     ]
    }
    BackgroundBlur_1 {
     id=#7
     possessed_class=BackgroundBlur
     tracks [
      {
       type=MovieSceneFloatTrack
       display_name=Blur Strength
       sections [{
         type=MovieSceneFloatSection
         end_frame=18
         start_frame=0
         channels [None {type=MovieSceneScriptingFloatChannel, keys [[0,0],[3.602,20],[5.403,20]]}]
        }]
      }
      {
       type=MovieSceneFloatTrack
       display_name=Render Opacity
       sections [{
         type=MovieSceneFloatSection
         end_frame=18
         start_frame=0
         channels [None {type=MovieSceneScriptingFloatChannel, keys [[0,0],[3.6023,1],[5.4033,1]]}]
        }]
      }
     ]
    }
    ModalOverlay {
     id=#8
     possessed_class=Overlay
     tracks [
      {
       type=MovieScene2DTransformTrack
       sections [{
         type=MovieScene2DTransformSection
         end_frame=18
         start_frame=-1
         channels [
          Translation.X {type=MovieSceneScriptingFloatChannel}
          Translation.Y {type=MovieSceneScriptingFloatChannel}
          Angle {type=MovieSceneScriptingFloatChannel}
          Scale.X {type=MovieSceneScriptingFloatChannel, keys [[0,0],[4,1.15],[11,1]]}
          Scale.Y {type=MovieSceneScriptingFloatChannel, keys [[0,0],[4,1.15],[11,1]]}
          Shear.X {type=MovieSceneScriptingFloatChannel}
          Shear.Y {type=MovieSceneScriptingFloatChannel}
         ]
        }]
      }
      {
       type=MovieSceneFloatTrack
       display_name=Render Opacity
       sections [{type=MovieSceneFloatSection, end_frame=18, start_frame=0, channels [None {type=MovieSceneScriptingFloatChannel, const [1,2]}]}]
      }
     ]
    }
    Border_Darken {
     id=#9
     possessed_class=Border
     tracks [{
       type=MovieSceneFloatTrack
       display_name=Render Opacity
       sections [{
         type=MovieSceneFloatSection
         end_frame=18
         start_frame=0
         channels [None {type=MovieSceneScriptingFloatChannel, keys [[0,0],[3.602,1],[5.403,1]]}]
        }]
      }]
    }
   ]}
 }
 Out {
  binding_count=3
  length_seconds=0.3187
  object=&self:Out
  sequence {bindings [
    BackgroundBlur_1 {
     id=#7
     possessed_class=BackgroundBlur
     tracks [
      {
       type=MovieSceneFloatTrack
       display_name=Blur Strength
       sections [{
         type=MovieSceneFloatSection
         end_frame=6
         start_frame=0
         channels [None {type=MovieSceneScriptingFloatChannel, keys [[0,20],[2,20],[6,0],[6.3433,0]]}]
        }]
      }
      {
       type=MovieSceneFloatTrack
       display_name=Render Opacity
       sections [{
         type=MovieSceneFloatSection
         end_frame=6
         start_frame=0
         channels [None {type=MovieSceneScriptingFloatChannel, keys [[0,1],[2,1],[6,0],[6.3433,0]]}]
        }]
      }
     ]
    }
    ModalOverlay {
     id=#8
     possessed_class=Overlay
     tracks [
      {
       type=MovieScene2DTransformTrack
       sections [{
         type=MovieScene2DTransformSection
         end_frame=6
         start_frame=-1
         channels [
          Translation.X {type=MovieSceneScriptingFloatChannel}
          Translation.Y {type=MovieSceneScriptingFloatChannel}
          Angle {type=MovieSceneScriptingFloatChannel}
          Scale.X {type=MovieSceneScriptingFloatChannel, keys [[0,1],[6.3433,0.5]]}
          Scale.Y {type=MovieSceneScriptingFloatChannel, keys [[0,1],[6.3433,0.5]]}
          Shear.X {type=MovieSceneScriptingFloatChannel}
          Shear.Y {type=MovieSceneScriptingFloatChannel}
         ]
        }]
      }
      {
       type=MovieSceneFloatTrack
       display_name=Render Opacity
       sections [{type=MovieSceneFloatSection, end_frame=6, start_frame=0, channels [None {type=MovieSceneScriptingFloatChannel, keys [[0,1],[6.369,0]]}]}]
      }
     ]
    }
    Border_Darken {
     id=#9
     possessed_class=Border
     tracks [{
       type=MovieSceneFloatTrack
       display_name=Render Opacity
       sections [{
         type=MovieSceneFloatSection
         end_frame=6
         start_frame=0
         channels [None {type=MovieSceneScriptingFloatChannel, keys [[0,1],[2,1],[6,0],[6.3433,0]]}]
        }]
      }]
    }
   ]}
 }
]
graphs [EventGraph {nodes [
   n1 = "On Button Clicked (Button_CloseTouch)"!() -> (OutputDelegate, then, Button:Common Button Legacy Object Reference) {is=title K2Node_ComponentBoundEvent}
   PlayAnimation(execute=[@n1.then, @n2.then], InAnimation=Get Out() -> (Out:Widget Animation Object Reference))
   > BroadcastTutorialPopupCanceled(self:Fort Athena Tutorial Screen Object Reference)
   "On Button Clicked (Skip_Button)"!() -> (OutputDelegate, then, Button:Common Button Legacy Object Reference) {is=title K2Node_ComponentBoundEvent}
   > SkipGuidedTutorial(self:Fort Athena Tutorial Screen Object Reference)
   > Execute Client Event!(
    execute:Exec
    EventType:Fort Client Event Name Structure=(CategoryName="UserInterface",EventName="LeaveZone")
    EventSource:Object Reference=Self-Reference()
    EventFocus:Object Reference=Self-Reference()
   ) -> (then:Exec) {is=title K2Node_ExecuteClientEvent}
   > SendLeaveZoneAnalytic(self:Fort Global UIContext Object Reference=Get FortGlobalUIContext() -> (ReturnValue:Fort Global UIContext Object Reference) {is=title K2Node_GetSubsystem})
   > LeaveAction!(
    execute
    self:Fort Party Context Object Reference=Get FortPartyContext!() -> (ReturnValue:Fort Party Context Object Reference) {is=title K2Node_GetBlueprintContextEx}
    bInStayWithParty:bool=false
   ) -> (then)
   n2 = "On Button Clicked (Cancel_Button)"!() -> (OutputDelegate, then, Button:Common Button Legacy Object Reference) {is=title K2Node_ComponentBoundEvent}
   Event Construct()
   > n3 = For Each Loop(Array:Array of Athena Modal Popup Steps Object References=Make Array!(
     "[0]":Athena Modal Popup Steps Object Reference=Get Athena_ModalPopup_Steps_1() -> (Athena_ModalPopup_Steps_1:Athena Modal Popup Steps Object Reference)
     "[1]":Athena Modal Popup Steps Object Reference=Get Athena_ModalPopup_Steps_2() -> (Athena_ModalPopup_Steps_2:Athena Modal Popup Steps Object Reference)
     "[2]":Athena Modal Popup Steps Object Reference=Get Athena_ModalPopup_Steps_3() -> (Athena_ModalPopup_Steps_3:Athena Modal Popup Steps Object Reference)
     "[3]":Athena Modal Popup Steps Object Reference=Get Athena_ModalPopup_Steps_4() -> (Athena_ModalPopup_Steps_4:Athena Modal Popup Steps Object Reference)
     "[4]":Athena Modal Popup Steps Object Reference=Get Athena_ModalPopup_Steps_5() -> (Athena_ModalPopup_Steps_5:Athena Modal Popup Steps Object Reference)
     "[5]":Athena Modal Popup Steps Object Reference=Get Athena_ModalPopup_Steps_6() -> (Athena_ModalPopup_Steps_6:Athena Modal Popup Steps Object Reference)
     "[6]":Athena Modal Popup Steps Object Reference=Get Athena_ModalPopup_Steps_7() -> (Athena_ModalPopup_Steps_7:Athena Modal Popup Steps Object Reference)
    ) -> (Array:Array of Athena Modal Popup Steps Object References)) -> (Array Element:Athena Modal Popup Steps Object Reference)
   > SetVisibility(
    self=Get Checkmark_Image!(self:Fort Athena Tutorial Skip Popup Element Object Reference=@n3) -> (Checkmark_Image:Image Object Reference)
    InVisibility=Collapsed
   )
   n4 = Event OnAnimationStarted!() -> (OutputDelegate, then, Animation:Widget Animation Object Reference) {is=title K2Node_Event}
   > n5 = Branch(Condition="Equal (Object)"(A=@n4.Animation, B=Get In() -> (In:Widget Animation Object Reference)) {is=title K2Node_CallFunction})
   > PlaySound2D(Sound=/Game/Sounds/Fort_UI_Sounds/Tutorial/UI_Tutorial_Popup_FrontEnd_Cue)
   PlaySound2D(execute=@n5.else, Sound=/Game/Sounds/Fort_UI_Sounds/Tutorial/UI_Tutorial_Close_FrontEnd_Cue)
   n6 = Event OnTutorialProgressUpdated!() -> (OutputDelegate, then, CurrentPhase:int=0) {is=title K2Node_Event}
   > n7 = For Each Loop(Array=Make Array(
     "[0]"=Get Athena_ModalPopup_Steps_1()
     "[1]"=Get Athena_ModalPopup_Steps_2()
     "[2]"=Get Athena_ModalPopup_Steps_3()
     "[3]"=Get Athena_ModalPopup_Steps_4()
     "[4]"=Get Athena_ModalPopup_Steps_5()
     "[5]"=Get Athena_ModalPopup_Steps_6()
     "[6]"=Get Athena_ModalPopup_Steps_7()
    ))
   > SetVisibility(
    self=Get Checkmark_Image(self=@n7)
    InVisibility=Select(
     Option 0=Collapsed
     Option 1=SelfHitTestInvisible
     Index="integer < integer"!(A:int=@n7.Array Index, B:int=@n6.CurrentPhase) -> (ReturnValue:bool) {is=title K2Node_CallFunction}
    )
   )
   "Animation Finished (Out)"!() -> (OutputDelegate:Delegate, then:Exec) {is=title K2Node_WidgetAnimationEvent}
   > SetVisibility(InVisibility=Collapsed)
  ]}]
variables []
widgets {
 info {inheritedWidgetCount=5, parentClass=/Script/FortniteUI.FortAthenaTutorialScreen, rootWidgetClass=&CanvasPanel, widgetCount=31}
 widgets [
  CanvasPanel_18 {widgetClassPath=&CanvasPanel, widget=CanvasPanel_18}
  Overlay_181 {
   widgetClassPath=&Overlay
   parent=CanvasPanel_18
   widget=Overlay_181
   slot {class=CanvasPanelSlot, properties {layoutData {alignment=<0 0>, anchors {maximum=<1 1>, minimum=<0 0>}, offsets=<ltrb 0 0 0 0>}}}
  }
  BackgroundBlur_1 {
   widgetClassPath=&BackgroundBlur
   parent=Overlay_181
   widget=BackgroundBlur_1
   properties {
    blurStrength=20
    lowQualityFallbackBrush {
     bIsDynamicallyLoaded=false
     drawAs=Image
     imageType=NoImage
     mirroring=NoMirror
     resourceName=None
     resourceObject=/Game/UI/Foundation/Textures/LoadingScreens/BacchusTutorial/LoadingScreenBG01
     tiling=NoTile
     imageSize=<1024 768>
     margin=<ltrb 0 0 0 0>
     outlineSettings {
      bUseBrushTransparency=false
      roundingType=HalfHeightRadius
      width=0
      color {colorUseRule=UseColor_Specified, specifiedColor=<rgba 0 0 0 0>}
      cornerRadii=<0 0 0 0>
     }
     tintColor {colorUseRule=UseColor_Specified, specifiedColor=<rgba 0 0 0 0.33>}
     uVRegion {bIsValid=false, max=<0 0>, min=<0 0>}
    }
   }
   slot {class=OverlaySlot, properties {horizontalAlignment=HAlign_Fill, verticalAlignment=VAlign_Fill}}
  }
  Border_Darken {
   widgetClassPath=&Border
   bIsVariable=true
   parent=Overlay_181
   widget=Border_Darken
   properties {brushColor=<rgba 0.0052 0.0052 0.0056 0.1571>, padding=<ltrb 0 0 0 0>}
   slot {class=OverlaySlot, properties {horizontalAlignment=HAlign_Fill, verticalAlignment=VAlign_Fill}}
  }
  SizeBox_337 {
   widgetClassPath=&SizeBox
   parent=Overlay_181
   widget=SizeBox_337
   properties {bOverride_HeightOverride=true, bOverride_WidthOverride=true, heightOverride=740, widthOverride=1300}
   slot {class=OverlaySlot, properties {horizontalAlignment=HAlign_Center, verticalAlignment=VAlign_Center}}
  }
  ModalOverlay {widgetClassPath=&Overlay, parent=SizeBox_337, widget=ModalOverlay, slot {class=SizeBoxSlot}}
  ModalContent {
   widgetClassPath=&CommonBorder
   parent=ModalOverlay
   widget=ModalContent
   properties {
    background {
     bIsDynamicallyLoaded=false
     drawAs=Image
     imageType=NoImage
     mirroring=NoMirror
     resourceName=None
     resourceObject=/Game/Athena/UI/Frontend/Art/M_UI_RadialGradientMaster_Translucent_Creative
     tiling=NoTile
     imageSize=<32 32>
     margin=<ltrb 0 0 0 0>
     outlineSettings {
      bUseBrushTransparency=false
      roundingType=HalfHeightRadius
      width=0
      color {colorUseRule=UseColor_Specified, specifiedColor=<rgba 0 0 0 0>}
      cornerRadii=<0 0 0 0>
     }
     tintColor {colorUseRule=UseColor_Specified, specifiedColor=<rgba 1 1 1 1>}
     uVRegion {bIsValid=false, max=<0 0>, min=<0 0>}
    }
    padding=<ltrb 0 0 0 0>
    renderTransform {angle=-0.6, scale=<1 1>, shear=<-3 0>, translation=<0 0>}
   }
   slot {class=OverlaySlot, properties {horizontalAlignment=HAlign_Fill, verticalAlignment=VAlign_Fill}}
  }
  Modal_Triangles {
   widgetClassPath=&CommonBorder
   parent=ModalOverlay
   widget=Modal_Triangles
   properties {
    background {
     bIsDynamicallyLoaded=false
     drawAs=Image
     imageType=NoImage
     mirroring=NoMirror
     resourceName=None
     resourceObject=/Game/UI/Frontend/ItemShop/Com/vfx/MI_Pattern_Triangles
     tiling=Both
     imageSize=<1500 1500>
     margin=<ltrb 0 0 0 0>
     outlineSettings {
      bUseBrushTransparency=false
      roundingType=HalfHeightRadius
      width=0
      color {colorUseRule=UseColor_Specified, specifiedColor=<rgba 0 0 0 0>}
      cornerRadii=<0 0 0 0>
     }
     tintColor {colorUseRule=UseColor_Specified, specifiedColor=<rgba 1 1 1 1>}
     uVRegion {bIsValid=false, max=<0 0>, min=<0 0>}
    }
    padding=<ltrb 0 0 0 0>
    renderTransform {angle=-0.6, scale=<1 1>, shear=<-3 0>, translation=<0 0>}
   }
   slot {class=OverlaySlot, properties {horizontalAlignment=HAlign_Fill, verticalAlignment=VAlign_Fill}}
  }
  VerticalBox_2 {
   widgetClassPath=&VerticalBox
   parent=ModalOverlay
   widget=VerticalBox_2
   slot {class=OverlaySlot, properties {horizontalAlignment=HAlign_Fill, verticalAlignment=VAlign_Fill, padding=<ltrb 50 20 50 40>}}
  }
  HorizontalBox_0 {
   widgetClassPath=&HorizontalBox
   parent=VerticalBox_2
   widget=HorizontalBox_0
   slot {class=VerticalBoxSlot, properties {verticalAlignment=VAlign_Top}}
  }
  Header_1 {
   widgetClassPath=&CommonTextBlock
   bIsVariable=true
   parent=HorizontalBox_0
   widget=Header_1
   properties {
    style=/FortUILibrary/Styles/Typography/Foundation/TextStyle-Header-XL.TextStyle-Header-XL_C
    text=SKIP TUTORIAL
    font {
     bForceMonospaced=false
     bMaterialIsStencil=false
     fontMaterial=null
     fontObject=&Burbank
     letterSpacing=0
     monospacedWidth=1
     size=72
     skewAmount=0
     typefaceFontName=Burbank-Black
     outlineSettings {
      bApplyOutlineToDropShadows=false
      bMiteredCorners=false
      bSeparateFillAlpha=false
      outlineMaterial=null
      outlineSize=0
      outlineColor=<rgba 0 0 0 1>
     }
    }
    shadowOffset=<0 0>
   }
   slot {class=HorizontalBoxSlot}
  }
  VisibilityBorder_Touch {
   widgetClassPath=&CommonHardwareVisibilityBorder
   parent=HorizontalBox_0
   widget=VisibilityBorder_Touch
   properties {
    padding=<ltrb 4 2 4 2>
    renderTransform {angle=0, scale=<1 1>, shear=<0 0>, translation=<40 0>}
    visibilityQuery {
     autoDescription=" ANY( Input.Touch )"
     tokenStreamVersion=0
     userDescription=Touch
     queryTokenStream [0,1,1,1,0]
     tagDictionary [{tagName=Input.Touch}]
    }
   }
   slot {class=HorizontalBoxSlot, properties {horizontalAlignment=HAlign_Right, verticalAlignment=VAlign_Top, size {sizeRule=Fill, value=1}}}
  }
  Button_CloseTouch {
   widgetClassPath=/FortUILibrary/Foundation/Buttons/CloseButton
   bIsVariable=true
   parent=VisibilityBorder_Touch
   widget=Button_CloseTouch
   properties {
    onButtonBaseClicked=(&self:WidgetTree.Button_CloseTouch.HandleOnButtonClicked)
    onButtonBaseDoubleClicked=(&self:WidgetTree.Button_CloseTouch.HandleOnButtonDoubleClicked)
    onButtonBaseHovered=(&self:WidgetTree.Button_CloseTouch.HandleOnButtonHovered)
    onButtonBaseUnhovered=(&self:WidgetTree.Button_CloseTouch.HandleOnButtonUnhovered)
    onSelectedChangedBase=(&self:WidgetTree.Button_CloseTouch.HandleOnSelectedChanged)
   }
   slot {class=BorderSlot}
  }
  SizeBox_82 {
   widgetClassPath=&SizeBox
   parent=VerticalBox_2
   widget=SizeBox_82
   properties {maxDesiredHeight=500, widthOverride=850}
   slot {class=VerticalBoxSlot}
  }
  RichText_Desc {
   widgetClassPath=/Script/UMG.RichTextBlock
   bIsVariable=true
   parent=SizeBox_82
   widget=RichText_Desc
   properties {
    text="<S_Blue>Here are the</> <S_LightBlue>SKILLS</> <S_Blue>you are about to learn. Are you sure you want to</> <S_LightBlue>SKIP?</>"
    textStyleSet=/Game/Athena/HUD/LTM/Tutorial/Materials/TutorialCalloutStyles
    wrapTextAt=1225
   }
   slot {class=SizeBoxSlot}
  }
  UniformGridPanel_2 {
   widgetClassPath=/Script/UMG.UniformGridPanel
   parent=VerticalBox_2
   widget=UniformGridPanel_2
   slot {class=VerticalBoxSlot, properties {padding=<ltrb 0 20 0 30>, size {sizeRule=Fill, value=0.9}}}
  }
  Athena_ModalPopup_Steps_1 {
   widgetClassPath=&/Athena_ModalPopup_Steps
   bIsVariable=true
   parent=UniformGridPanel_2
   widget=Athena_ModalPopup_Steps_1
   properties {
    stepDescription=1. Movement
    stepIconSlateBrush {
     bIsDynamicallyLoaded=false
     drawAs=Image
     imageType=NoImage
     mirroring=NoMirror
     resourceName=None
     resourceObject=/Game/Athena/HUD/Art/Tutorial/FTUE_MovementStep
     tiling=NoTile
     imageSize=<54 53>
     margin=<ltrb 0 0 0 0>
     outlineSettings {
      bUseBrushTransparency=false
      roundingType=HalfHeightRadius
      width=0
      color {colorUseRule=UseColor_Specified, specifiedColor=<rgba 0 0 0 0>}
      cornerRadii=<0 0 0 0>
     }
     tintColor {colorUseRule=UseColor_Specified, specifiedColor=<rgba 1 1 1 1>}
     uVRegion {bIsValid=false, max=<0 0>, min=<0 0>}
    }
   }
   slot {class=UniformGridSlot, properties {verticalAlignment=VAlign_Center}}
  }
  Athena_ModalPopup_Steps_2 {
   widgetClassPath=&Athena_ModalPopup_Steps
   bIsVariable=true
   parent=UniformGridPanel_2
   widget=Athena_ModalPopup_Steps_2
   properties {
    stepDescription=2. Harvesting
    stepIconSlateBrush {
     bIsDynamicallyLoaded=false
     drawAs=Image
     imageType=NoImage
     mirroring=NoMirror
     resourceName=None
     resourceObject=/Game/Athena/HUD/Art/BacchusHudIcons/Harvest_Normal
     tiling=NoTile
     imageSize=<72 84>
     margin=<ltrb 0 0 0 0>
     outlineSettings {
      bUseBrushTransparency=false
      roundingType=HalfHeightRadius
      width=0
      color {colorUseRule=UseColor_Specified, specifiedColor=<rgba 0 0 0 0>}
      cornerRadii=<0 0 0 0>
     }
     tintColor {colorUseRule=UseColor_Specified, specifiedColor=<rgba 1 1 1 1>}
     uVRegion {bIsValid=false, max=<0 0>, min=<0 0>}
    }
   }
   slot {class=UniformGridSlot, properties {row=1, verticalAlignment=VAlign_Center}}
  }
  Athena_ModalPopup_Steps_3 {
   widgetClassPath=&Athena_ModalPopup_Steps
   bIsVariable=true
   parent=UniformGridPanel_2
   widget=Athena_ModalPopup_Steps_3
   properties {
    stepDescription=3. Shooting
    stepIconSlateBrush {
     bIsDynamicallyLoaded=false
     drawAs=Image
     imageType=NoImage
     mirroring=NoMirror
     resourceName=None
     resourceObject=/Game/UI/Foundation/Textures/Icons/Workers/T-Icon-Weapon-Skill-64
     tiling=NoTile
     imageSize=<64 64>
     margin=<ltrb 0 0 0 0>
     outlineSettings {
      bUseBrushTransparency=false
      roundingType=HalfHeightRadius
      width=0
      color {colorUseRule=UseColor_Specified, specifiedColor=<rgba 0 0 0 0>}
      cornerRadii=<0 0 0 0>
     }
     tintColor {colorUseRule=UseColor_Specified, specifiedColor=<rgba 1 1 1 1>}
     uVRegion {bIsValid=false, max=<0 0>, min=<0 0>}
    }
   }
   slot {class=UniformGridSlot, properties {row=2, verticalAlignment=VAlign_Center}}
  }
  Athena_ModalPopup_Steps_4 {
   widgetClassPath=&Athena_ModalPopup_Steps
   bIsVariable=true
   parent=UniformGridPanel_2
   widget=Athena_ModalPopup_Steps_4
   properties {
    stepDescription=4. Healing
    stepIconSlateBrush {
     bIsDynamicallyLoaded=false
     drawAs=Image
     imageType=NoImage
     mirroring=NoMirror
     resourceName=None
     resourceObject=/Game/Athena/HUD/Art/cross_Shadow
     tiling=NoTile
     imageSize=<64 64>
     margin=<ltrb 0 0 0 0>
     outlineSettings {
      bUseBrushTransparency=false
      roundingType=HalfHeightRadius
      width=0
      color {colorUseRule=UseColor_Specified, specifiedColor=<rgba 0 0 0 0>}
      cornerRadii=<0 0 0 0>
     }
     tintColor {colorUseRule=UseColor_Specified, specifiedColor=<rgba 1 1 1 1>}
     uVRegion {bIsValid=false, max=<0 0>, min=<0 0>}
    }
   }
   slot {class=UniformGridSlot, properties {row=3, verticalAlignment=VAlign_Center}}
  }
  Athena_ModalPopup_Steps_5 {
   widgetClassPath=&Athena_ModalPopup_Steps
   bIsVariable=true
   parent=UniformGridPanel_2
   widget=Athena_ModalPopup_Steps_5
   properties {
    stepDescription=5. Building
    stepIconSlateBrush {
     bIsDynamicallyLoaded=false
     drawAs=Image
     imageType=NoImage
     mirroring=NoMirror
     resourceName=None
     resourceObject=/Game/UI/Foundation/Textures/Icons/GamepadOptions/BuildMode
     tiling=NoTile
     imageSize=<64 64>
     margin=<ltrb 0 0 0 0>
     outlineSettings {
      bUseBrushTransparency=false
      roundingType=HalfHeightRadius
      width=0
      color {colorUseRule=UseColor_Specified, specifiedColor=<rgba 0 0 0 0>}
      cornerRadii=<0 0 0 0>
     }
     tintColor {colorUseRule=UseColor_Specified, specifiedColor=<rgba 1 1 1 1>}
     uVRegion {bIsValid=false, max=<0 0>, min=<0 0>}
    }
   }
   slot {class=UniformGridSlot, properties {column=1, horizontalAlignment=HAlign_Right, verticalAlignment=VAlign_Center}}
  }
  Athena_ModalPopup_Steps_6 {
   widgetClassPath=&Athena_ModalPopup_Steps
   bIsVariable=true
   parent=UniformGridPanel_2
   widget=Athena_ModalPopup_Steps_6
   properties {
    stepDescription=6. Loot chests
    stepIconSlateBrush {
     bIsDynamicallyLoaded=false
     drawAs=Image
     imageType=NoImage
     mirroring=NoMirror
     resourceName=None
     resourceObject=/Game/Athena/HUD/Art/BacchusHudIcons/Open_Box
     tiling=NoTile
     imageSize=<88 92>
     margin=<ltrb 0 0 0 0>
     outlineSettings {
      bUseBrushTransparency=false
      roundingType=HalfHeightRadius
      width=0
      color {colorUseRule=UseColor_Specified, specifiedColor=<rgba 0 0 0 0>}
      cornerRadii=<0 0 0 0>
     }
     tintColor {colorUseRule=UseColor_Specified, specifiedColor=<rgba 1 1 1 1>}
     uVRegion {bIsValid=false, max=<0 0>, min=<0 0>}
    }
   }
   slot {class=UniformGridSlot, properties {column=1, horizontalAlignment=HAlign_Right, row=1, verticalAlignment=VAlign_Center}}
  }
  Athena_ModalPopup_Steps_7 {
   widgetClassPath=&Athena_ModalPopup_Steps
   bIsVariable=true
   parent=UniformGridPanel_2
   widget=Athena_ModalPopup_Steps_7
   properties {
    stepDescription=7. Aim Down Sights
    stepIconSlateBrush {
     bIsDynamicallyLoaded=false
     drawAs=Image
     imageType=NoImage
     mirroring=NoMirror
     resourceName=None
     resourceObject=/Game/Athena/HUD/Art/Tutorial/FTUE_TargetCounter
     tiling=NoTile
     imageSize=<86 86>
     margin=<ltrb 0 0 0 0>
     outlineSettings {
      bUseBrushTransparency=false
      roundingType=HalfHeightRadius
      width=0
      color {colorUseRule=UseColor_Specified, specifiedColor=<rgba 0 0 0 0>}
      cornerRadii=<0 0 0 0>
     }
     tintColor {colorUseRule=UseColor_Specified, specifiedColor=<rgba 1 1 1 1>}
     uVRegion {bIsValid=false, max=<0 0>, min=<0 0>}
    }
   }
   slot {class=UniformGridSlot, properties {column=1, horizontalAlignment=HAlign_Right, row=2, verticalAlignment=VAlign_Center}}
  }
  HorizontalBox_3 {
   widgetClassPath=&HorizontalBox
   parent=VerticalBox_2
   widget=HorizontalBox_3
   slot {class=VerticalBoxSlot, properties {horizontalAlignment=HAlign_Right, verticalAlignment=VAlign_Bottom}}
  }
  Cancel_Button {
   widgetClassPath=&IconTextButton
   bIsVariable=true
   parent=HorizontalBox_3
   widget=Cancel_Button
   properties {
    buttonText=CANCEL
    contentAlignment=HAlign_Center
    controllerInputStyle=/FortUILibrary/Foundation/Buttons/ButtonStyles/Athena/Athena_ButtonStyle_AngledBlueMenuButton.Athena_ButtonStyle_AngledBlueMenuButton_C
    inputModeOverride=Game
    inputPriority=6
    minHeight=48
    minWidth=250
    onButtonBaseClicked=(&self:WidgetTree.Cancel_Button.HandleOnButtonClicked)
    onButtonBaseDoubleClicked=(&self:WidgetTree.Cancel_Button.HandleOnButtonDoubleClicked)
    onButtonBaseHovered=(&self:WidgetTree.Cancel_Button.HandleOnButtonHovered)
    onButtonBaseUnhovered=(&self:WidgetTree.Cancel_Button.HandleOnButtonUnhovered)
    onSelectedChangedBase=(&self:WidgetTree.Cancel_Button.HandleOnSelectedChanged)
    priority=10000
    style=&Athena_ButtonStyle_AngledBlueMenuButton_C
    triggeringInputAction {dataTable=&/Tutorial_InputActionDataTable, rowName=Tutorial_SkipPopup_Cancel}
   }
   slot {class=HorizontalBoxSlot, properties {padding=<ltrb 0 0 20 0>}}
  }
  Skip_Button {
   widgetClassPath=&IconTextButton
   bIsVariable=true
   parent=HorizontalBox_3
   widget=Skip_Button
   properties {
    buttonText=SKIP
    contentAlignment=HAlign_Center
    controllerInputStyle=/FortUILibrary/Foundation/Buttons/ButtonStyles/ButtonStyle_Feature_M_Yellow_BurbSmall.ButtonStyle_Feature_M_Yellow_BurbSmall_C
    inputModeOverride=Game
    inputPriority=5
    minHeight=48
    minWidth=250
    onButtonBaseClicked=(&self:WidgetTree.Skip_Button.HandleOnButtonClicked)
    onButtonBaseDoubleClicked=(&self:WidgetTree.Skip_Button.HandleOnButtonDoubleClicked)
    onButtonBaseHovered=(&self:WidgetTree.Skip_Button.HandleOnButtonHovered)
    onButtonBaseUnhovered=(&self:WidgetTree.Skip_Button.HandleOnButtonUnhovered)
    onSelectedChangedBase=(&self:WidgetTree.Skip_Button.HandleOnSelectedChanged)
    priority=10000
    style=&ButtonStyle_Feature_M_Yellow_BurbSmall_C
    triggeringInputAction {dataTable=&Tutorial_InputActionDataTable, rowName=Tutorial_EndTutorial}
   }
   slot {class=HorizontalBoxSlot, properties {padding=<ltrb 20 0 0 0>}}
  }
  Interaction_TutorialStep {widgetClassPath=/Script/FortniteUI.FortAthenaTutorialBaseInteraction, bInherited=true, widget=null}
  InteractionRegular_TutorialStep {widgetClassPath=&FortAthenaTutorialBaseInteraction, bInherited=true, widget=null}
  MoveableHighlight {widgetClassPath=&UserWidget, bInherited=true, widget=null}
  MoveableHighlight_Regular {widgetClassPath=&UserWidget, bInherited=true, widget=null}
  Switcher {widgetClassPath=&CommonWidgetSwitcherLegacy, bInherited=true, widget=null}
 ]
}
