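# M_WinGlow_White is an additive UI material for a soft radial glow in the Athena victory HUD, white by default.
# A RadialGradientExponential function, driven by Radius and Density parameters with inverted density, supplies the falloff mask.
# That mask is the alpha of a lerp from the unset A input to the Color parameter's RGB, so tint and shape are adjustable through material instances.
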Material /Game/Athena/HUD/Materials/Victory/M_WinGlow_White
graph {
 properties {blendMode=BLEND_Additive, materialDomain=MD_UI, pixelDepthOffsetMode=PDOM_Legacy, usedShadingModels=Default Lit}
 nodes [LinearInterpolate(
   B=VectorParameter() {parameterName=Color, defaultValue=<rgba 1 1 1 1>}.RGB
   Alpha=MaterialFunctionCall(
    Radius=ScalarParameter() {defaultValue=0.5, parameterName=Radius}
    Density=ScalarParameter() {defaultValue=3, parameterName=Density}
    Invert Density=StaticBool() {value=true}
   ) {materialFunction=&RadialGradientExponential}.RadialGradientExponential
  )]
}
