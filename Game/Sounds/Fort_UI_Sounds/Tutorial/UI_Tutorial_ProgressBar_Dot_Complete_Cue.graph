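# SoundCue for the UI sound of a completed dot on the tutorial progress bar, judging by its name.
# A single wave player feeds the root directly with the Tutorial_ProgressBar_Dot_Complete wave, with no randomization or mixing.
# The cue is assigned to the Fort_UI sound class at reduced volume, and its concurrency override settings specify a prevent-new resolution rule.
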SoundCue /Game/Sounds/Fort_UI_Sounds/Tutorial/UI_Tutorial_ProgressBar_Dot_Complete_Cue
root=WavePlayer_2
nodes [{id=WavePlayer_2, properties {soundWaveAssetPtr=&/Tutorial_ProgressBar_Dot_Complete}}]
properties {
 maxDistance=2097152
 soundClassObject=&Fort_UI
 volumeMultiplier=0.6
 concurrencyOverrides {resolutionRule=PreventNew}
}
