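# Athena tutorial HUD widget showing a labeled progress bar of seven Tutorial_ProgressDot widgets joined by strokes, built on FortAthenaTutorialProgressBar.
# Construct sets every dot to its Inactive animation, and ResetProgressBarVisuals activates the first dot.
# StartNewProgressBarPhase completes the dot before CurrentPhase, then activates that phase's dot if valid, each with a sound.
# Both events bracket these steps with the In_Out fade-and-scale animation, using Delays for pacing.
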WidgetBlueprint /Game/Athena/HUD/Tutorial/Athena_Tutorial_ProgressBar
animations [In_Out {
  binding_count=2
  length_seconds=0.75
  object=&self:In_Out
  sequence {bindings [
    TutorialProgress {
     id=#1
     possessed_class=RichTextBlock
     tracks [{
       type=MovieScene2DTransformTrack
       sections [{
         type=MovieScene2DTransformSection
         end_frame=10
         start_frame=0
         channels [
          Translation.X {type=MovieSceneScriptingFloatChannel}
          Translation.Y {type=MovieSceneScriptingFloatChannel}
          Angle {type=MovieSceneScriptingFloatChannel}
          Scale.X {type=MovieSceneScriptingFloatChannel, keys [[0,0.8],[10,1.2]]}
          Scale.Y {type=MovieSceneScriptingFloatChannel, keys [[0,0.8],[10,1.2]]}
          Shear.X {type=MovieSceneScriptingFloatChannel}
          Shear.Y {type=MovieSceneScriptingFloatChannel}
         ]
        }]
      }]
    }
    Overlay_176 {
     id=#2
     possessed_class=Overlay
     tracks [{
       type=MovieSceneFloatTrack
       display_name=Render Opacity
       sections [{type=MovieSceneFloatSection, end_frame=10, start_frame=-5, channels [None {type=MovieSceneScriptingFloatChannel, keys [[0,0.25],[10,1]]}]}]
      }]
    }
   ]}
 }]
graphs [EventGraph {nodes [
   Event Construct()
   > n1 = For Each Loop(Array:Array of Tutorial Progress Dot Object References=Make Array!(
     "[0]":Tutorial Progress Dot Object Reference=Get Tutorial_ProgressDot() -> (Tutorial_ProgressDot:Tutorial Progress Dot Object Reference)
     "[1]":Tutorial Progress Dot Object Reference=Get Tutorial_ProgressDot_1() -> (Tutorial_ProgressDot_1:Tutorial Progress Dot Object Reference)
     "[2]":Tutorial Progress Dot Object Reference=Get Tutorial_ProgressDot_2() -> (Tutorial_ProgressDot_2:Tutorial Progress Dot Object Reference)
     "[3]":Tutorial Progress Dot Object Reference=Get Tutorial_ProgressDot_3() -> (Tutorial_ProgressDot_3:Tutorial Progress Dot Object Reference)
     "[4]":Tutorial Progress Dot Object Reference=Get Tutorial_ProgressDot_4() -> (Tutorial_ProgressDot_4:Tutorial Progress Dot Object Reference)
     "[5]":Tutorial Progress Dot Object Reference=Get Tutorial_ProgressDot_5() -> (Tutorial_ProgressDot_5:Tutorial Progress Dot Object Reference)
     "[6]":Tutorial Progress Dot Object Reference=Get Tutorial_ProgressDot_6() -> (Tutorial_ProgressDot_6:Tutorial Progress Dot Object Reference)
    ) -> (Array:Array of Tutorial Progress Dot Object References)) -> (Array Element:Tutorial Progress Dot Object Reference)
   > PlayAnimation(self=@n1, InAnimation=Get Inactive!(self:Tutorial Progress Dot Object Reference=@n1) -> (Inactive:Widget Animation Object Reference))
   Event ResetProgressBarVisuals() {is=title K2Node_Event}
   > PlayAnimation(InAnimation=Get In_Out() -> (In_Out:Widget Animation Object Reference))
   > Delay(Duration=1.5)
   n2 = Get Tutorial_ProgressDot()
   > PlayAnimation(self=@n2, InAnimation=Get Active!(self:Tutorial Progress Dot Object Reference=@n2) -> (Active:Widget Animation Object Reference))
   > PlaySound2D(Sound=Get Active Sound() -> (Active Sound:Sound Base Object Reference))
   > Delay(Duration=2)
   > PlayAnimation(InAnimation=Get In_Out(), PlayMode=Reverse)
   n3 = Event StartNewProgressBarPhase!() -> (OutputDelegate, then, CurrentPhase:int=0) {is=title K2Node_Event}
   > PlayAnimation(InAnimation=Get In_Out())
   > Delay(Duration=1.5)
   n4 = Make Array(
    "[0]"=Get Tutorial_ProgressDot()
    "[1]"=Get Tutorial_ProgressDot_1()
    "[2]"=Get Tutorial_ProgressDot_2()
    "[3]"=Get Tutorial_ProgressDot_3()
    "[4]"=Get Tutorial_ProgressDot_4()
    "[5]"=Get Tutorial_ProgressDot_5()
    "[6]"=Get Tutorial_ProgressDot_6()
   )
   n5 = "Get (a copy)"(Array:Array of Tutorial Progress Dot Object References=@n4, Dimension 1=int - int(A=@n3.CurrentPhase, B=1) {is=title K2Node_CallFunction}) -> (Output:Tutorial Progress Dot Object Reference)
   > PlayAnimation(self=@n5, InAnimation=Get Completed!(self:Tutorial Progress Dot Object Reference=@n5) -> (Completed:Widget Animation Object Reference))
   > PlaySound2D(Sound=Get Complete Sound() -> (Complete Sound:Sound Base Object Reference))
   > Delay(Duration=1.5)
   n6 = "Get (a copy)"(Array=@n4, Dimension 1=@n3.CurrentPhase)
   > Branch(Condition=IsValid(Object=@n6))
   > PlayAnimation(self=@n6, InAnimation=Get Active(self=@n6))
   > PlaySound2D(Sound=Get Active Sound())
   > Delay(Duration=2)
   > PlayAnimation(InAnimation=Get In_Out(), PlayMode=Reverse)
   Event PreConstruct()
   Event Tick()
  ]}]
variables [
 Active Sound {type=SoundBase, default=/Game/Sounds/Fort_UI_Sounds/Tutorial/UI_Tutorial_ProgressBar_Dot_Active_Cue}
 Complete Sound {type=SoundBase, default=/Game/Sounds/Fort_UI_Sounds/Tutorial/UI_Tutorial_ProgressBar_Dot_Complete_Cue}
]
widgets {
 info {inheritedWidgetCount=0, parentClass=/Script/FortniteUI.FortAthenaTutorialProgressBar, rootWidgetClass=&Overlay, widgetCount=17}
 widgets [
  Overlay_176 {widgetClassPath=&Overlay, widget=Overlay_176, properties {renderOpacity=0.25}}
  HorizontalBox_Strokes {
   widgetClassPath=&HorizontalBox
   parent=Overlay_176
   widget=HorizontalBox_Strokes
   slot {class=OverlaySlot, properties {horizontalAlignment=HAlign_Center, verticalAlignment=VAlign_Center}}
  }
  Stroke1 {
   widgetClassPath=&Image
   bIsVariable=true
   parent=HorizontalBox_Strokes
   widget=Stroke1
   properties {brush {
     bIsDynamicallyLoaded=false
     drawAs=Image
     imageType=NoImage
     mirroring=NoMirror
     resourceName=None
     resourceObject=null
     tiling=NoTile
     imageSize=<33 6>
     margin=<ltrb 0 0 0 0>
     outlineSettings {
      bUseBrushTransparency=false
      roundingType=HalfHeightRadius
      width=0
      color {colorUseRule=UseColor_Specified, specifiedColor=<rgba 0 0 0 0>}
      cornerRadii=<0 0 0 0>
     }
     tintColor {colorUseRule=UseColor_Specified, specifiedColor=<rgba 1 1 1 1>}
     uVRegion {bIsValid=false, max=<0 0>, min=<0 0>}
    }}
   slot {class=HorizontalBoxSlot, properties {verticalAlignment=VAlign_Center}}
  }
  Stroke2 {
   widgetClassPath=&Image
   bIsVariable=true
   parent=HorizontalBox_Strokes
   widget=Stroke2
   properties {brush {
     bIsDynamicallyLoaded=false
     drawAs=Image
     imageType=NoImage
     mirroring=NoMirror
     resourceName=None
     resourceObject=null
     tiling=NoTile
     imageSize=<33 6>
     margin=<ltrb 0 0 0 0>
     outlineSettings {
      bUseBrushTransparency=false
      roundingType=HalfHeightRadius
      width=0
      color {colorUseRule=UseColor_Specified, specifiedColor=<rgba 0 0 0 0>}
      cornerRadii=<0 0 0 0>
     }
     tintColor {colorUseRule=UseColor_Specified, specifiedColor=<rgba 1 1 1 1>}
     uVRegion {bIsValid=false, max=<0 0>, min=<0 0>}
    }}
   slot {class=HorizontalBoxSlot, properties {verticalAlignment=VAlign_Center, padding=<ltrb 48 0 0 0>}}
  }
  Stroke3 {
   widgetClassPath=&Image
   bIsVariable=true
   parent=HorizontalBox_Strokes
   widget=Stroke3
   properties {brush {
     bIsDynamicallyLoaded=false
     drawAs=Image
     imageType=NoImage
     mirroring=NoMirror
     resourceName=None
     resourceObject=null
     tiling=NoTile
     imageSize=<33 6>
     margin=<ltrb 0 0 0 0>
     outlineSettings {
      bUseBrushTransparency=false
      roundingType=HalfHeightRadius
      width=0
      color {colorUseRule=UseColor_Specified, specifiedColor=<rgba 0 0 0 0>}
      cornerRadii=<0 0 0 0>
     }
     tintColor {colorUseRule=UseColor_Specified, specifiedColor=<rgba 1 1 1 1>}
     uVRegion {bIsValid=false, max=<0 0>, min=<0 0>}
    }}
   slot {class=HorizontalBoxSlot, properties {verticalAlignment=VAlign_Center, padding=<ltrb 48 0 0 0>}}
  }
  Stroke4 {
   widgetClassPath=&Image
   bIsVariable=true
   parent=HorizontalBox_Strokes
   widget=Stroke4
   properties {brush {
     bIsDynamicallyLoaded=false
     drawAs=Image
     imageType=NoImage
     mirroring=NoMirror
     resourceName=None
     resourceObject=null
     tiling=NoTile
     imageSize=<33 6>
     margin=<ltrb 0 0 0 0>
     outlineSettings {
      bUseBrushTransparency=false
      roundingType=HalfHeightRadius
      width=0
      color {colorUseRule=UseColor_Specified, specifiedColor=<rgba 0 0 0 0>}
      cornerRadii=<0 0 0 0>
     }
     tintColor {colorUseRule=UseColor_Specified, specifiedColor=<rgba 1 1 1 1>}
     uVRegion {bIsValid=false, max=<0 0>, min=<0 0>}
    }}
   slot {class=HorizontalBoxSlot, properties {verticalAlignment=VAlign_Center, padding=<ltrb 48 0 0 0>}}
  }
  Stroke5 {
   widgetClassPath=&Image
   bIsVariable=true
   parent=HorizontalBox_Strokes
   widget=Stroke5
   properties {brush {
     bIsDynamicallyLoaded=false
     drawAs=Image
     imageType=NoImage
     mirroring=NoMirror
     resourceName=None
     resourceObject=null
     tiling=NoTile
     imageSize=<33 6>
     margin=<ltrb 0 0 0 0>
     outlineSettings {
      bUseBrushTransparency=false
      roundingType=HalfHeightRadius
      width=0
      color {colorUseRule=UseColor_Specified, specifiedColor=<rgba 0 0 0 0>}
      cornerRadii=<0 0 0 0>
     }
     tintColor {colorUseRule=UseColor_Specified, specifiedColor=<rgba 1 1 1 1>}
     uVRegion {bIsValid=false, max=<0 0>, min=<0 0>}
    }}
   slot {class=HorizontalBoxSlot, properties {verticalAlignment=VAlign_Center, padding=<ltrb 48 0 0 0>}}
  }
  Stroke6 {
   widgetClassPath=&Image
   bIsVariable=true
   parent=HorizontalBox_Strokes
   widget=Stroke6
   properties {brush {
     bIsDynamicallyLoaded=false
     drawAs=Image
     imageType=NoImage
     mirroring=NoMirror
     resourceName=None
     resourceObject=null
     tiling=NoTile
     imageSize=<33 6>
     margin=<ltrb 0 0 0 0>
     outlineSettings {
      bUseBrushTransparency=false
      roundingType=HalfHeightRadius
      width=0
      color {colorUseRule=UseColor_Specified, specifiedColor=<rgba 0 0 0 0>}
      cornerRadii=<0 0 0 0>
     }
     tintColor {colorUseRule=UseColor_Specified, specifiedColor=<rgba 1 1 1 1>}
     uVRegion {bIsValid=false, max=<0 0>, min=<0 0>}
    }}
   slot {class=HorizontalBoxSlot, properties {verticalAlignment=VAlign_Center, padding=<ltrb 48 0 0 0>}}
  }
  HorizontalBox_Dots {widgetClassPath=&HorizontalBox, parent=Overlay_176, widget=HorizontalBox_Dots, slot {class=OverlaySlot}}
  Tutorial_ProgressDot {
   widgetClassPath=&/Tutorial_ProgressDot
   bIsVariable=true
   parent=HorizontalBox_Dots
   widget=Tutorial_ProgressDot
   properties {padding=<ltrb 0 0 25 0>}
   slot {class=HorizontalBoxSlot}
  }
  Tutorial_ProgressDot_1 {
   widgetClassPath=&Tutorial_ProgressDot
   bIsVariable=true
   parent=HorizontalBox_Dots
   widget=Tutorial_ProgressDot_1
   properties {padding=<ltrb 0 0 25 0>}
   slot {class=HorizontalBoxSlot}
  }
  Tutorial_ProgressDot_2 {
   widgetClassPath=&Tutorial_ProgressDot
   bIsVariable=true
   parent=HorizontalBox_Dots
   widget=Tutorial_ProgressDot_2
   properties {padding=<ltrb 0 0 25 0>}
   slot {class=HorizontalBoxSlot}
  }
  Tutorial_ProgressDot_3 {
   widgetClassPath=&Tutorial_ProgressDot
   bIsVariable=true
   parent=HorizontalBox_Dots
   widget=Tutorial_ProgressDot_3
   properties {padding=<ltrb 0 0 25 0>}
   slot {class=HorizontalBoxSlot}
  }
  Tutorial_ProgressDot_4 {
   widgetClassPath=&Tutorial_ProgressDot
   bIsVariable=true
   parent=HorizontalBox_Dots
   widget=Tutorial_ProgressDot_4
   properties {padding=<ltrb 0 0 25 0>}
   slot {class=HorizontalBoxSlot}
  }
  Tutorial_ProgressDot_5 {
   widgetClassPath=&Tutorial_ProgressDot
   bIsVariable=true
   parent=HorizontalBox_Dots
   widget=Tutorial_ProgressDot_5
   properties {padding=<ltrb 0 0 25 0>}
   slot {class=HorizontalBoxSlot}
  }
  Tutorial_ProgressDot_6 {
   widgetClassPath=&Tutorial_ProgressDot
   bIsVariable=true
   parent=HorizontalBox_Dots
   widget=Tutorial_ProgressDot_6
   slot {class=HorizontalBoxSlot}
  }
  TutorialProgress {
   widgetClassPath=/Script/UMG.RichTextBlock
   bIsVariable=true
   parent=Overlay_176
   widget=TutorialProgress
   properties {
    text=TUTORIAL PROGRESS
    textStyleSet=/Game/Athena/HUD/LTM/Tutorial/Materials/TutorialCalloutStyles
    renderTransform {angle=0, scale=<1.2 1.2>, shear=<0 0>, translation=<0 -45>}
    renderTransformPivot=<0 1>
   }
   slot {class=OverlaySlot}
  }
 ]
}
