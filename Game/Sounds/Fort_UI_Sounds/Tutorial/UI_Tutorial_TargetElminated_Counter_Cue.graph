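# Sound cue for the tutorial UI's target-eliminated counter, judging by its name.
# A single wave player feeds a Modulator left at its default settings.
# It uses a 2D hit-notification sound class at reduced volume, with concurrency override settings that specify preventing new instances.
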SoundCue /Game/Sounds/Fort_UI_Sounds/Tutorial/UI_Tutorial_TargetElminated_Counter_Cue
root=Modulator_0
nodes [
 {id=Modulator_0, children [WavePlayer_5]}
 {id=WavePlayer_5, properties {soundWaveAssetPtr=&/UI_Tutorial_TargetElminated_Counter_01}}
]
properties {
 maxDistance=2097152
 soundClassObject=&Fort_2DHitNotifications_Instigate_PSM
 volumeMultiplier=0.6
 concurrencyOverrides {resolutionRule=PreventNew}
}
