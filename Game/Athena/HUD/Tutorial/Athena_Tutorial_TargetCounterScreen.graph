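# Tutorial HUD widget showing a row of five target counters, built on FortAthenaTutorialTargetCounter.
# UpdateTargetCounterVisuals plays the Eliminated animation on the counter indexed by NumberOfTargets, plays a sound, and runs End once the last index is reached.
# Construct plays Intro; ResetCounterVisuals sets every counter to Idle and replays it.
# Intro and End slide and fade the counters vertically, staggered from the center outward.
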WidgetBlueprint /Game/Athena/HUD/Tutorial/Athena_Tutorial_TargetCounterScreen
animations [
 End {
  binding_count=5
  length_seconds=0.35
  object=&self:End
  sequence {bindings [
    Tutorial_TargetCounter3 {
     id=#1
     possessed_class=Tutorial_TargetCounter_C
     tracks [
      {
       type=MovieScene2DTransformTrack
       sections [{
         type=MovieScene2DTransformSection
         end_frame=7
         start_frame=0
         channels [
          Translation.X {type=MovieSceneScriptingFloatChannel}
          Translation.Y {type=MovieSceneScriptingFloatChannel, keys [[0,0],[4.4443,-60]]}
          Angle {type=MovieSceneScriptingFloatChannel}
          Scale.X {type=MovieSceneScriptingFloatChannel}
          Scale.Y {type=MovieSceneScriptingFloatChannel}
          Shear.X {type=MovieSceneScriptingFloatChannel}
          Shear.Y {type=MovieSceneScriptingFloatChannel}
         ]
        }]
      }
      {
       type=MovieSceneFloatTrack
       display_name=Render Opacity
       sections [{
         type=MovieSceneFloatSection
         end_frame=7
         start_frame=0
         channels [None {type=MovieSceneScriptingFloatChannel, keys [[0,1],[0.8887,1],[4.4443,0]]}]
        }]
      }
     ]
    }
    Tutorial_TargetCounter2 {
     id=#2
     possessed_class=Tutorial_TargetCounter_C
     tracks [
      {
       type=MovieScene2DTransformTrack
       sections [{
         type=MovieScene2DTransformSection
         end_frame=6
         start_frame=0
         channels [
          Translation.X {type=MovieSceneScriptingFloatChannel, const [0,1]}
          Translation.Y {type=MovieSceneScriptingFloatChannel, keys [[0.8887,0],[5.333,-60]]}
          Angle {type=MovieSceneScriptingFloatChannel}
          Scale.X {type=MovieSceneScriptingFloatChannel}
          Scale.Y {type=MovieSceneScriptingFloatChannel}
          Shear.X {type=MovieSceneScriptingFloatChannel}
          Shear.Y {type=MovieSceneScriptingFloatChannel}
         ]
        }]
      }
      {
       type=MovieSceneFloatTrack
       display_name=Render Opacity
       sections [{
         type=MovieSceneFloatSection
         end_frame=7
         start_frame=0
         channels [None {type=MovieSceneScriptingFloatChannel, keys [[0,1],[1.7777,1],[5.3333,0]]}]
        }]
      }
     ]
    }
    Tutorial_TargetCounter4 {
     id=#3
     possessed_class=Tutorial_TargetCounter_C
     tracks [
      {
       type=MovieScene2DTransformTrack
       sections [{
         type=MovieScene2DTransformSection
         end_frame=7
         start_frame=0
         channels [
          Translation.X {type=MovieSceneScriptingFloatChannel}
          Translation.Y {type=MovieSceneScriptingFloatChannel, keys [[0.8887,0],[5.3333,-60]]}
          Angle {type=MovieSceneScriptingFloatChannel}
          Scale.X {type=MovieSceneScriptingFloatChannel}
          Scale.Y {type=MovieSceneScriptingFloatChannel}
          Shear.X {type=MovieSceneScriptingFloatChannel}
          Shear.Y {type=MovieSceneScriptingFloatChannel}
         ]
        }]
      }
      {
       type=MovieSceneFloatTrack
       display_name=Render Opacity
       sections [{
         type=MovieSceneFloatSection
         end_frame=7
         start_frame=0
         channels [None {type=MovieSceneScriptingFloatChannel, keys [[0,1],[1.7777,1],[5.3333,0]]}]
        }]
      }
     ]
    }
    Tutorial_TargetCounter {
     id=#4
     possessed_class=Tutorial_TargetCounter_C
     tracks [
      {
       type=MovieScene2DTransformTrack
       sections [{
         type=MovieScene2DTransformSection
         end_frame=7
         start_frame=1
         channels [
          Translation.X {type=MovieSceneScriptingFloatChannel}
          Translation.Y {type=MovieSceneScriptingFloatChannel, keys [[1.7777,0],[6.222,-60]]}
          Angle {type=MovieSceneScriptingFloatChannel}
          Scale.X {type=MovieSceneScriptingFloatChannel}
          Scale.Y {type=MovieSceneScriptingFloatChannel}
          Shear.X {type=MovieSceneScriptingFloatChannel}
          Shear.Y {type=MovieSceneScriptingFloatChannel}
         ]
        }]
      }
      {
       type=MovieSceneFloatTrack
       display_name=Render Opacity
       sections [{
         type=MovieSceneFloatSection
         end_frame=7
         start_frame=0
         channels [None {type=MovieSceneScriptingFloatChannel, keys [[0,1],[2.6667,1],[6.222,0]]}]
        }]
      }
     ]
    }
    Tutorial_TargetCounter5 {
     id=#5
     possessed_class=Tutorial_TargetCounter_C
     tracks [
      {
       type=MovieScene2DTransformTrack
       sections [{
         type=MovieScene2DTransformSection
         end_frame=7
         start_frame=1
         channels [
          Translation.X {type=MovieSceneScriptingFloatChannel}
          Translation.Y {type=MovieSceneScriptingFloatChannel, keys [[1.7777,0],[6.222,-60]]}
          Angle {type=MovieSceneScriptingFloatChannel}
          Scale.X {type=MovieSceneScriptingFloatChannel}
          Scale.Y {type=MovieSceneScriptingFloatChannel}
          Shear.X {type=MovieSceneScriptingFloatChannel}
          Shear.Y {type=MovieSceneScriptingFloatChannel}
         ]
        }]
      }
      {
       type=MovieSceneFloatTrack
       display_name=Render Opacity
       sections [{
         type=MovieSceneFloatSection
         end_frame=7
         start_frame=0
         channels [None {type=MovieSceneScriptingFloatChannel, keys [[0,1],[2.6667,1],[6.222,0]]}]
        }]
      }
     ]
    }
   ]}
 }
 Intro {
  binding_count=5
  length_seconds=0.35
  object=&self:Intro
  sequence {bindings [
    Tutorial_TargetCounter3 {
     id=#1
     possessed_class=Tutorial_TargetCounter_C
     tracks [
      {
       type=MovieScene2DTransformTrack
       sections [{
         type=MovieScene2DTransformSection
         end_frame=7
         start_frame=0
         channels [
          Translation.X {type=MovieSceneScriptingFloatChannel}
          Translation.Y {type=MovieSceneScriptingFloatChannel, keys [[0,-60],[4.4443,0]]}
          Angle {type=MovieSceneScriptingFloatChannel}
          Scale.X {type=MovieSceneScriptingFloatChannel}
          Scale.Y {type=MovieSceneScriptingFloatChannel}
          Shear.X {type=MovieSceneScriptingFloatChannel}
          Shear.Y {type=MovieSceneScriptingFloatChannel}
         ]
        }]
      }
      {
       type=MovieSceneFloatTrack
       display_name=Render Opacity
       sections [{
         type=MovieSceneFloatSection
         end_frame=7
         start_frame=0
         channels [None {type=MovieSceneScriptingFloatChannel, keys [[0,0],[0.8887,0],[4.4443,1]]}]
        }]
      }
     ]
    }
    Tutorial_TargetCounter2 {
     id=#2
     possessed_class=Tutorial_TargetCounter_C
     tracks [
      {
       type=MovieScene2DTransformTrack
       sections [{
         type=MovieScene2DTransformSection
         end_frame=6
         start_frame=0
         channels [
          Translation.X {type=MovieSceneScriptingFloatChannel}
          Translation.Y {type=MovieSceneScriptingFloatChannel, keys [[0.8887,-60],[5.333,0]]}
          Angle {type=MovieSceneScriptingFloatChannel}
          Scale.X {type=MovieSceneScriptingFloatChannel}
          Scale.Y {type=MovieSceneScriptingFloatChannel}
          Shear.X {type=MovieSceneScriptingFloatChannel}
          Shear.Y {type=MovieSceneScriptingFloatChannel}
         ]
        }]
      }
      {
       type=MovieSceneFloatTrack
       display_name=Render Opacity
       sections [{
         type=MovieSceneFloatSection
         end_frame=7
         start_frame=0
         channels [None {type=MovieSceneScriptingFloatChannel, keys [[0,0],[1.7777,0],[5.3333,1]]}]
        }]
      }
     ]
    }
    Tutorial_TargetCounter4 {
     id=#3
     possessed_class=Tutorial_TargetCounter_C
     tracks [
      {
       type=MovieScene2DTransformTrack
       sections [{
         type=MovieScene2DTransformSection
         end_frame=7
         start_frame=0
         channels [
          Translation.X {type=MovieSceneScriptingFloatChannel}
          Translation.Y {type=MovieSceneScriptingFloatChannel, keys [[0.8887,-60],[5.3333,0]]}
          Angle {type=MovieSceneScriptingFloatChannel}
          Scale.X {type=MovieSceneScriptingFloatChannel}
          Scale.Y {type=MovieSceneScriptingFloatChannel}
          Shear.X {type=MovieSceneScriptingFloatChannel}
          Shear.Y {type=MovieSceneScriptingFloatChannel}
         ]
        }]
      }
      {
       type=MovieSceneFloatTrack
       display_name=Render Opacity
       sections [{
         type=MovieSceneFloatSection
         end_frame=7
         start_frame=0
         channels [None {type=MovieSceneScriptingFloatChannel, keys [[0,0],[1.7777,0],[5.3333,1]]}]
        }]
      }
     ]
    }
    Tutorial_TargetCounter {
     id=#4
     possessed_class=Tutorial_TargetCounter_C
     tracks [
      {
       type=MovieScene2DTransformTrack
       sections [{
         type=MovieScene2DTransformSection
         end_frame=7
         start_frame=1
         channels [
          Translation.X {type=MovieSceneScriptingFloatChannel}
          Translation.Y {type=MovieSceneScriptingFloatChannel, keys [[1.7777,-60],[6.222,0]]}
          Angle {type=MovieSceneScriptingFloatChannel}
          Scale.X {type=MovieSceneScriptingFloatChannel}
          Scale.Y {type=MovieSceneScriptingFloatChannel}
          Shear.X {type=MovieSceneScriptingFloatChannel}
          Shear.Y {type=MovieSceneScriptingFloatChannel}
         ]
        }]
      }
      {
       type=MovieSceneFloatTrack
       display_name=Render Opacity
       sections [{
         type=MovieSceneFloatSection
         end_frame=7
         start_frame=0
         channels [None {type=MovieSceneScriptingFloatChannel, keys [[0,0],[2.6667,0],[6.222,1]]}]
        }]
      }
     ]
    }
    Tutorial_TargetCounter5 {
     id=#5
     possessed_class=Tutorial_TargetCounter_C
     tracks [
      {
       type=MovieScene2DTransformTrack
       sections [{
         type=MovieScene2DTransformSection
         end_frame=7
         start_frame=1
         channels [
          Translation.X {type=MovieSceneScriptingFloatChannel}
          Translation.Y {type=MovieSceneScriptingFloatChannel, keys [[1.7777,-60],[6.222,0]]}
          Angle {type=MovieSceneScriptingFloatChannel}
          Scale.X {type=MovieSceneScriptingFloatChannel}
          Scale.Y {type=MovieSceneScriptingFloatChannel}
          Shear.X {type=MovieSceneScriptingFloatChannel}
          Shear.Y {type=MovieSceneScriptingFloatChannel}
         ]
        }]
      }
      {
       type=MovieSceneFloatTrack
       display_name=Render Opacity
       sections [{
         type=MovieSceneFloatSection
         end_frame=7
         start_frame=0
         channels [None {type=MovieSceneScriptingFloatChannel, keys [[0,0],[2.6667,0],[6.222,1]]}]
        }]
      }
     ]
    }
   ]}
 }
]
graphs [EventGraph {nodes [
   Event Construct()
   > PlayAnimation(InAnimation=Get Intro())
   n1 = Event UpdateTargetCounterVisuals!() -> (OutputDelegate, then, NumberOfTargets:int=0) {is=title K2Node_Event}
   n2 = "Get (a copy)"(
    Array:Array of Tutorial Target Counter Object References=Make Array!(
     "[0]":Tutorial Target Counter Object Reference=Get Tutorial_TargetCounter() -> (Tutorial_TargetCounter:Tutorial Target Counter Object Reference)
     "[1]":Tutorial Target Counter Object Reference=Get Tutorial_TargetCounter2() -> (Tutorial_TargetCounter2:Tutorial Target Counter Object Reference)
     "[2]":Tutorial Target Counter Object Reference=Get Tutorial_TargetCounter3() -> (Tutorial_TargetCounter3:Tutorial Target Counter Object Reference)
     "[3]":Tutorial Target Counter Object Reference=Get Tutorial_TargetCounter4() -> (Tutorial_TargetCounter4:Tutorial Target Counter Object Reference)
     "[4]":Tutorial Target Counter Object Reference=Get Tutorial_TargetCounter5() -> (Tutorial_TargetCounter5:Tutorial Target Counter Object Reference)
    ) -> (Array:Array of Tutorial Target Counter Object References)
    Dimension 1=@n1.NumberOfTargets
   ) -> (Output:Tutorial Target Counter Object Reference)
   > PlayAnimation(self=@n2, InAnimation=Get Eliminated!(self:Tutorial Target Counter Object Reference=@n2) -> (Eliminated:Widget Animation Object Reference))
   > PlaySound2D(Sound=/Game/Sounds/Fort_UI_Sounds/Tutorial/UI_Tutorial_TargetElminated_Counter_Cue)
   > Branch(Condition="integer >= integer"!(A:int=@n1.NumberOfTargets, B:int=4) -> (ReturnValue:bool) {is=title K2Node_CallFunction})
   > PlayAnimation(InAnimation=Get End() -> (End:Widget Animation Object Reference))
   Event ResetCounterVisuals() {is=title K2Node_Event}
   > n3 = For Each Loop(Array:Array of Tutorial Target Counter Object References=Make Array(
     "[0]"=Get Tutorial_TargetCounter()
     "[1]"=Get Tutorial_TargetCounter2()
     "[2]"=Get Tutorial_TargetCounter3()
     "[3]"=Get Tutorial_TargetCounter4()
     "[4]"=Get Tutorial_TargetCounter5()
    )) -> (Array Element:Tutorial Target Counter Object Reference)
   > PlayAnimation(self=@n3, InAnimation=Get Idle!(self:Tutorial Target Counter Object Reference=@n3) -> (Idle:Widget Animation Object Reference))
   > PlayAnimation(InAnimation=Get Intro())
   Event PreConstruct()
   Event Tick()
  ]}]
variables []
widgets {
 info {
  inheritedWidgetCount=0
  parentClass=/Script/FortniteUI.FortAthenaTutorialTargetCounter
  rootWidgetClass=&HorizontalBox
  widgetCount=6
 }
 widgets [
  HorizontalBox_28 {widgetClassPath=&HorizontalBox, widget=HorizontalBox_28}
  Tutorial_TargetCounter {
   widgetClassPath=&/Tutorial_TargetCounter
   bIsVariable=true
   parent=HorizontalBox_28
   widget=Tutorial_TargetCounter
   slot {class=HorizontalBoxSlot}
  }
  Tutorial_TargetCounter2 {
   widgetClassPath=&Tutorial_TargetCounter
   bIsVariable=true
   parent=HorizontalBox_28
   widget=Tutorial_TargetCounter2
   slot {class=HorizontalBoxSlot, properties {padding=<ltrb 15 0 0 0>}}
  }
  Tutorial_TargetCounter3 {
   widgetClassPath=&Tutorial_TargetCounter
   bIsVariable=true
   parent=HorizontalBox_28
   widget=Tutorial_TargetCounter3
   slot {class=HorizontalBoxSlot, properties {padding=<ltrb 15 0 0 0>}}
  }
  Tutorial_TargetCounter4 {
   widgetClassPath=&Tutorial_TargetCounter
   bIsVariable=true
   parent=HorizontalBox_28
   widget=Tutorial_TargetCounter4
   slot {class=HorizontalBoxSlot, properties {padding=<ltrb 15 0 0 0>}}
  }
  Tutorial_TargetCounter5 {
   widgetClassPath=&Tutorial_TargetCounter
   bIsVariable=true
   parent=HorizontalBox_28
   widget=Tutorial_TargetCounter5
   slot {class=HorizontalBoxSlot, properties {padding=<ltrb 15 0 0 0>}}
  }
 ]
}
