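# "Tutorial Completed" banner widget, built from Victory Royale materials, text styles and widget names.
# The OnCompleted event plays the Victory_v2_b2 intro and a reward-screen sound.
# The intro reveals the material backplate, pops in shadowed rich text, and animates light shafts, flares and shine sweeps; some bindings name widgets absent from the tree.
# When it finishes, SpinningSpikeball loops indefinitely, rotating the spikeball flare and replaying backplate material keys.
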WidgetBlueprint /Game/Athena/HUD/Tutorial/Tutorial_Complete_V2
animations [
 SpinningSpikeball {
  binding_count=1
  display_label=""
  length_seconds=35.7931
  object=&self:SpinningSpikeball
  sequence {bindings [
    ImageSpinningSpikeball1 {
     id=#1
     possessed_class=Image
     tracks [{
       type=MovieScene2DTransformTrack
       sections [{
         type=MovieScene2DTransformSection
         end_frame=715
         start_frame=0
         channels [
          Translation.X {type=MovieSceneScriptingFloatChannel}
          Translation.Y {type=MovieSceneScriptingFloatChannel}
          Angle {type=MovieSceneScriptingFloatChannel, keys [[0,0],[715.8623,360]]}
          Scale.X {type=MovieSceneScriptingFloatChannel}
          Scale.Y {type=MovieSceneScriptingFloatChannel}
          Shear.X {type=MovieSceneScriptingFloatChannel}
          Shear.Y {type=MovieSceneScriptingFloatChannel}
         ]
        }]
      }]
    }
    ImageBackplate {
     id=#2
     possessed_class=Image
     tracks [{
       type=MovieSceneWidgetMaterialTrack
       display_name=Brush Material
       sections [{
         type=MovieSceneParameterSection
         end_frame=51
         start_frame=0
         channels [
          RevealDensity {type=MovieSceneScriptingFloatChannel, keys [[0.02,0.0632],[6.824,1.3382],[7.147,10]]}
          RevealRadius {type=MovieSceneScriptingFloatChannel, keys [[0.02,0],[6.824,1.3973],[7.147,3.2234]]}
          ShineOffset {type=MovieSceneScriptingFloatChannel, keys [[13.2353,0.6576],[28.4263,-1.0328]]}
          PatternRep {type=MovieSceneScriptingFloatChannel, const [0.9,1]}
         ]
        }]
      }]
    }
   ]}
 }
 Victory_v2_b2 {
  binding_count=28
  display_label=""
  length_seconds=3.3348
  object=&self:Victory_v2_b2
  sequence {
   bindings [
    Image_107 {
     id=#3
     possessed_class=Image
     tracks [
      {
       type=MovieSceneFloatTrack
       display_name=Render Opacity
       sections [{
         type=MovieSceneFloatSection
         end_frame=312
         start_frame=0
         channels [None {type=MovieSceneScriptingFloatChannel, keys [[86.312,0],[117.23,0.8097],[150.842,1],[227.17,0]]}]
        }]
      }
      {
       type=MovieScene2DTransformTrack
       sections [{
         type=MovieScene2DTransformSection
         end_frame=314
         start_frame=0
         channels [
          Translation.X {type=MovieSceneScriptingFloatChannel, keys [[85.75,-12.0191],[228.604,186.407]]}
          Translation.Y {type=MovieSceneScriptingFloatChannel, keys [[85.75,-181.193],[228.604,-205.39]]}
          Angle {type=MovieSceneScriptingFloatChannel, keys [[85.75,0],[228.604,20.494]]}
          Scale.X {type=MovieSceneScriptingFloatChannel}
          Scale.Y {type=MovieSceneScriptingFloatChannel}
          Shear.X {type=MovieSceneScriptingFloatChannel}
          Shear.Y {type=MovieSceneScriptingFloatChannel}
         ]
        }]
      }
     ]
    }
    ImageCross {
     id=#4
     possessed_class=Image
     tracks [{
       type=MovieSceneFloatTrack
       display_name=Render Opacity
       sections [{
         type=MovieSceneFloatSection
         end_frame=302
         start_frame=0
         channels [None {type=MovieSceneScriptingFloatChannel, keys [[90.854,0],[107.244,0.8097],[140.856,1],[217.184,0]]}]
        }]
      }]
    }
    Image_Lightshaft {
     id=#5
     possessed_class=Image
     tracks [
      {
       type=MovieSceneWidgetMaterialTrack
       display_name=Brush Material
       sections [{
         type=MovieSceneParameterSection
         end_frame=340
         start_frame=0
         channels [
          ShaftMax {type=MovieSceneScriptingFloatChannel, const [1.0018,2]}
          ShaftMin {type=MovieSceneScriptingFloatChannel, keys [[44,0.9713],[60,-2.7486],[88,0.9713]]}
         ]
        }]
      }
      {
       type=MovieSceneFloatTrack
       display_name=Render Opacity
       sections [{
         type=MovieSceneFloatSection
         end_frame=340
         start_frame=0
         channels [None {type=MovieSceneScriptingFloatChannel, keys [[44.002,0],[48,1],[88,0]]}]
        }]
      }
     ]
    }
    Image_text1flarebg {
     id=#6
     possessed_class=Image
     tracks [
      {
       type=MovieSceneFloatTrack
       display_name=Render Opacity
       sections [{
         type=MovieSceneFloatSection
         end_frame=257
         start_frame=0
         channels [None {type=MovieSceneScriptingFloatChannel, keys [[44.982,0],[60.314,0.8097],[93.926,1],[122.134,0]]}]
        }]
      }
      {
       type=MovieScene2DTransformTrack
       sections [{
         type=MovieScene2DTransformSection
         end_frame=232
         start_frame=0
         channels [
          Translation.X {type=MovieSceneScriptingFloatChannel}
          Translation.Y {type=MovieSceneScriptingFloatChannel}
          Angle {type=MovieSceneScriptingFloatChannel}
          Scale.X {type=MovieSceneScriptingFloatChannel, keys [[45.456,1],[62.084,1.2],[154.678,1]]}
          Scale.Y {type=MovieSceneScriptingFloatChannel, keys [[45.456,0.14],[62.084,1.2],[154.678,1]]}
          Shear.X {type=MovieSceneScriptingFloatChannel}
          Shear.Y {type=MovieSceneScriptingFloatChannel}
         ]
        }]
      }
     ]
    }
    CommonTextBlock_5 {
     id=#7
     display_name=RichTextBlock_Royale
     possessed_class=CommonTextBlock
     tracks [
      {
       type=MovieScene2DTransformTrack
       sections [{
         type=MovieScene2DTransformSection
         end_frame=223
         start_frame=0
         channels [
          Translation.X {type=MovieSceneScriptingFloatChannel}
          Translation.Y {type=MovieSceneScriptingFloatChannel}
          Angle {type=MovieSceneScriptingFloatChannel}
          Scale.X {type=MovieSceneScriptingFloatChannel, keys [[0.118,0],[18,0.001],[36.77,1.222],[43.538,1],[56.192,1]]}
          Scale.Y {type=MovieSceneScriptingFloatChannel, keys [[0.118,0],[18,0.001],[36.77,1.222],[43.538,1],[56.192,1]]}
          Shear.X {type=MovieSceneScriptingFloatChannel}
          Shear.Y {type=MovieSceneScriptingFloatChannel}
         ]
        }]
      }
      {
       type=MovieSceneWidgetMaterialTrack
       display_name=Font Material
       sections [{
         type=MovieSceneParameterSection
         end_frame=319
         start_frame=0
         channels [ShineOffset {type=MovieSceneScriptingFloatChannel, keys [[179.638,0.6],[261.638,-0.7262]]}]
        }]
      }
     ]
    }
    Image_text1flare1 {
     id=#8
     possessed_class=Image
     tracks [{
       type=MovieSceneFloatTrack
       display_name=Render Opacity
       sections [{
         type=MovieSceneFloatSection
         end_frame=257
         start_frame=0
         channels [None {type=MovieSceneScriptingFloatChannel, keys [[44.982,0],[60.314,0.8097],[93.926,1],[179.638,0]]}]
        }]
      }]
    }
    OverlayRoyaleBadge {
     id=#9
     possessed_class=Overlay
     tracks [{
       type=MovieSceneFloatTrack
       display_name=Render Opacity
       sections [{
         type=MovieSceneFloatSection
         end_frame=308
         start_frame=-61
         channels [None {type=MovieSceneScriptingFloatChannel, keys [[-60.176,0],[0.972,0],[25.97,1]]}]
        }]
      }]
    }
    Image_text2flarebg {
     id=#10
     possessed_class=Image
     tracks [
      {
       type=MovieSceneFloatTrack
       display_name=Render Opacity
       sections [{
         type=MovieSceneFloatSection
         end_frame=233
         start_frame=0
         channels [None {type=MovieSceneScriptingFloatChannel, keys [[45.456,0],[72.388,0.8097],[106,1],[155.746,0]]}]
        }]
      }
      {
       type=MovieScene2DTransformTrack
       sections [{
         type=MovieScene2DTransformSection
         end_frame=265
         start_frame=0
         channels [
          Translation.X {type=MovieSceneScriptingFloatChannel}
          Translation.Y {type=MovieSceneScriptingFloatChannel}
          Angle {type=MovieSceneScriptingFloatChannel}
          Scale.X {type=MovieSceneScriptingFloatChannel, keys [[45.93,1],[62.084,1.2],[188.29,1]]}
          Scale.Y {type=MovieSceneScriptingFloatChannel, keys [[45.93,0.37],[62.084,1.2],[188.29,1]]}
          Shear.X {type=MovieSceneScriptingFloatChannel}
          Shear.Y {type=MovieSceneScriptingFloatChannel}
         ]
        }]
      }
     ]
    }
    Imagetext2flare1 {
     id=#11
     possessed_class=Image
     tracks [{
       type=MovieSceneFloatTrack
       display_name=Render Opacity
       sections [{
         type=MovieSceneFloatSection
         end_frame=290
         start_frame=0
         channels [None {type=MovieSceneScriptingFloatChannel, keys [[45.456,0],[93.926,0.8097],[127.538,1],[213.25,0]]}]
        }]
      }]
    }
    ImageBackplate {
     id=#12
     possessed_class=Image
     tracks [{
       type=MovieSceneWidgetMaterialTrack
       display_name=Brush Material
       sections [{
         type=MovieSceneParameterSection
         end_frame=311
         start_frame=0
         channels [
          RevealDensity {type=MovieSceneScriptingFloatChannel, keys [[0.12,0.0632],[40.944,1.3382],[42.882,10]]}
          RevealRadius {type=MovieSceneScriptingFloatChannel, keys [[0.12,0],[40.944,1.3973],[42.882,3.2234]]}
          ShineOffset {type=MovieSceneScriptingFloatChannel, keys [[79.412,0.6576],[170.558,-1.0328]]}
          PatternRep {type=MovieSceneScriptingFloatChannel, const [0.9,1]}
         ]
        }]
      }]
    }
    ImageBackplateShadow {
     id=#13
     possessed_class=Image
     tracks [{
       type=MovieSceneWidgetMaterialTrack
       display_name=Brush Material
       sections [{
         type=MovieSceneParameterSection
         end_frame=311
         start_frame=0
         channels [
          RevealDensity {type=MovieSceneScriptingFloatChannel, keys [[0.12,0.0632],[40.944,1.3382],[42.882,10]]}
          RevealRadius {type=MovieSceneScriptingFloatChannel, keys [[0.12,0],[40.944,1.3973],[42.882,3.2234]]}
          ShineOffset {type=MovieSceneScriptingFloatChannel}
          PatternRep {type=MovieSceneScriptingFloatChannel, const [0.9,1]}
         ]
        }]
      }]
    }
    Imagespikeball2 {
     id=#14
     possessed_class=Image
     tracks [{
       type=MovieSceneFloatTrack
       display_name=Render Opacity
       sections [{
         type=MovieSceneFloatSection
         end_frame=330
         start_frame=0
         channels [None {type=MovieSceneScriptingFloatChannel, keys [[143.574,0],[159.964,0.8097],[193.576,1],[269.904,0]]}]
        }]
      }]
    }
    ImageSpinningSpikeball1 {
     id=#15
     possessed_class=Image
     tracks [{
       type=MovieSceneFloatTrack
       display_name=Render Opacity
       sections [{
         type=MovieSceneFloatSection
         end_frame=302
         start_frame=0
         channels [None {type=MovieSceneScriptingFloatChannel, keys [[177.452,0],[193.558,0.8097],[227.17,1]]}]
        }]
      }]
    }
    Image_108 {
     id=#16
     possessed_class=Image
     tracks [{
       type=MovieSceneFloatTrack
       display_name=Render Opacity
       sections [{
         type=MovieSceneFloatSection
         end_frame=319
         start_frame=0
         channels [None {type=MovieSceneScriptingFloatChannel, keys [[94.484,0],[110.874,0.8097],[144.486,1],[220.814,0]]}]
        }]
      }]
    }
    Image_text2flare2 {
     id=#17
     possessed_class=Image
     tracks [{
       type=MovieSceneFloatTrack
       display_name=Render Opacity
       sections [{
         type=MovieSceneFloatSection
         end_frame=290
         start_frame=0
         channels [None {type=MovieSceneScriptingFloatChannel, keys [[45.456,0],[93.926,0.8097],[127.538,1],[213.25,0]]}]
        }]
      }]
    }
    CommonTextRoyale {
     id=#18
     display_name=RichTextBlock_RoyaleShadow
     possessed_class=CommonTextBlock
     tracks [{
       type=MovieScene2DTransformTrack
       sections [{
         type=MovieScene2DTransformSection
         end_frame=299
         start_frame=0
         channels [
          Translation.X {type=MovieSceneScriptingFloatChannel}
          Translation.Y {type=MovieSceneScriptingFloatChannel}
          Angle {type=MovieSceneScriptingFloatChannel}
          Scale.X {type=MovieSceneScriptingFloatChannel, keys [[0.118,0],[18,0.001],[36.77,1.222],[43.538,1],[56.192,1]]}
          Scale.Y {type=MovieSceneScriptingFloatChannel, keys [[0.118,0],[18,0.001],[36.77,1.222],[43.538,1],[56.192,1]]}
          Shear.X {type=MovieSceneScriptingFloatChannel}
          Shear.Y {type=MovieSceneScriptingFloatChannel}
         ]
        }]
      }]
    }
    CommonTextBlock_4 {
     id=#19
     display_name=RichTextBlock_Victory
     possessed_class=CommonTextBlock
     tracks [
      {
       type=MovieScene2DTransformTrack
       sections [{
         type=MovieScene2DTransformSection
         end_frame=319
         start_frame=0
         channels [
          Translation.X {type=MovieSceneScriptingFloatChannel}
          Translation.Y {type=MovieSceneScriptingFloatChannel}
          Angle {type=MovieSceneScriptingFloatChannel}
          Scale.X {type=MovieSceneScriptingFloatChannel, keys [[0.118,0],[18,0.001],[36.77,1.222],[43.538,1],[56.192,1]]}
          Scale.Y {type=MovieSceneScriptingFloatChannel, keys [[0.118,0],[18,0.001],[36.77,1.222],[43.538,1],[56.192,1]]}
          Shear.X {type=MovieSceneScriptingFloatChannel}
          Shear.Y {type=MovieSceneScriptingFloatChannel}
         ]
        }]
      }
      {
       type=MovieSceneWidgetMaterialTrack
       display_name=Font Material
       sections [{
         type=MovieSceneParameterSection
         end_frame=319
         start_frame=36
         channels [ShineOffset {type=MovieSceneScriptingFloatChannel, keys [[165.05,0.59],[247.05,-0.7]]}]
        }]
      }
     ]
    }
    Image_text1flare2 {
     id=#20
     possessed_class=Image
     tracks [{
       type=MovieSceneFloatTrack
       display_name=Render Opacity
       sections [{
         type=MovieSceneFloatSection
         end_frame=257
         start_frame=0
         channels [None {type=MovieSceneScriptingFloatChannel, keys [[44.982,0],[60.314,0.8097],[93.926,1],[179.638,0]]}]
        }]
      }]
    }
    CommonTextBlock_6 {
     id=#21
     display_name=RichTextBlock_VictoryShadow
     possessed_class=CommonTextBlock
     tracks [{
       type=MovieScene2DTransformTrack
       sections [{
         type=MovieScene2DTransformSection
         end_frame=301
         start_frame=0
         channels [
          Translation.X {type=MovieSceneScriptingFloatChannel}
          Translation.Y {type=MovieSceneScriptingFloatChannel}
          Angle {type=MovieSceneScriptingFloatChannel}
          Scale.X {type=MovieSceneScriptingFloatChannel, keys [[0.118,0],[18,0.001],[36.77,1.222],[43.538,1],[56.192,1]]}
          Scale.Y {type=MovieSceneScriptingFloatChannel, keys [[0.118,0],[18,0.001],[36.77,1.222],[43.538,1],[56.192,1]]}
          Shear.X {type=MovieSceneScriptingFloatChannel}
          Shear.Y {type=MovieSceneScriptingFloatChannel}
         ]
        }]
      }]
    }
    Image1Cross {
     id=#22
     possessed_class=Image
     tracks [{
       type=MovieSceneFloatTrack
       display_name=Render Opacity
       sections [{
         type=MovieSceneFloatSection
         end_frame=335
         start_frame=0
         channels [None {type=MovieSceneScriptingFloatChannel, keys [[179.98,0],[207.196,1],[283.524,0]]}]
        }]
      }]
    }
   ]
   display_rate {numerator=120}
  }
 }
]
graphs [EventGraph {nodes [
   OnCompleted() {is=title K2Node_CustomEvent}
   > PlayAnimation(InAnimation=Get Victory_v2_b2() -> (Victory_v2_b2:Widget Animation Object Reference))
   > PlaySound2D(Sound=/Game/Sounds/Fort_UI_Sounds/Rewards/fort_rewardscreen_Continue)
   "Animation Finished (Victory_v2_b2)"!() -> (OutputDelegate:Delegate, then:Exec) {is=title K2Node_WidgetAnimationEvent}
   > PlayAnimation(InAnimation=Get SpinningSpikeball() -> (SpinningSpikeball:Widget Animation Object Reference), NumLoopsToPlay=0)
  ]}]
variables []
widgets {
 info {inheritedWidgetCount=0, parentClass=&UserWidget, rootWidgetClass=&Overlay, widgetCount=20}
 widgets [
  OverlayWinV2 {widgetClassPath=&Overlay, widget=OverlayWinV2}
  OverlayRoyaleBadge {
   widgetClassPath=&Overlay
   parent=OverlayWinV2
   widget=OverlayRoyaleBadge
   slot {class=OverlaySlot, properties {horizontalAlignment=HAlign_Fill, verticalAlignment=VAlign_Fill}}
  }
  Image_text1flarebg {
   widgetClassPath=&Image
   parent=OverlayRoyaleBadge
   widget=Image_text1flarebg
   properties {
    brush {
     bIsDynamicallyLoaded=false
     drawAs=Image
     imageType=NoImage
     mirroring=NoMirror
     resourceName=None
     resourceObject=/Game/Athena/HUD/Materials/Victory/M_UI_LightFlare
     tiling=NoTile
     imageSize=<1371.1171 946.7446>
     margin=<ltrb 0 0 0 0>
     outlineSettings {
      bUseBrushTransparency=false
      roundingType=HalfHeightRadius
      width=0
      color {colorUseRule=UseColor_Specified, specifiedColor=<rgba 0 0 0 0>}
      cornerRadii=<0 0 0 0>
     }
     tintColor {colorUseRule=UseColor_Specified, specifiedColor=<rgba 1 1 1 1>}
     uVRegion {bIsValid=false, max=<0 0>, min=<0 0>}
    }
    renderTransform {angle=0, scale=<1.192 1.192>, shear=<0 0>, translation=<37 -86>}
   }
   slot {class=OverlaySlot, properties {horizontalAlignment=HAlign_Center, verticalAlignment=VAlign_Center}}
  }
  OverlayBody {
   widgetClassPath=&Overlay
   parent=OverlayRoyaleBadge
   widget=OverlayBody
   properties {renderTransform {angle=0, scale=<0.9 0.9>, shear=<0 0>, translation=<0 0>}}
   slot {class=OverlaySlot, properties {horizontalAlignment=HAlign_Center, verticalAlignment=VAlign_Center}}
  }
  Image_Lightshaft {
   widgetClassPath=&Image
   parent=OverlayBody
   widget=Image_Lightshaft
   properties {
    renderOpacity=0
    brush {
     bIsDynamicallyLoaded=false
     drawAs=Image
     imageType=NoImage
     mirroring=NoMirror
     resourceName=None
     resourceObject=/Game/Athena/HUD/Materials/Victory/M_Victory_LightShafts_Inst
     tiling=NoTile
     imageSize=<1210.3491 446.3852>
     margin=<ltrb 0 0 0 0>
     outlineSettings {
      bUseBrushTransparency=false
      roundingType=HalfHeightRadius
      width=0
      color {colorUseRule=UseColor_Specified, specifiedColor=<rgba 0 0 0 0>}
      cornerRadii=<0 0 0 0>
     }
     tintColor {colorUseRule=UseColor_Specified, specifiedColor=<rgba 1 1 1 1>}
     uVRegion {bIsValid=false, max=<0 0>, min=<0 0>}
    }
    renderTransform {angle=0, scale=<1 1>, shear=<0 0>, translation=<0 -82>}
   }
   slot {class=OverlaySlot, properties {horizontalAlignment=HAlign_Center, verticalAlignment=VAlign_Center, padding=<ltrb -1000 -1000 -1000 -1000>}}
  }
  Overlay_Backplate {
   widgetClassPath=&Overlay
   parent=OverlayBody
   widget=Overlay_Backplate
   properties {renderTransform {angle=-4.11, scale=<0.92 1>, shear=<-36 0>, translation=<0 -90>}}
   slot {class=OverlaySlot, properties {horizontalAlignment=HAlign_Center}}
  }
  ImageBackplateShadow {
   widgetClassPath=&Image
   parent=Overlay_Backplate
   widget=ImageBackplateShadow
   properties {
    brush {
     bIsDynamicallyLoaded=false
     drawAs=Image
     imageType=NoImage
     mirroring=NoMirror
     resourceName=None
     resourceObject=/Game/Athena/HUD/Materials/Victory/M_VictoryBG2_Inst
     tiling=NoTile
     imageSize=<740 270>
     margin=<ltrb 0 0 0 0>
     outlineSettings {
      bUseBrushTransparency=false
      roundingType=HalfHeightRadius
      width=0
      color {colorUseRule=UseColor_Specified, specifiedColor=<rgba 0 0 0 0>}
      cornerRadii=<0 0 0 0>
     }
     tintColor {colorUseRule=UseColor_Specified, specifiedColor=<rgba 1 1 1 1>}
     uVRegion {bIsValid=false, max=<0 0>, min=<0 0>}
    }
    colorAndOpacity=<rgba 0 0.0103 0.1301 0.8>
    renderTransform {angle=0, scale=<1 1>, shear=<0 0>, translation=<12 10>}
   }
   slot {class=OverlaySlot}
  }
  ImageBackplate {
   widgetClassPath=&Image
   parent=Overlay_Backplate
   widget=ImageBackplate
   properties {brush {
     bIsDynamicallyLoaded=false
     drawAs=Image
     imageType=NoImage
     mirroring=NoMirror
     resourceName=None
     resourceObject=&M_VictoryBG2_Inst
     tiling=NoTile
     imageSize=<740 270>
     margin=<ltrb 0 0 0 0>
     outlineSettings {
      bUseBrushTransparency=false
      roundingType=HalfHeightRadius
      width=0
      color {colorUseRule=UseColor_Specified, specifiedColor=<rgba 0 0 0 0>}
      cornerRadii=<0 0 0 0>
     }
     tintColor {colorUseRule=UseColor_Specified, specifiedColor=<rgba 1 1 1 1>}
     uVRegion {bIsValid=false, max=<0 0>, min=<0 0>}
    }}
   slot {class=OverlaySlot}
  }
  SizeBox_1 {
   widgetClassPath=&SizeBox
   parent=OverlayBody
   widget=SizeBox_1
   properties {bOverride_WidthOverride=true, maxDesiredWidth=420, widthOverride=420}
   slot {class=OverlaySlot, properties {horizontalAlignment=HAlign_Center, verticalAlignment=VAlign_Center}}
  }
  ScaleBox_1 {widgetClassPath=&ScaleBox, parent=SizeBox_1, widget=ScaleBox_1, properties {stretchDirection=DownOnly}, slot {class=SizeBoxSlot}}
  OverlayVictoryText {
   widgetClassPath=&Overlay
   bIsVariable=true
   parent=ScaleBox_1
   widget=OverlayVictoryText
   properties {renderTransform {angle=-4.11, scale=<1.05 1>, shear=<0 0>, translation=<19 -99>}}
   slot {class=ScaleBoxSlot}
  }
  Border_1 {
   widgetClassPath=&Border
   parent=OverlayVictoryText
   widget=Border_1
   properties {
    horizontalAlignment=HAlign_Left
    background {
     bIsDynamicallyLoaded=false
     drawAs=NoDrawType
     imageType=NoImage
     mirroring=NoMirror
     resourceName=None
     resourceObject=null
     tiling=NoTile
     imageSize=<32 32>
     margin=<ltrb 0 0 0 0>
     outlineSettings {
      bUseBrushTransparency=false
      roundingType=HalfHeightRadius
      width=0
      color {colorUseRule=UseColor_Specified, specifiedColor=<rgba 0 0 0 0>}
      cornerRadii=<0 0 0 0>
     }
     tintColor {colorUseRule=UseColor_Specified, specifiedColor=<rgba 1 1 1 1>}
     uVRegion {bIsValid=false, max=<0 0>, min=<0 0>}
    }
    contentColorAndOpacity=<rgba 0.0307 0.0307 0.1812 0.9>
    padding=<ltrb 0 0 0 0>
   }
   slot {class=OverlaySlot}
  }
  VerticalBox_2 {
   widgetClassPath=&VerticalBox
   parent=Border_1
   widget=VerticalBox_2
   properties {renderTransform {angle=0, scale=<1 1>, shear=<-4 0>, translation=<0 0>}}
   slot {class=BorderSlot, properties {horizontalAlignment=HAlign_Left, padding=<ltrb 0 0 0 0>}}
  }
  RichTextBlock_VictoryShadow {
   widgetClassPath=/Script/UMG.RichTextBlock
   bIsVariable=true
   parent=VerticalBox_2
   widget=RichTextBlock_VictoryShadow
   properties {text="<cap>T</>UTORIAL", textStyleSet=/Game/Athena/HUD/EndGame/VictoryRichTextStyles}
   slot {class=VerticalBoxSlot, properties {horizontalAlignment=HAlign_Left}}
  }
  RichTextBlock_RoyaleShadow {
   widgetClassPath=&RichTextBlock
   bIsVariable=true
   parent=VerticalBox_2
   widget=RichTextBlock_RoyaleShadow
   properties {
    text="<cap>C</>OMPLETED"
    textStyleSet=&VictoryRichTextStyles
    decoratorClasses [/Script/UMG.RichTextBlockImageDecorator]
   }
   slot {class=VerticalBoxSlot, properties {padding=<ltrb 85 -65 0 0>}}
  }
  VerticalBox_1 {
   widgetClassPath=&VerticalBox
   parent=OverlayVictoryText
   widget=VerticalBox_1
   properties {renderTransform {angle=0, scale=<1 1>, shear=<-4 0>, translation=<-5 -8>}}
   slot {class=OverlaySlot}
  }
  RichTextBlock_Victory {
   widgetClassPath=&RichTextBlock
   bIsVariable=true
   parent=VerticalBox_1
   widget=RichTextBlock_Victory
   properties {text="<cap>T</>UTORIAL", textStyleSet=&VictoryRichTextStyles}
   slot {class=VerticalBoxSlot, properties {horizontalAlignment=HAlign_Left}}
  }
  RichTextBlock_Royale {
   widgetClassPath=&RichTextBlock
   bIsVariable=true
   parent=VerticalBox_1
   widget=RichTextBlock_Royale
   properties {
    text="<cap>C</>OMPLETED"
    textStyleSet=&VictoryRichTextStyles
    decoratorClasses [&RichTextBlockImageDecorator]
   }
   slot {class=VerticalBoxSlot, properties {padding=<ltrb 85 -65 0 0>}}
  }
  Image1Cross {
   widgetClassPath=&Image
   parent=OverlayRoyaleBadge
   widget=Image1Cross
   properties {
    brush {
     bIsDynamicallyLoaded=false
     drawAs=Image
     imageType=NoImage
     mirroring=NoMirror
     resourceName=None
     resourceObject=/Game/Athena/HUD/Materials/Victory/M_Cross_Flare
     tiling=NoTile
     imageSize=<128 128>
     margin=<ltrb 0 0 0 0>
     outlineSettings {
      bUseBrushTransparency=false
      roundingType=HalfHeightRadius
      width=0
      color {colorUseRule=UseColor_Specified, specifiedColor=<rgba 0 0 0 0>}
      cornerRadii=<0 0 0 0>
     }
     tintColor {colorUseRule=UseColor_Specified, specifiedColor=<rgba 1 1 1 1>}
     uVRegion {bIsValid=false, max=<0 0>, min=<0 0>}
    }
    renderTransform {angle=0, scale=<1 1>, shear=<0 0>, translation=<-185 -146>}
   }
   slot {class=OverlaySlot, properties {horizontalAlignment=HAlign_Center, verticalAlignment=VAlign_Center}}
  }
  ImageSpinningSpikeball1 {
   widgetClassPath=&Image
   parent=OverlayRoyaleBadge
   widget=ImageSpinningSpikeball1
   properties {
    brush {
     bIsDynamicallyLoaded=false
     drawAs=Image
     imageType=NoImage
     mirroring=NoMirror
     resourceName=None
     resourceObject=/Game/Athena/HUD/Materials/Victory/M_Flare_Spikeball
     tiling=NoTile
     imageSize=<256 256>
     margin=<ltrb 0 0 0 0>
     outlineSettings {
      bUseBrushTransparency=false
      roundingType=HalfHeightRadius
      width=0
      color {colorUseRule=UseColor_Specified, specifiedColor=<rgba 0 0 0 0>}
      cornerRadii=<0 0 0 0>
     }
     tintColor {colorUseRule=UseColor_Specified, specifiedColor=<rgba 1 1 1 1>}
     uVRegion {bIsValid=false, max=<0 0>, min=<0 0>}
    }
    renderTransform {angle=0, scale=<1 1>, shear=<0 0>, translation=<-185 -146>}
   }
   slot {class=OverlaySlot, properties {horizontalAlignment=HAlign_Center, verticalAlignment=VAlign_Center}}
  }
 ]
}
