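# Tutorial HUD widget that shows a pulsing gold radial glow filling its area.
# A stretched overlay holds one Image using a blurry thumbstick material instance tinted gold.
# On Construct it plays the one-second Idle animation endlessly, easing the overlay's render opacity from a quarter to half and back.
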WidgetBlueprint /Game/Athena/HUD/Tutorial/Tutorial_RadialGradient
animations [Idle {
  binding_count=1
  length_seconds=1
  object=&self:Idle
  sequence {
   bindings [Overlay_42 {
     id=#1
     possessed_class=Overlay
     tracks [{
       type=MovieSceneFloatTrack
       display_name=Render Opacity
       sections [{
         type=MovieSceneFloatSection
         end_frame=60
         start_frame=0
         channels [None {type=MovieSceneScriptingFloatChannel, keys [[0,0.25],[30,0.5],[60,0.25]]}]
        }]
      }]
    }]
   display_rate {numerator=60}
  }
 }]
graphs [EventGraph {nodes [
   Event Construct()
   > PlayAnimation(InAnimation=Get Idle() -> (Idle:Widget Animation Object Reference), NumLoopsToPlay=0)
   Event PreConstruct()
   Event Tick()
  ]}]
variables []
widgets {
 info {inheritedWidgetCount=0, parentClass=&UserWidget, rootWidgetClass=&CanvasPanel, widgetCount=3}
 widgets [
  CanvasPanel_27 {widgetClassPath=&CanvasPanel, widget=CanvasPanel_27}
  Overlay_42 {
   widgetClassPath=&Overlay
   parent=CanvasPanel_27
   widget=Overlay_42
   properties {renderOpacity=0.5}
   slot {
    class=CanvasPanelSlot
    properties {bAutoSize=true, layoutData {alignment=<0 0>, anchors {maximum=<1 1>, minimum=<0 0>}, offsets=<ltrb 0 0 0 0>}}
   }
  }
  IMG_RadialGradient {
   widgetClassPath=&Image
   bIsVariable=true
   parent=Overlay_42
   widget=IMG_RadialGradient
   properties {brush {
     bIsDynamicallyLoaded=false
     drawAs=Image
     imageType=NoImage
     mirroring=NoMirror
     resourceName=None
     resourceObject=/Game/Athena/HUD/Art/Materials/MI_HUD_Thumbstick_Blurry
     tiling=NoTile
     imageSize=<256 256>
     margin=<ltrb 0 0 0 0>
     outlineSettings {
      bUseBrushTransparency=false
      roundingType=HalfHeightRadius
      width=0
      color {colorUseRule=UseColor_Specified, specifiedColor=<rgba 0 0 0 0>}
      cornerRadii=<0 0 0 0>
     }
     tintColor {colorUseRule=UseColor_Specified, specifiedColor=<rgba 0.9047 0.6308 0.0865 1>}
     uVRegion {bIsValid=false, max=<0 0>, min=<0 0>}
    }}
   slot {class=OverlaySlot, properties {horizontalAlignment=HAlign_Fill, verticalAlignment=VAlign_Fill}}
  }
 ]
}
