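# UI sound cue for the tutorial's Next button.
# A single wave player plays the Tutorial_NextButton sound through a Modulator left at default settings, which gives slight random pitch and volume variation.
# The cue uses the Fort_UI sound class at reduced volume, and its concurrency override settings specify PreventNew as the resolution rule.
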SoundCue /Game/Sounds/Fort_UI_Sounds/Tutorial/UI_Tutorial_NextButton_Cue
root=Modulator_0
nodes [
 {id=Modulator_0, children [WavePlayer_1]}
 {id=WavePlayer_1, properties {soundWaveAssetPtr=&/Tutorial_NextButton}}
]
properties {
 maxDistance=2097152
 soundClassObject=&Fort_UI
 volumeMultiplier=0.4
 concurrencyOverrides {resolutionRule=PreventNew}
}
