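# Tutorial NEXT button widget that advances tutorials by broadcasting TutorialButtonPressed and playing a UI sound on click.
# An IconTextButton bound to the Tutorial_Next input action, with high priority and Game input override, sits over a gold outline image.
# PlayIntro starts the squash-in Intro and enables interaction; when Intro finishes, Idle_V2 loops a pulsing outline.
# Mouse down/up drive the Pressed animation, which has no bindings, and return Unhandled.
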WidgetBlueprint /Game/Athena/HUD/Tutorial/Tutorial_NextButton_Input
animations [
 Pressed {
  binding_count=0
  display_label=""
  length_seconds=0.1
  object=&self:Pressed
  sequence {bindings [], display_rate {numerator=60}}
 }
 Intro {
  binding_count=1
  length_seconds=0.25
  object=&self:Intro
  sequence {bindings [Overlay_0 {
     id=#1
     possessed_class=Overlay
     tracks [{
       type=MovieScene2DTransformTrack
       sections [{
         type=MovieScene2DTransformSection
         end_frame=5
         start_frame=0
         channels [
          Translation.X {type=MovieSceneScriptingFloatChannel, keys [[0,-80],[2,-10],[3,0],[5,0]]}
          Translation.Y {type=MovieSceneScriptingFloatChannel, const [0,2]}
          Angle {type=MovieSceneScriptingFloatChannel}
          Scale.X {type=MovieSceneScriptingFloatChannel, keys [[0,0.125],[1,0.3001],[2,1],[3,1],[4,0.9134],[5,1]]}
          Scale.Y {type=MovieSceneScriptingFloatChannel, keys [[0,1.75],[1,1],[2,0.6676],[3,1],[4,1.0888],[5,1]]}
          Shear.X {type=MovieSceneScriptingFloatChannel}
          Shear.Y {type=MovieSceneScriptingFloatChannel}
         ]
        }]
      }]
    }]}
 }
 Idle {
  binding_count=2
  length_seconds=1
  object=&self:Idle
  sequence {
   bindings [
    OverlaySlot_3 {
     id=#2
     parent_id=#3
     possessed_class=OverlaySlot
     tracks [{
       type=MovieSceneMarginTrack
       display_name=Padding
       sections [{
         type=MovieSceneMarginSection
         end_frame=60
         start_frame=0
         channels [
          Left {type=MovieSceneScriptingFloatChannel, keys [[0,0],[30,-220],[60,0]]}
          Top {type=MovieSceneScriptingFloatChannel, keys [[0,0],[30,-160],[60,0]]}
          Right {type=MovieSceneScriptingFloatChannel, keys [[0,0],[30,-220],[60,0]]}
          Bottom {type=MovieSceneScriptingFloatChannel, keys [[0,0],[30,-160],[60,0]]}
         ]
        }]
      }]
    }
    Image_56 {
     id=#3
     possessed_class=Image
     tracks [
      {
       type=MovieSceneFloatTrack
       display_name=Render Opacity
       sections [{type=MovieSceneFloatSection, end_frame=60, start_frame=0, channels [None {type=MovieSceneScriptingFloatChannel, keys [[0,1],[30,0],[60,1]]}]}]
      }
      {
       type=MovieScene2DTransformTrack
       sections [{
         type=MovieScene2DTransformSection
         end_frame=60
         start_frame=0
         channels [
          Translation.X {type=MovieSceneScriptingFloatChannel}
          Translation.Y {type=MovieSceneScriptingFloatChannel}
          Angle {type=MovieSceneScriptingFloatChannel}
          Scale.X {type=MovieSceneScriptingFloatChannel}
          Scale.Y {type=MovieSceneScriptingFloatChannel}
          Shear.X {type=MovieSceneScriptingFloatChannel}
          Shear.Y {type=MovieSceneScriptingFloatChannel}
         ]
        }]
      }
     ]
    }
   ]
   display_rate {numerator=60}
  }
 }
 Idle_V2 {
  binding_count=2
  length_seconds=1
  object=&self:Idle_V2
  sequence {
   bindings [
    OverlaySlot_3 {id=#2, parent_id=#3, possessed_class=OverlaySlot, tracks []}
    Image_56 {
     id=#3
     possessed_class=Image
     tracks [
      {
       type=MovieSceneFloatTrack
       display_name=Render Opacity
       sections [{type=MovieSceneFloatSection, end_frame=60, start_frame=0, channels [None {type=MovieSceneScriptingFloatChannel, keys [[0,1],[30,0],[60,1]]}]}]
      }
      {
       type=MovieScene2DTransformTrack
       sections [{
         type=MovieScene2DTransformSection
         end_frame=60
         start_frame=0
         channels [
          Translation.X {type=MovieSceneScriptingFloatChannel}
          Translation.Y {type=MovieSceneScriptingFloatChannel}
          Angle {type=MovieSceneScriptingFloatChannel}
          Scale.X {type=MovieSceneScriptingFloatChannel, keys [[0,0],[30,1.75],[60,0]]}
          Scale.Y {type=MovieSceneScriptingFloatChannel, keys [[0,0],[30,2],[60,0]]}
          Shear.X {type=MovieSceneScriptingFloatChannel}
          Shear.Y {type=MovieSceneScriptingFloatChannel}
         ]
        }]
      }
     ]
    }
   ]
   display_rate {numerator=60}
  }
 }
]
graphs [
 EventGraph {nodes [
   "On Button Clicked (Button_Tutorial)"!() -> (OutputDelegate, then, Button:Common Button Legacy Object Reference) {is=title K2Node_ComponentBoundEvent}
   > Call Tutorial Button Pressed!(execute:Exec, self:Tutorial Next Button Input Object Reference) -> (then:Exec) {is=title K2Node_CallDelegate}
   > PlaySound2D(Sound=/Game/Sounds/Fort_UI_Sounds/Tutorial/UI_Tutorial_NextButton_Cue)
   PlayIntro() {is=title K2Node_CustomEvent}
   > PlayAnimation(InAnimation=Get Intro())
   > SetIsInteractionEnabled!(
    execute
    self:Common Button Base Object Reference=Get Button_Tutorial() -> (Button_Tutorial:Icon Text Button Object Reference)
    bInIsInteractionEnabled:bool=true
   ) -> (then)
   "Animation Finished (Intro)"!() -> (OutputDelegate:Delegate, then:Exec) {is=title K2Node_WidgetAnimationEvent}
   > PlayAnimation(InAnimation=Get Idle_V2() -> (Idle_V2:Widget Animation Object Reference), NumLoopsToPlay=0)
  ]}
 OnMouseButtonDown {nodes [
   OnMouseButtonDown!() -> (then, MyGeometry:Geometry Structure, MouseEvent:"Pointer Event Structure (by ref)") {is=title K2Node_FunctionEntry}
   > Anim Transition!(
    Exec:Exec
    In:bool=true
    Animation:Widget Animation Object Reference=Get Pressed() -> (Pressed:Widget Animation Object Reference)
    PlaySpeed:double=2
    Widget:User Widget Object Reference=Self-Reference()
   ) -> (ExecOut:Exec) {is=title K2Node_MacroInstance}
   > Return Node!(execute, ReturnValue:Event Reply Structure=Unhandled())
  ]}
 OnMouseButtonUp {nodes [
   OnMouseButtonUp!() -> (then, MyGeometry:Geometry Structure, MouseEvent:"Pointer Event Structure (by ref)") {is=title K2Node_FunctionEntry}
   > Anim Transition(Animation=Get Pressed(), PlaySpeed=1, Widget=Self-Reference())
   > Return Node(ReturnValue=Unhandled())
  ]}
 TutorialButtonPressed {nodes [TutorialButtonPressed() {is=title K2Node_FunctionEntry}]}
]
variables [TutorialButtonPressed {}]
widgets {
 info {inheritedWidgetCount=0, parentClass=&UserWidget, rootWidgetClass=&Overlay, widgetCount=3}
 widgets [
  Overlay_0 {widgetClassPath=&Overlay, widget=Overlay_0}
  Image_56 {
   widgetClassPath=&Image
   bIsVariable=true
   parent=Overlay_0
   widget=Image_56
   properties {
    visibility=SelfHitTestInvisible
    brush {
     bIsDynamicallyLoaded=false
     drawAs=Box
     imageType=NoImage
     mirroring=NoMirror
     resourceName=None
     resourceObject=/Game/Athena/HUD/Art/Highlight/MI_FTUE_Outline_Square
     tiling=NoTile
     imageSize=<64 64>
     margin=<ltrb 0.14 0.14 0.14 0.14>
     outlineSettings {
      bUseBrushTransparency=false
      roundingType=HalfHeightRadius
      width=0
      color {colorUseRule=UseColor_Specified, specifiedColor=<rgba 0 0 0 0>}
      cornerRadii=<0 0 0 0>
     }
     tintColor {colorUseRule=UseColor_Specified, specifiedColor=<rgba 0.9131 0.6308 0.0865 1>}
     uVRegion {bIsValid=false, max=<0 0>, min=<0 0>}
    }
    renderTransform {angle=0, scale=<0.5 0.5>, shear=<-6 0>, translation=<0 0>}
   }
   slot {class=OverlaySlot, properties {horizontalAlignment=HAlign_Fill, verticalAlignment=VAlign_Fill}}
  }
  Button_Tutorial {
   widgetClassPath=&IconTextButton
   bIsVariable=true
   parent=Overlay_0
   widget=Button_Tutorial
   properties {
    bConsumePointerInput=true
    bSelectable=true
    bToggleable=true
    buttonText=NEXT
    controllerInputStyle=/FortUILibrary/Foundation/Buttons/ButtonStyles/ButtonStyle_Feature_L_Skew.ButtonStyle_Feature_L_Skew_C
    inputModeOverride=Game
    inputPriority=10000
    onButtonBaseClicked=(&self:WidgetTree.Button_Tutorial.HandleOnButtonClicked)
    onButtonBaseDoubleClicked=(&self:WidgetTree.Button_Tutorial.HandleOnButtonDoubleClicked)
    onButtonBaseHovered=(&self:WidgetTree.Button_Tutorial.HandleOnButtonHovered)
    onButtonBaseUnhovered=(&self:WidgetTree.Button_Tutorial.HandleOnButtonUnhovered)
    onSelectedChangedBase=(&self:WidgetTree.Button_Tutorial.HandleOnSelectedChanged)
    priority=10000
    style=&ButtonStyle_Feature_L_Skew_C
    renderTransform {angle=0, scale=<0.7 0.7>, shear=<0 0>, translation=<0 0>}
    triggeringInputAction {dataTable=&/Tutorial_InputActionDataTable, rowName=Tutorial_Next}
   }
   slot {class=OverlaySlot, properties {verticalAlignment=VAlign_Center}}
  }
 ]
}
