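# Athena tutorial HUD prompt that shows a "PRESS" callout with input glyphs or keybinds, built on FortAthenaTutorialTextPrompt.
# Activation plays Intro_V2 and a sound; delayed events reveal an animated target icon or the Next button, whose press calls NextButtonPressed.
# Functions set or collapse the action glyph and toggle the bottom-left HUD through a gameplay tag.
# Intro and Feedback_Positive_V2 animations also exist, but the event graph never plays them.
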WidgetBlueprint /Game/Athena/HUD/Tutorial/Tutorial_TextPrompt_Regular_Press
animations [
 Intro {
  binding_count=7
  display_label=""
  length_seconds=0.7
  object=&self:Intro
  sequence {
   bindings [
    Overlay_Container {
     id=#1
     possessed_class=Overlay
     tracks [{
       type=MovieSceneFloatTrack
       display_name=Render Opacity
       sections [{type=MovieSceneFloatSection, end_frame=13, start_frame=0, channels [None {type=MovieSceneScriptingFloatChannel, keys [[0,0],[3.564,1]]}]}]
      }]
    }
    Intro_Feedback {
     id=#2
     possessed_class=Image
     tracks [
      {
       type=MovieSceneFloatTrack
       display_name=Render Opacity
       sections [{type=MovieSceneFloatSection, end_frame=42, start_frame=0, channels [None {type=MovieSceneScriptingFloatChannel, keys [[9,0.8],[42,0]]}]}]
      }
      {
       type=MovieScene2DTransformTrack
       sections [{
         type=MovieScene2DTransformSection
         end_frame=42
         start_frame=0
         channels [
          Translation.X {type=MovieSceneScriptingFloatChannel}
          Translation.Y {type=MovieSceneScriptingFloatChannel}
          Angle {type=MovieSceneScriptingFloatChannel}
          Scale.X {type=MovieSceneScriptingFloatChannel, keys [[9,0],[42,4.5]]}
          Scale.Y {type=MovieSceneScriptingFloatChannel, keys [[9,0],[42,3.75]]}
          Shear.X {type=MovieSceneScriptingFloatChannel}
          Shear.Y {type=MovieSceneScriptingFloatChannel}
         ]
        }]
      }
     ]
    }
    SizeBox_Message {
     id=#3
     possessed_class=SizeBox
     tracks [{
       type=MovieScene2DTransformTrack
       sections [{
         type=MovieScene2DTransformSection
         end_frame=37
         start_frame=14
         channels [
          Translation.X {type=MovieSceneScriptingFloatChannel}
          Translation.Y {type=MovieSceneScriptingFloatChannel}
          Angle {type=MovieSceneScriptingFloatChannel}
          Scale.X {type=MovieSceneScriptingFloatChannel, keys [[14.13,1],[15.765,1.1],[19.1,1]]}
          Scale.Y {type=MovieSceneScriptingFloatChannel, keys [[14.13,1],[15.765,1.1],[19.1,1]]}
          Shear.X {type=MovieSceneScriptingFloatChannel}
          Shear.Y {type=MovieSceneScriptingFloatChannel}
         ]
        }]
      }]
    }
    Overlay_Dialog {
     id=#4
     possessed_class=Overlay
     tracks [
      {
       type=MovieScene2DTransformTrack
       sections [{
         type=MovieScene2DTransformSection
         end_frame=40
         start_frame=0
         channels [
          Translation.X {type=MovieSceneScriptingFloatChannel}
          Translation.Y {type=MovieSceneScriptingFloatChannel}
          Angle {type=MovieSceneScriptingFloatChannel, keys [[4.726,0],[12.738,0.1397],[16.439,0]]}
          Scale.X {type=MovieSceneScriptingFloatChannel, keys [[3.929,0],[13.725,1.1282],[16.365,1],[19.487,1]]}
          Scale.Y {type=MovieSceneScriptingFloatChannel, keys [[3.929,0],[13.725,1.1282],[16.365,1],[19.487,1]]}
          Shear.X {type=MovieSceneScriptingFloatChannel}
          Shear.Y {type=MovieSceneScriptingFloatChannel}
         ]
        }]
      }
      {
       type=MovieSceneFloatTrack
       display_name=Render Opacity
       sections [{type=MovieSceneFloatSection, end_frame=19, start_frame=0, channels [None {type=MovieSceneScriptingFloatChannel, const [1,1]}]}]
      }
     ]
    }
   ]
   display_rate {numerator=60}
  }
 }
 Feedback_Positive_V2 {
  binding_count=13
  length_seconds=2.0001
  object=&self:Feedback_Positive_V2
  sequence {bindings [
    Checkmark {
     id=#5
     possessed_class=Image
     tracks [
      {
       type=MovieScene2DTransformTrack
       sections [{
         type=MovieScene2DTransformSection
         channels [
          Translation.X {type=MovieSceneScriptingFloatChannel}
          Translation.Y {type=MovieSceneScriptingFloatChannel}
          Angle {type=MovieSceneScriptingFloatChannel}
          Scale.X {type=MovieSceneScriptingFloatChannel, keys [[0,10],[4,1],[15,1]]}
          Scale.Y {type=MovieSceneScriptingFloatChannel, keys [[0,10],[4,1],[15,1]]}
          Shear.X {type=MovieSceneScriptingFloatChannel}
          Shear.Y {type=MovieSceneScriptingFloatChannel}
         ]
        }]
      }
      {
       type=MovieSceneFloatTrack
       display_name=Render Opacity
       sections [{type=MovieSceneFloatSection, channels [None {type=MovieSceneScriptingFloatChannel, thinned_from=4, keys [[0,0],[4,1],[15,1]]}]}]
      }
     ]
    }
    Overlaytriangles {
     id=#6
     possessed_class=Overlay
     tracks [{
       type=MovieSceneFloatTrack
       display_name=Render Opacity
       sections [{type=MovieSceneFloatSection, channels [None {type=MovieSceneScriptingFloatChannel, keys [[9.5523,0],[11.9017,1],[40.0007,0]]}]}]
      }]
    }
    Overlay_PositiveFeedback_2 {
     id=#7
     possessed_class=Overlay
     tracks [
      {
       type=MovieScene2DTransformTrack
       sections [{
         type=MovieScene2DTransformSection
         channels [
          Translation.X {type=MovieSceneScriptingFloatChannel}
          Translation.Y {type=MovieSceneScriptingFloatChannel}
          Angle {type=MovieSceneScriptingFloatChannel}
          Scale.X {type=MovieSceneScriptingFloatChannel, keys [[5.9133,0],[16.947,4.5043]]}
          Scale.Y {type=MovieSceneScriptingFloatChannel, keys [[5.9133,0],[16.947,3.7578]]}
          Shear.X {type=MovieSceneScriptingFloatChannel}
          Shear.Y {type=MovieSceneScriptingFloatChannel}
         ]
        }]
      }
      {
       type=MovieSceneFloatTrack
       display_name=Render Opacity
       sections [{type=MovieSceneFloatSection, channels [None {type=MovieSceneScriptingFloatChannel, keys [[5.9133,0.8],[16.8953,0]]}]}]
      }
     ]
    }
    Overlay_Dialog {
     id=#8
     possessed_class=Overlay
     tracks [
      {
       type=MovieSceneFloatVectorTrack
       display_name=Pivot
       sections [{
         type=MovieSceneFloatVectorSection
         channels [
          Vector.X {type=MovieSceneScriptingFloatChannel, keys [[0,0],[1,0.5]]}
          Vector.Y {type=MovieSceneScriptingFloatChannel, keys [[0,0.6439],[1,0.5]]}
         ]
        }]
      }
      {
       type=MovieScene2DTransformTrack
       sections [{
         type=MovieScene2DTransformSection
         channels [
          Translation.X {type=MovieSceneScriptingFloatChannel}
          Translation.Y {type=MovieSceneScriptingFloatChannel}
          Angle {type=MovieSceneScriptingFloatChannel}
          Scale.X {
           type=MovieSceneScriptingFloatChannel
           keys [[0,1],[3.3423,1],[3.8023,0.8401],[5.3423,1.3471],[6.3423,0.6676],[7.3423,1],[8.2907,0.9351],[9.3423,1],[10.3423,1]]
          }
          Scale.Y {
           type=MovieSceneScriptingFloatChannel
           keys [[0,1],[3.3423,1],[3.8023,0.8401],[5.3113,1.3122],[6.332,0.7477],[7.425,1.0209],[8.3423,0.9134],[9.3423,1],[10.3423,1]]
          }
          Shear.X {type=MovieSceneScriptingFloatChannel}
          Shear.Y {type=MovieSceneScriptingFloatChannel}
         ]
        }]
      }
     ]
    }
    Checkmark_Shadow {
     id=#9
     possessed_class=Image
     tracks [
      {
       type=MovieSceneFloatTrack
       display_name=Render Opacity
       sections [{type=MovieSceneFloatSection, channels [None {type=MovieSceneScriptingFloatChannel, keys [[0,0],[4,0],[10.0003,1],[15,1]]}]}]
      }
      {
       type=MovieScene2DTransformTrack
       sections [{
         type=MovieScene2DTransformSection
         channels [
          Translation.X {type=MovieSceneScriptingFloatChannel, keys [[4,0],[10,2.1769],[15,2.1769]]}
          Translation.Y {type=MovieSceneScriptingFloatChannel, keys [[4,0],[10,9.1148],[15,9.1148]]}
          Angle {type=MovieSceneScriptingFloatChannel}
          Scale.X {type=MovieSceneScriptingFloatChannel, keys [[2,2],[4,1],[15,1]]}
          Scale.Y {type=MovieSceneScriptingFloatChannel, keys [[2,2],[4,1],[15,1]]}
          Shear.X {type=MovieSceneScriptingFloatChannel}
          Shear.Y {type=MovieSceneScriptingFloatChannel}
         ]
        }]
      }
     ]
    }
    CommonTextBlock_40 {
     id=#10
     possessed_class=CommonTextBlock
     tracks [{
       type=MovieSceneFloatTrack
       display_name=Render Opacity
       sections [{type=MovieSceneFloatSection, channels [None {type=MovieSceneScriptingFloatChannel, keys [[0,1],[3.0073,0]]}]}]
      }]
    }
    OverlayTextContent {
     id=#11
     possessed_class=Overlay
     tracks [{
       type=MovieScene2DTransformTrack
       sections [{
         type=MovieScene2DTransformSection
         end_frame=4
         start_frame=0
         channels [
          Translation.X {type=MovieSceneScriptingFloatChannel}
          Translation.Y {type=MovieSceneScriptingFloatChannel}
          Angle {type=MovieSceneScriptingFloatChannel}
          Scale.X {type=MovieSceneScriptingFloatChannel, const [1,3]}
          Scale.Y {type=MovieSceneScriptingFloatChannel, const [1,3]}
          Shear.X {type=MovieSceneScriptingFloatChannel}
          Shear.Y {type=MovieSceneScriptingFloatChannel}
         ]
        }]
      }]
    }
    Image_BGInset_Corner {
     id=#12
     possessed_class=Image
     tracks [{
       type=MovieSceneColorTrack
       display_name=Color and Opacity
       sections [{
         type=MovieSceneColorSection
         channels [
          Color.R {type=MovieSceneScriptingFloatChannel, keys [[3.7973,0],[7.2027,1]]}
          Color.G {type=MovieSceneScriptingFloatChannel, keys [[3.7973,0.117],[7.2027,1]]}
          Color.B {type=MovieSceneScriptingFloatChannel, const [1,2]}
          Color.A {type=MovieSceneScriptingFloatChannel, const [1,2]}
         ]
        }]
      }]
    }
    Image1Cross {
     id=#13
     possessed_class=Image
     tracks [
      {
       type=MovieSceneFloatTrack
       display_name=Render Opacity
       sections [{type=MovieSceneFloatSection, channels [None {type=MovieSceneScriptingFloatChannel, keys [[0,0],[3.7973,0],[8.7973,1],[20.1847,0]]}]}]
      }
      {
       type=MovieScene2DTransformTrack
       sections [{
         type=MovieScene2DTransformSection
         channels [
          Translation.X {type=MovieSceneScriptingFloatChannel, const [0,1]}
          Translation.Y {type=MovieSceneScriptingFloatChannel, const [0,1]}
          Scale.X {type=MovieSceneScriptingFloatChannel, keys [[3.7973,2],[13.982,2.4527],[26.809,2],[31.797,2]]}
          Scale.Y {type=MovieSceneScriptingFloatChannel, keys [[3.7973,2],[13.982,2.0272],[26.809,2],[31.797,2]]}
          Shear.X {type=MovieSceneScriptingFloatChannel, const [0,1]}
          Shear.Y {type=MovieSceneScriptingFloatChannel, const [0,1]}
         ]
        }]
      }
     ]
    }
    Overlay_PositiveFeedback {
     id=#14
     possessed_class=Overlay
     tracks [
      {
       type=MovieScene2DTransformTrack
       sections [{
         type=MovieScene2DTransformSection
         channels [
          Translation.X {type=MovieSceneScriptingFloatChannel}
          Translation.Y {type=MovieSceneScriptingFloatChannel}
          Angle {type=MovieSceneScriptingFloatChannel}
          Scale.X {type=MovieSceneScriptingFloatChannel, keys [[3,0],[14.0337,4.5043]]}
          Scale.Y {type=MovieSceneScriptingFloatChannel, keys [[3,0],[14.0337,3.7578]]}
          Shear.X {type=MovieSceneScriptingFloatChannel}
          Shear.Y {type=MovieSceneScriptingFloatChannel}
         ]
        }]
      }
      {
       type=MovieSceneFloatTrack
       display_name=Render Opacity
       sections [{type=MovieSceneFloatSection, channels [None {type=MovieSceneScriptingFloatChannel, keys [[3,0.8],[13.982,0]]}]}]
      }
     ]
    }
    Overlay_PositiveFeedback {
     id=#15
     possessed_class=Overlay
     tracks [
      {
       type=MovieScene2DTransformTrack
       sections [{
         type=MovieScene2DTransformSection
         channels [
          Translation.X {type=MovieSceneScriptingFloatChannel}
          Translation.Y {type=MovieSceneScriptingFloatChannel}
          Angle {type=MovieSceneScriptingFloatChannel}
          Scale.X {type=MovieSceneScriptingFloatChannel, keys [[3,0],[14.0337,4.5043]]}
          Scale.Y {type=MovieSceneScriptingFloatChannel, keys [[3,0],[14.0337,3.7578]]}
          Shear.X {type=MovieSceneScriptingFloatChannel}
          Shear.Y {type=MovieSceneScriptingFloatChannel}
         ]
        }]
      }
      {
       type=MovieSceneFloatTrack
       display_name=Render Opacity
       sections [{type=MovieSceneFloatSection, channels [None {type=MovieSceneScriptingFloatChannel, keys [[3,0.8],[13.982,0]]}]}]
      }
     ]
    }
    Image_BGOutline {
     id=#16
     possessed_class=Image
     tracks [{
       type=MovieSceneColorTrack
       display_name=Color and Opacity
       sections [{
         type=MovieSceneColorSection
         channels [
          Color.R {type=MovieSceneScriptingFloatChannel, keys [[3.7973,1],[7.2027,0]]}
          Color.G {type=MovieSceneScriptingFloatChannel, const [1,2]}
          Color.B {type=MovieSceneScriptingFloatChannel, keys [[3.7973,1],[7.2027,0.2576]]}
          Color.A {type=MovieSceneScriptingFloatChannel, const [1,1]}
         ]
        }]
      }]
    }
    ImageSpinningSpikeball1 {
     id=#17
     possessed_class=Image
     tracks [
      {
       type=MovieSceneFloatTrack
       display_name=Render Opacity
       sections [{
         type=MovieSceneFloatSection
         channels [None {type=MovieSceneScriptingFloatChannel, keys [[0,0],[3.7973,0],[8.7973,1],[31.797,1],[33.7973,0]]}]
        }]
      }
      {
       type=MovieScene2DTransformTrack
       sections [{
         type=MovieScene2DTransformSection
         channels [
          Translation.X {type=MovieSceneScriptingFloatChannel}
          Translation.Y {type=MovieSceneScriptingFloatChannel}
          Angle {type=MovieSceneScriptingFloatChannel, keys [[3.7973,0],[33.7973,15]]}
          Scale.X {type=MovieSceneScriptingFloatChannel, keys [[7.894,2],[34.7977,1]]}
          Scale.Y {type=MovieSceneScriptingFloatChannel, keys [[7.894,2],[34.7977,1]]}
          Shear.X {type=MovieSceneScriptingFloatChannel}
          Shear.Y {type=MovieSceneScriptingFloatChannel}
         ]
        }]
      }
     ]
    }
   ]}
 }
 Intro_V2 {
  binding_count=8
  length_seconds=0.7
  object=&self:Intro_V2
  sequence {
   bindings [
    Overlay_Container {
     id=#1
     possessed_class=Overlay
     tracks [{
       type=MovieSceneFloatTrack
       display_name=Render Opacity
       sections [{
         type=MovieSceneFloatSection
         end_frame=42
         start_frame=0
         channels [None {type=MovieSceneScriptingFloatChannel, keys [[0,0],[6.96,0],[10.524,1],[29.914,1]]}]
        }]
      }]
    }
    TargetIcon {
     id=#18
     possessed_class=Overlay
     tracks [{
       type=MovieSceneFloatTrack
       display_name=Render Opacity
       sections [{type=MovieSceneFloatSection, end_frame=39, start_frame=0, channels [None {type=MovieSceneScriptingFloatChannel, const [0,2]}]}]
      }]
    }
    Intro_Feedback {
     id=#2
     possessed_class=Image
     tracks [
      {
       type=MovieSceneFloatTrack
       display_name=Render Opacity
       sections [{
         type=MovieSceneFloatSection
         end_frame=42
         start_frame=0
         channels [None {type=MovieSceneScriptingFloatChannel, keys [[0,0],[6.96,0],[30.941,1]]}]
        }]
      }
      {
       type=MovieScene2DTransformTrack
       sections [{
         type=MovieScene2DTransformSection
         end_frame=42
         start_frame=0
         channels [
          Translation.X {type=MovieSceneScriptingFloatChannel}
          Translation.Y {type=MovieSceneScriptingFloatChannel}
          Angle {type=MovieSceneScriptingFloatChannel}
          Scale.X {type=MovieSceneScriptingFloatChannel, keys [[0,3],[6.96,3],[30.941,0.5]]}
          Scale.Y {type=MovieSceneScriptingFloatChannel, keys [[0,3.75],[6.96,3.75],[30.941,0.5]]}
          Shear.X {type=MovieSceneScriptingFloatChannel}
          Shear.Y {type=MovieSceneScriptingFloatChannel}
         ]
        }]
      }
     ]
    }
    SizeBox_Message {
     id=#3
     possessed_class=SizeBox
     tracks [{
       type=MovieScene2DTransformTrack
       sections [{
         type=MovieScene2DTransformSection
         end_frame=42
         start_frame=0
         channels [
          Translation.X {type=MovieSceneScriptingFloatChannel}
          Translation.Y {type=MovieSceneScriptingFloatChannel}
          Angle {type=MovieSceneScriptingFloatChannel}
          Scale.X {type=MovieSceneScriptingFloatChannel, keys [[0,1],[33.103,1],[34.738,1.1],[38.073,1]]}
          Scale.Y {type=MovieSceneScriptingFloatChannel, keys [[0,1],[33.103,1],[34.738,1.1],[38.073,1]]}
          Shear.X {type=MovieSceneScriptingFloatChannel}
          Shear.Y {type=MovieSceneScriptingFloatChannel}
         ]
        }]
      }]
    }
    Overlay_Dialog {
     id=#4
     possessed_class=Overlay
     tracks [
      {
       type=MovieScene2DTransformTrack
       sections [{
         type=MovieScene2DTransformSection
         end_frame=42
         start_frame=0
         channels [
          Translation.X {type=MovieSceneScriptingFloatChannel}
          Translation.Y {type=MovieSceneScriptingFloatChannel}
          Angle {type=MovieSceneScriptingFloatChannel, keys [[0,0],[23.699,0],[31.711,0.1397],[35.412,0]]}
          Scale.X {type=MovieSceneScriptingFloatChannel, keys [[0,0],[22.902,0],[32.698,1.1282],[35.338,1],[38.46,1]]}
          Scale.Y {type=MovieSceneScriptingFloatChannel, keys [[0,0],[22.902,0],[32.698,1.1282],[35.338,1],[38.46,1]]}
          Shear.X {type=MovieSceneScriptingFloatChannel}
          Shear.Y {type=MovieSceneScriptingFloatChannel}
         ]
        }]
      }
      {
       type=MovieSceneFloatTrack
       display_name=Render Opacity
       sections [{type=MovieSceneFloatSection, end_frame=22, start_frame=0, channels [None {type=MovieSceneScriptingFloatChannel, const [1,2]}]}]
      }
     ]
    }
   ]
   display_rate {numerator=60}
  }
 }
 TargetFeedback {
  binding_count=6
  length_seconds=1.7
  object=&self:TargetFeedback
  sequence {bindings [
    Ico_Target_Shadow {
     id=#19
     possessed_class=Image
     tracks [
      {
       type=MovieSceneFloatTrack
       display_name=Render Opacity
       sections [{
         type=MovieSceneFloatSection
         end_frame=34
         start_frame=0
         channels [None {type=MovieSceneScriptingFloatChannel, keys [[0,0],[4,0],[10,1],[15,1]]}]
        }]
      }
      {
       type=MovieScene2DTransformTrack
       sections [{
         type=MovieScene2DTransformSection
         end_frame=34
         start_frame=0
         channels [
          Translation.X {type=MovieSceneScriptingFloatChannel, keys [[4,0],[10,4],[15,4]]}
          Translation.Y {type=MovieSceneScriptingFloatChannel, keys [[4,0],[10,4],[15,4]]}
          Angle {type=MovieSceneScriptingFloatChannel}
          Scale.X {type=MovieSceneScriptingFloatChannel, keys [[2,2],[4,1],[15,1]]}
          Scale.Y {type=MovieSceneScriptingFloatChannel, keys [[2,2],[4,1],[15,1]]}
          Shear.X {type=MovieSceneScriptingFloatChannel}
          Shear.Y {type=MovieSceneScriptingFloatChannel}
         ]
        }]
      }
     ]
    }
    Ico_Target {
     id=#20
     possessed_class=Image
     tracks [
      {
       type=MovieSceneFloatTrack
       display_name=Render Opacity
       sections [{
         type=MovieSceneFloatSection
         end_frame=15
         start_frame=0
         channels [None {type=MovieSceneScriptingFloatChannel, thinned_from=4, keys [[0,0],[4,1],[15,1]]}]
        }]
      }
      {
       type=MovieScene2DTransformTrack
       sections [{
         type=MovieScene2DTransformSection
         end_frame=15
         start_frame=0
         channels [
          Translation.X {type=MovieSceneScriptingFloatChannel}
          Translation.Y {type=MovieSceneScriptingFloatChannel}
          Angle {type=MovieSceneScriptingFloatChannel}
          Scale.X {type=MovieSceneScriptingFloatChannel, keys [[0,10],[4,1],[15,1]]}
          Scale.Y {type=MovieSceneScriptingFloatChannel, keys [[0,10],[4,1],[15,1]]}
          Shear.X {type=MovieSceneScriptingFloatChannel}
          Shear.Y {type=MovieSceneScriptingFloatChannel}
         ]
        }]
      }
     ]
    }
    TargetIcon {
     id=#21
     possessed_class=Overlay
     tracks [{
       type=MovieSceneFloatTrack
       display_name=Render Opacity
       sections [{type=MovieSceneFloatSection, end_frame=15, start_frame=0, channels [None {type=MovieSceneScriptingFloatChannel, keys [[0,0],[1,1],[15,1]]}]}]
      }]
    }
    Overlay_Dialog {
     id=#22
     possessed_class=Overlay
     tracks [
      {
       type=MovieSceneFloatVectorTrack
       display_name=Pivot
       sections [{
         type=MovieSceneFloatVectorSection
         channels [
          Vector.X {type=MovieSceneScriptingFloatChannel, keys [[0,0],[1,0.5]]}
          Vector.Y {type=MovieSceneScriptingFloatChannel, keys [[0,0.6439],[1,0.5]]}
         ]
        }]
      }
      {
       type=MovieScene2DTransformTrack
       sections [{
         type=MovieScene2DTransformSection
         channels [
          Translation.X {type=MovieSceneScriptingFloatChannel}
          Translation.Y {type=MovieSceneScriptingFloatChannel}
          Angle {type=MovieSceneScriptingFloatChannel}
          Scale.X {
           type=MovieSceneScriptingFloatChannel
           keys [[0,1],[3.3423,1],[3.8023,0.8401],[5.3423,1.3471],[6.3423,0.6676],[7.3423,1],[8.2907,0.9351],[9.3423,1],[10.3423,1]]
          }
          Scale.Y {
           type=MovieSceneScriptingFloatChannel
           keys [[0,1],[3.3423,1],[3.8023,0.8401],[5.3113,1.3122],[6.332,0.7477],[7.425,1.0209],[8.3423,0.9134],[9.3423,1],[10.3423,1]]
          }
          Shear.X {type=MovieSceneScriptingFloatChannel}
          Shear.Y {type=MovieSceneScriptingFloatChannel}
         ]
        }]
      }
     ]
    }
    OverlayTextContent {
     id=#23
     possessed_class=Overlay
     tracks [{
       type=MovieScene2DTransformTrack
       sections [{
         type=MovieScene2DTransformSection
         end_frame=4
         start_frame=0
         channels [
          Translation.X {type=MovieSceneScriptingFloatChannel}
          Translation.Y {type=MovieSceneScriptingFloatChannel}
          Angle {type=MovieSceneScriptingFloatChannel}
          Scale.X {type=MovieSceneScriptingFloatChannel, const [1,3]}
          Scale.Y {type=MovieSceneScriptingFloatChannel, const [1,3]}
          Shear.X {type=MovieSceneScriptingFloatChannel}
          Shear.Y {type=MovieSceneScriptingFloatChannel}
         ]
        }]
      }]
    }
    RedDot {
     id=#24
     possessed_class=Image
     tracks [
      {
       type=MovieScene2DTransformTrack
       sections [{
         type=MovieScene2DTransformSection
         end_frame=15
         start_frame=0
         channels [
          Translation.X {type=MovieSceneScriptingFloatChannel}
          Translation.Y {type=MovieSceneScriptingFloatChannel}
          Angle {type=MovieSceneScriptingFloatChannel}
          Scale.X {type=MovieSceneScriptingFloatChannel, keys [[0,10],[4,1],[15,1]]}
          Scale.Y {type=MovieSceneScriptingFloatChannel, keys [[0,10],[4,1],[15,1]]}
          Shear.X {type=MovieSceneScriptingFloatChannel}
          Shear.Y {type=MovieSceneScriptingFloatChannel}
         ]
        }]
      }
      {
       type=MovieSceneFloatTrack
       display_name=Render Opacity
       sections [{
         type=MovieSceneFloatSection
         end_frame=15
         start_frame=0
         channels [None {type=MovieSceneScriptingFloatChannel, thinned_from=4, keys [[1,0],[4,1],[15,1]]}]
        }]
      }
     ]
    }
   ]}
 }
]
graphs [
 DeactivateInputAction {nodes [
   DeactivateInputAction() {is=title K2Node_FunctionEntry}
   > SetVisibility(self=Get ActionWidget_InputFeedback() -> (ActionWidget_InputFeedback:Common Action Widget Object Reference), InVisibility=Collapsed)
  ]}
 EventGraph {nodes [
   "Tutorial Button Pressed (Tutorial_NextButton_Input)"() {is=title K2Node_ComponentBoundEvent}
   > NextButtonPressed(self:Fort Athena Tutorial Text Prompt Object Reference)
   Event ActivateTextPrompt() {is=title K2Node_Event}
   > SetVisibility(self=Get TargetIcon() -> (TargetIcon:Overlay Object Reference), InVisibility=Collapsed)
   > PlayAnimation(InAnimation=Get Intro_V2() -> (Intro_V2:Widget Animation Object Reference))
   > PlaySound2D(Sound=/Game/Sounds/Fort_UI_Sounds/Tutorial/Tutorial_TextPrompt_Cue)
   Event DisplayNextButton() {is=title K2Node_Event}
   > Delay(Duration=1.5)
   n1 = Get Tutorial_NextButton_Input() -> (Tutorial_NextButton_Input:User Widget Object Reference)
   > SetVisibility(self=@n1, InVisibility=SelfHitTestInvisible)
   > n2 = Cast To Tutorial_NextButton_Input!(execute, Object=@n1) -> (then, CastFailed, AsTutorial Next Button Input:Tutorial Next Button Input Object Reference) {is=title K2Node_DynamicCast}
   > PlayIntro(self:Tutorial Next Button Input Object Reference=@n2)
   Event DisplayTarget() {is=title K2Node_Event}
   > Delay(Duration=0.5)
   > SetVisibility(self=Get TargetIcon(), InVisibility=SelfHitTestInvisible)
   > PlayAnimation(InAnimation=Get TargetFeedback() -> (TargetFeedback:Widget Animation Object Reference))
  ]}
 HUDToggle {nodes [
   n1 = HUDToggle!() -> (then, Hide:bool=~) {is=title K2Node_FunctionEntry}
   > n2 = Cast To FortPlayerControllerAthena(Object=GetOwningPlayer())
   > SetHUDElementVisibilityWithContext!(
    execute
    self:Fort Player Controller Object Reference=@n2
    HUDElementTags:"Gameplay Tag Container Structure (by ref)"=MakeGameplayTagContainerFromTag!(SingleTag:Gameplay Tag Structure=(TagName="HUD.Athena.Hide.Bottom.Left")) -> (ReturnValue:Gameplay Tag Container Structure)
    bHideElements:bool=@n1
    ReasonName:Name=None
    LogContext:String
   ) -> (then)
  ]}
 InitPosition {nodes [InitPosition() {is=title K2Node_FunctionEntry}]}
 SetActionInput {nodes [
   n1 = SetActionInput!() -> (then, DataTableInput:Data Table Row Handle Structure) {is=title K2Node_FunctionEntry}
   > SetInputAction!(execute, self:Common Action Widget Object Reference=Get ActionWidget_InputFeedback(), InputActionRow:Data Table Row Handle Structure=@n1) -> (then)
  ]}
]
variables []
widgets {
 info {inheritedWidgetCount=14, parentClass=/Script/FortniteUI.FortAthenaTutorialTextPrompt, rootWidgetClass=&ScaleBox, widgetCount=60}
 widgets [
  ScaleBox_1 {widgetClassPath=&ScaleBox, widget=ScaleBox_1, properties {stretch=None}}
  CanvasPanel_Container {
   widgetClassPath=&CanvasPanel
   parent=ScaleBox_1
   widget=CanvasPanel_Container
   slot {class=ScaleBoxSlot, properties {horizontalAlignment=HAlign_Fill, verticalAlignment=VAlign_Fill}}
  }
  Overlay_Container {
   widgetClassPath=&Overlay
   bIsVariable=true
   parent=CanvasPanel_Container
   widget=Overlay_Container
   slot {
    class=CanvasPanelSlot
    properties {bAutoSize=true, zOrder=1, layoutData {alignment=<0 0.5>, anchors {maximum=<0 0.5>, minimum=<0 0.5>}, offsets=<ltrb 0 0 100 30>}}
   }
  }
  Overlay_5 {
   widgetClassPath=&Overlay
   parent=Overlay_Container
   widget=Overlay_5
   properties {visibility=Collapsed}
   slot {class=OverlaySlot, properties {horizontalAlignment=HAlign_Fill, verticalAlignment=VAlign_Fill}}
  }
  Overlay_PositiveFeedback {
   widgetClassPath=&Overlay
   parent=Overlay_5
   widget=Overlay_PositiveFeedback
   properties {renderOpacity=0, visibility=Collapsed, renderTransform {angle=0, scale=<1 1>, shear=<-10 0>, translation=<0 0>}}
   slot {class=OverlaySlot, properties {horizontalAlignment=HAlign_Fill, verticalAlignment=VAlign_Fill}}
  }
  PositiveFeedback_BG {
   widgetClassPath=&Image
   parent=Overlay_PositiveFeedback
   widget=PositiveFeedback_BG
   properties {
    brush {
     bIsDynamicallyLoaded=false
     drawAs=Box
     imageType=NoImage
     mirroring=NoMirror
     resourceName=None
     resourceObject=/Game/Athena/HUD/LTM/Tutorial/Materials/M_UI_SkewBoxBorder_Inst
     tiling=NoTile
     imageSize=<64 64>
     margin=<ltrb 0.2 0.2 0.2 0.2>
     outlineSettings {
      bUseBrushTransparency=false
      roundingType=HalfHeightRadius
      width=0
      color {colorUseRule=UseColor_Specified, specifiedColor=<rgba 0 0 0 0>}
      cornerRadii=<0 0 0 0>
     }
     tintColor {colorUseRule=UseColor_Specified, specifiedColor=<rgba 0.0343 1 0.227 1>}
     uVRegion {bIsValid=false, max=<0 0>, min=<0 0>}
    }
    renderTransform {angle=0, scale=<1 1>, shear=<-5.0682 0>, translation=<0 0>}
   }
   slot {class=OverlaySlot, properties {horizontalAlignment=HAlign_Fill, verticalAlignment=VAlign_Fill}}
  }
  Overlay_PositiveFeedback_2 {
   widgetClassPath=&Overlay
   parent=Overlay_5
   widget=Overlay_PositiveFeedback_2
   properties {renderOpacity=0, visibility=Collapsed, renderTransform {angle=0, scale=<1 1>, shear=<-10 0>, translation=<0 0>}}
   slot {class=OverlaySlot, properties {horizontalAlignment=HAlign_Fill, verticalAlignment=VAlign_Fill}}
  }
  PositiveFeedback_BG_1 {
   widgetClassPath=&Image
   parent=Overlay_PositiveFeedback_2
   widget=PositiveFeedback_BG_1
   properties {
    brush {
     bIsDynamicallyLoaded=false
     drawAs=Box
     imageType=NoImage
     mirroring=NoMirror
     resourceName=None
     resourceObject=&M_UI_SkewBoxBorder_Inst
     tiling=NoTile
     imageSize=<64 64>
     margin=<ltrb 0.2 0.2 0.2 0.2>
     outlineSettings {
      bUseBrushTransparency=false
      roundingType=HalfHeightRadius
      width=0
      color {colorUseRule=UseColor_Specified, specifiedColor=<rgba 0 0 0 0>}
      cornerRadii=<0 0 0 0>
     }
     tintColor {colorUseRule=UseColor_Specified, specifiedColor=<rgba 0.0343 1 0.227 1>}
     uVRegion {bIsValid=false, max=<0 0>, min=<0 0>}
    }
    renderTransform {angle=0, scale=<1 1>, shear=<-5.0682 0>, translation=<0 0>}
   }
   slot {class=OverlaySlot, properties {horizontalAlignment=HAlign_Fill, verticalAlignment=VAlign_Fill}}
  }
  Intro_Feedback {
   widgetClassPath=&Image
   parent=Overlay_5
   widget=Intro_Feedback
   properties {
    renderOpacity=0
    visibility=Collapsed
    brush {
     bIsDynamicallyLoaded=false
     drawAs=Box
     imageType=NoImage
     mirroring=NoMirror
     resourceName=None
     resourceObject=/Game/Athena/HUD/Art/Highlight/MI_FTUE_Outline_Square
     tiling=NoTile
     imageSize=<64 64>
     margin=<ltrb 0.2 0.2 0.2 0.2>
     outlineSettings {
      bUseBrushTransparency=false
      roundingType=HalfHeightRadius
      width=0
      color {colorUseRule=UseColor_Specified, specifiedColor=<rgba 0 0 0 0>}
      cornerRadii=<0 0 0 0>
     }
     tintColor {colorUseRule=UseColor_Specified, specifiedColor=<rgba 0.9131 0.6308 0.0865 1>}
     uVRegion {bIsValid=false, max=<0 0>, min=<0 0>}
    }
    renderTransform {angle=0, scale=<1 1>, shear=<-17 0>, translation=<0 0>}
   }
   slot {class=OverlaySlot, properties {horizontalAlignment=HAlign_Fill, verticalAlignment=VAlign_Fill}}
  }
  Overlay_Dialog {
   widgetClassPath=&Overlay
   bIsVariable=true
   parent=Overlay_Container
   widget=Overlay_Dialog
   properties {renderTransformPivot=<0.5 0.6439>}
   slot {class=OverlaySlot, properties {verticalAlignment=VAlign_Center}}
  }
  OverlayTextContent {
   widgetClassPath=&Overlay
   parent=Overlay_Dialog
   widget=OverlayTextContent
   slot {class=OverlaySlot, properties {horizontalAlignment=HAlign_Fill, verticalAlignment=VAlign_Center}}
  }
  OverlayBG {
   widgetClassPath=&Overlay
   parent=OverlayTextContent
   widget=OverlayBG
   slot {class=OverlaySlot, properties {horizontalAlignment=HAlign_Fill, verticalAlignment=VAlign_Fill}}
  }
  Image_BGDropShadow {
   widgetClassPath=&Image
   parent=OverlayBG
   widget=Image_BGDropShadow
   properties {
    brush {
     bIsDynamicallyLoaded=false
     drawAs=Box
     imageType=NoImage
     mirroring=NoMirror
     resourceName=None
     resourceObject=/Game/Athena/HUD/LTM/Tutorial/Materials/M_UI_SkewBox_TutorialBG1
     tiling=NoTile
     imageSize=<256 128>
     margin=<ltrb 0.2 0 0.2 0>
     outlineSettings {
      bUseBrushTransparency=false
      roundingType=HalfHeightRadius
      width=0
      color {colorUseRule=UseColor_Specified, specifiedColor=<rgba 0 0 0 0>}
      cornerRadii=<0 0 0 0>
     }
     tintColor {colorUseRule=UseColor_Specified, specifiedColor=<rgba 0.0052 0.0052 0.0056 0.4>}
     uVRegion {bIsValid=false, max=<0 0>, min=<0 0>}
    }
    renderTransform {angle=1.5, scale=<1 1>, shear=<-2 0>, translation=<13 13>}
    renderTransformPivot=<0.2 0.5>
   }
   slot {class=OverlaySlot, properties {horizontalAlignment=HAlign_Fill, verticalAlignment=VAlign_Fill}}
  }
  Image_BGOutline {
   widgetClassPath=&Image
   parent=OverlayBG
   widget=Image_BGOutline
   properties {
    brush {
     bIsDynamicallyLoaded=false
     drawAs=Box
     imageType=NoImage
     mirroring=NoMirror
     resourceName=None
     resourceObject=&M_UI_SkewBox_TutorialBG1
     tiling=NoTile
     imageSize=<256 128>
     margin=<ltrb 0.2 0 0.2 0>
     outlineSettings {
      bUseBrushTransparency=false
      roundingType=HalfHeightRadius
      width=0
      color {colorUseRule=UseColor_Specified, specifiedColor=<rgba 0 0 0 0>}
      cornerRadii=<0 0 0 0>
     }
     tintColor {colorUseRule=UseColor_Specified, specifiedColor=<rgba 1 1 1 1>}
     uVRegion {bIsValid=false, max=<0 0>, min=<0 0>}
    }
    renderTransform {angle=0, scale=<1 1>, shear=<-1 0>, translation=<0 0>}
    renderTransformPivot=<0 0.5>
   }
   slot {class=OverlaySlot, properties {horizontalAlignment=HAlign_Fill, verticalAlignment=VAlign_Fill}}
  }
  Image_BGInset {
   widgetClassPath=&Image
   parent=OverlayBG
   widget=Image_BGInset
   properties {
    brush {
     bIsDynamicallyLoaded=false
     drawAs=Box
     imageType=NoImage
     mirroring=NoMirror
     resourceName=None
     resourceObject=&M_UI_SkewBox_TutorialBG1
     tiling=NoTile
     imageSize=<256 128>
     margin=<ltrb 0.2 0 0.3 0>
     outlineSettings {
      bUseBrushTransparency=false
      roundingType=HalfHeightRadius
      width=0
      color {colorUseRule=UseColor_Specified, specifiedColor=<rgba 0 0 0 0>}
      cornerRadii=<0 0 0 0>
     }
     tintColor {colorUseRule=UseColor_Specified, specifiedColor=<rgba 1 1 1 1>}
     uVRegion {bIsValid=false, max=<0 0>, min=<0 0>}
    }
    colorAndOpacity=<rgba 0.0273 0.0273 0.0273 1>
   }
   slot {class=OverlaySlot, properties {horizontalAlignment=HAlign_Fill, verticalAlignment=VAlign_Fill, padding=<ltrb 4 4 7 8>}}
  }
  Image_BGInset_Corner {
   widgetClassPath=&Image
   parent=OverlayBG
   widget=Image_BGInset_Corner
   properties {
    brush {
     bIsDynamicallyLoaded=false
     drawAs=Box
     imageType=NoImage
     mirroring=NoMirror
     resourceName=None
     resourceObject=/Game/Athena/HUD/LTM/Tutorial/Materials/M_UI_SkewBox_TutorialCornerBG
     tiling=NoTile
     imageSize=<256 128>
     margin=<ltrb 0 0 1 0>
     outlineSettings {
      bUseBrushTransparency=false
      roundingType=HalfHeightRadius
      width=0
      color {colorUseRule=UseColor_Specified, specifiedColor=<rgba 0 0 0 0>}
      cornerRadii=<0 0 0 0>
     }
     tintColor {colorUseRule=UseColor_Specified, specifiedColor=<rgba 1 1 1 1>}
     uVRegion {bIsValid=false, max=<0 0>, min=<0 0>}
    }
    colorAndOpacity=<rgba 0 0.117 1 1>
   }
   slot {class=OverlaySlot, properties {horizontalAlignment=HAlign_Fill, verticalAlignment=VAlign_Fill, padding=<ltrb 4 4 7 8>}}
  }
  BorderText {
   widgetClassPath=&Border
   parent=OverlayTextContent
   widget=BorderText
   properties {
    background {
     bIsDynamicallyLoaded=false
     drawAs=NoDrawType
     imageType=NoImage
     mirroring=NoMirror
     resourceName=None
     resourceObject=null
     tiling=NoTile
     imageSize=<32 32>
     margin=<ltrb 0 0 0 0>
     outlineSettings {
      bUseBrushTransparency=false
      roundingType=HalfHeightRadius
      width=0
      color {colorUseRule=UseColor_Specified, specifiedColor=<rgba 0 0 0 0>}
      cornerRadii=<0 0 0 0>
     }
     tintColor {colorUseRule=UseColor_Specified, specifiedColor=<rgba 1 1 1 1>}
     uVRegion {bIsValid=false, max=<0 0>, min=<0 0>}
    }
    padding=<ltrb 0 0 0 0>
   }
   slot {class=OverlaySlot, properties {verticalAlignment=VAlign_Fill, padding=<ltrb 40 25 80 30>}}
  }
  SizeBox_Message {
   widgetClassPath=&SizeBox
   parent=BorderText
   widget=SizeBox_Message
   properties {maxDesiredWidth=360, renderTransformPivot=<0.5 1>}
   slot {class=BorderSlot, properties {padding=<ltrb 0 0 0 0>}}
  }
  VerticalBox_0 {
   widgetClassPath=&VerticalBox
   parent=SizeBox_Message
   widget=VerticalBox_0
   slot {class=SizeBoxSlot, properties {verticalAlignment=VAlign_Center}}
  }
  HorizontalBox_1 {
   widgetClassPath=&HorizontalBox
   parent=VerticalBox_0
   widget=HorizontalBox_1
   slot {class=VerticalBoxSlot, properties {padding=<ltrb 0 0 0 4>}}
  }
  Spacer_68 {
   widgetClassPath=&Spacer
   parent=HorizontalBox_1
   widget=Spacer_68
   properties {size=<0 1>}
   slot {class=HorizontalBoxSlot, properties {size {sizeRule=Fill, value=1}}}
  }
  RichText_TopPrompt {
   widgetClassPath=/Script/UMG.RichTextBlock
   bInherited=true
   bIsVariable=true
   parent=HorizontalBox_1
   widget=RichText_TopPrompt
   properties {
    justification=Center
    lineHeightPercentage=0.77
    text=PRESS
    textStyleSet=/Game/Athena/HUD/LTM/Tutorial/Materials/TutorialCalloutStyles
    wrapTextAt=360
    renderTransform {angle=-0.35, scale=<1 1>, shear=<-12 0>, translation=<0 0>}
   }
   slot {class=HorizontalBoxSlot, properties {verticalAlignment=VAlign_Center, padding=<ltrb 0 0 5 0>}}
  }
  Overlay_AllWidgets {
   widgetClassPath=&Overlay
   bInherited=true
   parent=HorizontalBox_1
   widget=Overlay_AllWidgets
   slot {class=HorizontalBoxSlot, properties {verticalAlignment=VAlign_Center, padding=<ltrb 5 0 0 0>}}
  }
  HorizontalBox_2 {widgetClassPath=&HorizontalBox, parent=Overlay_AllWidgets, widget=HorizontalBox_2, slot {class=OverlaySlot}}
  Overlay_ActionWidgets {widgetClassPath=&Overlay, bInherited=true, parent=HorizontalBox_2, widget=Overlay_ActionWidgets, slot {class=HorizontalBoxSlot}}
  ActionWidget_InputFeedback {
   widgetClassPath=&CommonActionWidget
   bInherited=true
   bIsVariable=true
   parent=Overlay_ActionWidgets
   widget=ActionWidget_InputFeedback
   properties {
    visibility=Collapsed
    progressMaterialBrush {
     bIsDynamicallyLoaded=false
     drawAs=Image
     imageType=NoImage
     mirroring=NoMirror
     resourceName=None
     resourceObject=null
     tiling=NoTile
     imageSize=<46 46>
     margin=<ltrb 0 0 0 0>
     outlineSettings {
      bUseBrushTransparency=false
      roundingType=HalfHeightRadius
      width=0
      color {colorUseRule=UseColor_Specified, specifiedColor=<rgba 0 0 0 0>}
      cornerRadii=<0 0 0 0>
     }
     tintColor {colorUseRule=UseColor_Specified, specifiedColor=<rgba 1 1 1 1>}
     uVRegion {bIsValid=false, max=<0 0>, min=<0 0>}
    }
   }
   slot {class=OverlaySlot}
  }
  Overlay_KeybindWidgets {
   widgetClassPath=&Overlay
   bInherited=true
   parent=HorizontalBox_2
   widget=Overlay_KeybindWidgets
   slot {class=HorizontalBoxSlot, properties {horizontalAlignment=HAlign_Left, verticalAlignment=VAlign_Center}}
  }
  HorizontalBox_3 {
   widgetClassPath=&HorizontalBox
   parent=Overlay_KeybindWidgets
   widget=HorizontalBox_3
   slot {class=OverlaySlot, properties {verticalAlignment=VAlign_Center}}
  }
  Action_KeybindWidget {
   widgetClassPath=/FortUILibrary/Input/KeybindWidget
   bInherited=true
   parent=HorizontalBox_3
   widget=Action_KeybindWidget
   properties {axisScale=1}
   slot {class=HorizontalBoxSlot}
  }
  Action_KeybindWidget_1 {
   widgetClassPath=&KeybindWidget
   bInherited=true
   parent=HorizontalBox_3
   widget=Action_KeybindWidget_1
   properties {axisScale=-1}
   slot {class=HorizontalBoxSlot, properties {horizontalAlignment=HAlign_Left, verticalAlignment=VAlign_Center}}
  }
  Action_KeybindWidget_2 {
   widgetClassPath=&KeybindWidget
   bInherited=true
   parent=HorizontalBox_3
   widget=Action_KeybindWidget_2
   properties {axisScale=-1}
   slot {class=HorizontalBoxSlot, properties {horizontalAlignment=HAlign_Left, verticalAlignment=VAlign_Center}}
  }
  Action_KeybindWidget_3 {
   widgetClassPath=&KeybindWidget
   bInherited=true
   parent=HorizontalBox_3
   widget=Action_KeybindWidget_3
   properties {axisScale=1}
   slot {class=HorizontalBoxSlot, properties {horizontalAlignment=HAlign_Left, verticalAlignment=VAlign_Center}}
  }
  Spacer_1 {
   widgetClassPath=&Spacer
   parent=HorizontalBox_1
   widget=Spacer_1
   properties {size=<0 1>}
   slot {class=HorizontalBoxSlot, properties {size {sizeRule=Fill, value=1}}}
  }
  RichText_BottomPrompt {
   widgetClassPath=&RichTextBlock
   bInherited=true
   bIsVariable=true
   parent=VerticalBox_0
   widget=RichText_BottomPrompt
   properties {
    justification=Center
    lineHeightPercentage=0.77
    text="LOREM DOLOR AMIT <B>SEQUITOR</> <keyword>KEYWORD</>!"
    textStyleSet=&TutorialCalloutStyles
    wrapTextAt=360
    renderTransform {angle=-0.35, scale=<1 1>, shear=<-12 0>, translation=<0 0>}
   }
   slot {class=VerticalBoxSlot}
  }
  Overlay_ExtraText {widgetClassPath=&Overlay, bInherited=true, parent=VerticalBox_0, widget=Overlay_ExtraText, slot {class=VerticalBoxSlot}}
  HorizontalBox_363 {widgetClassPath=&HorizontalBox, parent=Overlay_ExtraText, widget=HorizontalBox_363, slot {class=OverlaySlot}}
  Action_KeybindWidget_4 {
   widgetClassPath=&KeybindWidget
   bInherited=true
   parent=HorizontalBox_363
   widget=Action_KeybindWidget_4
   properties {axisScale=1}
   slot {class=HorizontalBoxSlot, properties {horizontalAlignment=HAlign_Center, verticalAlignment=VAlign_Top}}
  }
  RichText_ExtraInputPrompt {
   widgetClassPath=&RichTextBlock
   bInherited=true
   bIsVariable=true
   parent=HorizontalBox_363
   widget=RichText_ExtraInputPrompt
   properties {
    justification=Center
    lineHeightPercentage=0.77
    text="LOREM DOLOR AMIT <B>SEQUITOR</> <keyword>KEYWORD</>!"
    textStyleSet=&TutorialCalloutStyles
    wrapTextAt=360
    renderTransform {angle=-0.35, scale=<1 1>, shear=<-12 0>, translation=<0 0>}
   }
   slot {class=HorizontalBoxSlot}
  }
  Overlay_2 {
   widgetClassPath=&Overlay
   parent=OverlayTextContent
   widget=Overlay_2
   properties {visibility=HitTestInvisible}
   slot {class=OverlaySlot, properties {horizontalAlignment=HAlign_Right}}
  }
  CommonTextBlock_40 {
   widgetClassPath=&CommonTextBlock
   parent=Overlay_2
   widget=CommonTextBlock_40
   properties {
    text="!!!!!!"
    visibility=Collapsed
    font {
     bForceMonospaced=false
     bMaterialIsStencil=false
     fontMaterial=null
     fontObject=&Burbank
     letterSpacing=0
     monospacedWidth=1
     size=50
     skewAmount=0
     typefaceFontName=Burbank-Black
     outlineSettings {
      bApplyOutlineToDropShadows=false
      bMiteredCorners=false
      bSeparateFillAlpha=false
      outlineMaterial=null
      outlineSize=0
      outlineColor=<rgba 0 0 0 1>
     }
    }
    renderTransform {angle=0, scale=<1 1>, shear=<-12 0>, translation=<-32 3>}
    shadowColorAndOpacity=<rgba 0.0052 0.0052 0.0056 0.8>
    shadowOffset=<8 8>
   }
   slot {class=OverlaySlot, properties {horizontalAlignment=HAlign_Center, verticalAlignment=VAlign_Center}}
  }
  OverlayCheck {
   widgetClassPath=&Overlay
   parent=Overlay_2
   widget=OverlayCheck
   properties {visibility=Collapsed, renderTransform {angle=0, scale=<1 1>, shear=<0 0>, translation=<0 -16>}}
   slot {class=OverlaySlot, properties {horizontalAlignment=HAlign_Right}}
  }
  Overlaytriangles {
   widgetClassPath=&Overlay
   parent=OverlayCheck
   widget=Overlaytriangles
   properties {renderOpacity=0}
   slot {class=OverlaySlot, properties {horizontalAlignment=HAlign_Center, verticalAlignment=VAlign_Fill, padding=<ltrb -50 0 -50 0>}}
  }
  HorizontalBox {
   widgetClassPath=&HorizontalBox
   parent=Overlaytriangles
   widget=HorizontalBox
   properties {renderTransform {angle=0, scale=<1 1>, shear=<-12 0>, translation=<0 0>}}
   slot {class=OverlaySlot, properties {horizontalAlignment=HAlign_Fill, verticalAlignment=VAlign_Center}}
  }
  ImageTriangleL_1 {
   widgetClassPath=&Image
   parent=HorizontalBox
   widget=ImageTriangleL_1
   properties {
    brush {
     bIsDynamicallyLoaded=false
     drawAs=Image
     imageType=NoImage
     mirroring=NoMirror
     resourceName=None
     resourceObject=/Game/UI/HUD/Materials/M_FX_TriangleScroll-Burst
     tiling=NoTile
     imageSize=<200 40>
     margin=<ltrb 0 0 0 0>
     outlineSettings {
      bUseBrushTransparency=false
      roundingType=HalfHeightRadius
      width=0
      color {colorUseRule=UseColor_Specified, specifiedColor=<rgba 0 0 0 0>}
      cornerRadii=<0 0 0 0>
     }
     tintColor {colorUseRule=UseColor_Specified, specifiedColor=<rgba 0 1 0.1907 1>}
     uVRegion {bIsValid=false, max=<0 0>, min=<0 0>}
    }
    renderTransform {angle=0, scale=<-1 1>, shear=<0 0>, translation=<0 0>}
    renderTransformPivot=<0.5 1>
   }
   slot {class=HorizontalBoxSlot, properties {horizontalAlignment=HAlign_Right, verticalAlignment=VAlign_Center, padding=<ltrb -200 0 0 0>}}
  }
  Spacer {widgetClassPath=&Spacer, parent=HorizontalBox, widget=Spacer, slot {class=HorizontalBoxSlot}}
  ImageTriangleR_1 {
   widgetClassPath=&Image
   parent=HorizontalBox
   widget=ImageTriangleR_1
   properties {
    brush {
     bIsDynamicallyLoaded=false
     drawAs=Image
     imageType=NoImage
     mirroring=NoMirror
     resourceName=None
     resourceObject=&M_FX_TriangleScroll-Burst
     tiling=NoTile
     imageSize=<200 40>
     margin=<ltrb 0 0 0 0>
     outlineSettings {
      bUseBrushTransparency=false
      roundingType=HalfHeightRadius
      width=0
      color {colorUseRule=UseColor_Specified, specifiedColor=<rgba 0 0 0 0>}
      cornerRadii=<0 0 0 0>
     }
     tintColor {colorUseRule=UseColor_Specified, specifiedColor=<rgba 0 1 0.1907 1>}
     uVRegion {bIsValid=false, max=<0 0>, min=<0 0>}
    }
    renderTransformPivot=<0.5 1>
   }
   slot {class=HorizontalBoxSlot, properties {verticalAlignment=VAlign_Center, padding=<ltrb 0 0 -200 0>}}
  }
  HorizontalBox_0 {
   widgetClassPath=&HorizontalBox
   parent=Overlaytriangles
   widget=HorizontalBox_0
   properties {renderTransform {angle=0, scale=<1 1>, shear=<-12 0>, translation=<0 0>}}
   slot {class=OverlaySlot, properties {verticalAlignment=VAlign_Center}}
  }
  ImageTriangleL {
   widgetClassPath=&Image
   parent=HorizontalBox_0
   widget=ImageTriangleL
   properties {
    brush {
     bIsDynamicallyLoaded=false
     drawAs=Image
     imageType=NoImage
     mirroring=NoMirror
     resourceName=None
     resourceObject=/Game/UI/HUD/Materials/M_FX_TriangleSpiral
     tiling=NoTile
     imageSize=<200 64>
     margin=<ltrb 0 0 0 0>
     outlineSettings {
      bUseBrushTransparency=false
      roundingType=HalfHeightRadius
      width=0
      color {colorUseRule=UseColor_Specified, specifiedColor=<rgba 0 0 0 0>}
      cornerRadii=<0 0 0 0>
     }
     tintColor {colorUseRule=UseColor_Specified, specifiedColor=<rgba 0 1 0.1907 1>}
     uVRegion {bIsValid=false, max=<0 0>, min=<0 0>}
    }
    renderTransform {angle=0, scale=<-1 1>, shear=<0 0>, translation=<0 0>}
    renderTransformPivot=<0.5 1>
   }
   slot {class=HorizontalBoxSlot, properties {horizontalAlignment=HAlign_Right, verticalAlignment=VAlign_Center, padding=<ltrb -200 0 0 0>}}
  }
  Spacer_70 {widgetClassPath=&Spacer, parent=HorizontalBox_0, widget=Spacer_70, slot {class=HorizontalBoxSlot}}
  ImageTriangleR {
   widgetClassPath=&Image
   parent=HorizontalBox_0
   widget=ImageTriangleR
   properties {
    brush {
     bIsDynamicallyLoaded=false
     drawAs=Image
     imageType=NoImage
     mirroring=NoMirror
     resourceName=None
     resourceObject=&M_FX_TriangleSpiral
     tiling=NoTile
     imageSize=<200 64>
     margin=<ltrb 0 0 0 0>
     outlineSettings {
      bUseBrushTransparency=false
      roundingType=HalfHeightRadius
      width=0
      color {colorUseRule=UseColor_Specified, specifiedColor=<rgba 0 0 0 0>}
      cornerRadii=<0 0 0 0>
     }
     tintColor {colorUseRule=UseColor_Specified, specifiedColor=<rgba 0 1 0.1907 1>}
     uVRegion {bIsValid=false, max=<0 0>, min=<0 0>}
    }
    renderTransformPivot=<0.5 1>
   }
   slot {class=HorizontalBoxSlot, properties {verticalAlignment=VAlign_Center, padding=<ltrb 0 0 -200 0>}}
  }
  Checkmark_Shadow {
   widgetClassPath=&Image
   bIsVariable=true
   parent=OverlayCheck
   widget=Checkmark_Shadow
   properties {
    renderOpacity=0
    brush {
     bIsDynamicallyLoaded=false
     drawAs=Image
     imageType=NoImage
     mirroring=NoMirror
     resourceName=None
     resourceObject=/Game/Athena/HUD/Atlas/Bacchus/Confirm_Normal_L_Sprite
     tiling=NoTile
     imageSize=<72 72>
     margin=<ltrb 0 0 0 0>
     outlineSettings {
      bUseBrushTransparency=false
      roundingType=HalfHeightRadius
      width=0
      color {colorUseRule=UseColor_Specified, specifiedColor=<rgba 0 0 0 0>}
      cornerRadii=<0 0 0 0>
     }
     tintColor {colorUseRule=UseColor_Specified, specifiedColor=<rgba 0 0 0 0.8>}
     uVRegion {bIsValid=false, max=<0 0>, min=<0 0>}
    }
    renderTransform {angle=0, scale=<1 1>, shear=<0 0>, translation=<4 8>}
   }
   slot {class=OverlaySlot, properties {horizontalAlignment=HAlign_Center, verticalAlignment=VAlign_Bottom}}
  }
  Checkmark {
   widgetClassPath=&Image
   bIsVariable=true
   parent=OverlayCheck
   widget=Checkmark
   properties {
    renderOpacity=0
    brush {
     bIsDynamicallyLoaded=false
     drawAs=Image
     imageType=NoImage
     mirroring=NoMirror
     resourceName=None
     resourceObject=&Confirm_Normal_L_Sprite
     tiling=NoTile
     imageSize=<72 72>
     margin=<ltrb 0 0 0 0>
     outlineSettings {
      bUseBrushTransparency=false
      roundingType=HalfHeightRadius
      width=0
      color {colorUseRule=UseColor_Specified, specifiedColor=<rgba 0 0 0 0>}
      cornerRadii=<0 0 0 0>
     }
     tintColor {colorUseRule=UseColor_Specified, specifiedColor=<rgba 0.0343 1 0.227 1>}
     uVRegion {bIsValid=false, max=<0 0>, min=<0 0>}
    }
   }
   slot {class=OverlaySlot, properties {horizontalAlignment=HAlign_Center}}
  }
  Overlay_Flare {
   widgetClassPath=&Overlay
   parent=OverlayCheck
   widget=Overlay_Flare
   properties {renderTransform {angle=0, scale=<0.8 0.8>, shear=<0 0>, translation=<0 0>}}
   slot {class=OverlaySlot, properties {horizontalAlignment=HAlign_Center, verticalAlignment=VAlign_Center}}
  }
  ImageSpinningSpikeball1 {
   widgetClassPath=&Image
   parent=Overlay_Flare
   widget=ImageSpinningSpikeball1
   properties {
    renderOpacity=0
    brush {
     bIsDynamicallyLoaded=false
     drawAs=Image
     imageType=NoImage
     mirroring=NoMirror
     resourceName=None
     resourceObject=/Game/Athena/HUD/Materials/Victory/M_Flare_Spikeball
     tiling=NoTile
     imageSize=<256 256>
     margin=<ltrb 0 0 0 0>
     outlineSettings {
      bUseBrushTransparency=false
      roundingType=HalfHeightRadius
      width=0
      color {colorUseRule=UseColor_Specified, specifiedColor=<rgba 0 0 0 0>}
      cornerRadii=<0 0 0 0>
     }
     tintColor {colorUseRule=UseColor_Specified, specifiedColor=<rgba 1 1 1 1>}
     uVRegion {bIsValid=false, max=<0 0>, min=<0 0>}
    }
    renderTransform {angle=13.5163, scale=<2 2>, shear=<0 0>, translation=<0 0>}
   }
   slot {class=OverlaySlot, properties {horizontalAlignment=HAlign_Center, verticalAlignment=VAlign_Center, padding=<ltrb -128 -128 -128 -128>}}
  }
  Image1Cross {
   widgetClassPath=&Image
   parent=Overlay_Flare
   widget=Image1Cross
   properties {
    renderOpacity=0
    brush {
     bIsDynamicallyLoaded=false
     drawAs=Image
     imageType=NoImage
     mirroring=NoMirror
     resourceName=None
     resourceObject=/Game/Athena/HUD/Materials/Victory/M_Cross_Flare
     tiling=NoTile
     imageSize=<128 128>
     margin=<ltrb 0 0 0 0>
     outlineSettings {
      bUseBrushTransparency=false
      roundingType=HalfHeightRadius
      width=0
      color {colorUseRule=UseColor_Specified, specifiedColor=<rgba 0 0 0 0>}
      cornerRadii=<0 0 0 0>
     }
     tintColor {colorUseRule=UseColor_Specified, specifiedColor=<rgba 1 1 1 1>}
     uVRegion {bIsValid=false, max=<0 0>, min=<0 0>}
    }
    renderTransform {angle=0, scale=<2 2>, shear=<0 0>, translation=<0 0>}
   }
   slot {class=OverlaySlot, properties {horizontalAlignment=HAlign_Center, verticalAlignment=VAlign_Center, padding=<ltrb -64 -64 -64 -64>}}
  }
  TargetIcon {
   widgetClassPath=&Overlay
   bIsVariable=true
   parent=OverlayTextContent
   widget=TargetIcon
   properties {visibility=Collapsed, renderTransform {angle=6, scale=<1 1>, shear=<0 0>, translation=<4 -12>}}
   slot {class=OverlaySlot, properties {horizontalAlignment=HAlign_Right}}
  }
  Ico_Target_Shadow {
   widgetClassPath=&Image
   bIsVariable=true
   parent=TargetIcon
   widget=Ico_Target_Shadow
   properties {
    renderOpacity=0.8
    brush {
     bIsDynamicallyLoaded=false
     drawAs=Image
     imageType=NoImage
     mirroring=NoMirror
     resourceName=None
     resourceObject=/Game/Athena/HUD/Art/Tutorial/FTUE_TargetCounter_Sprite
     tiling=NoTile
     imageSize=<86 86>
     margin=<ltrb 0 0 0 0>
     outlineSettings {
      bUseBrushTransparency=false
      roundingType=HalfHeightRadius
      width=0
      color {colorUseRule=UseColor_Specified, specifiedColor=<rgba 0 0 0 0>}
      cornerRadii=<0 0 0 0>
     }
     tintColor {colorUseRule=UseColor_Specified, specifiedColor=<rgba 0 0 0 1>}
     uVRegion {bIsValid=false, max=<0 0>, min=<0 0>}
    }
    renderTransform {angle=0, scale=<1 1>, shear=<0 0>, translation=<4 4>}
   }
   slot {class=OverlaySlot, properties {horizontalAlignment=HAlign_Fill, verticalAlignment=VAlign_Fill}}
  }
  RedDot {
   widgetClassPath=&Image
   bIsVariable=true
   parent=TargetIcon
   widget=RedDot
   properties {brush {
     bIsDynamicallyLoaded=false
     drawAs=Image
     imageType=NoImage
     mirroring=NoMirror
     resourceName=None
     resourceObject=/Game/Athena/HUD/Art/Materials/MI_HUD_Dot
     tiling=NoTile
     imageSize=<56 56>
     margin=<ltrb 0 0 0 0>
     outlineSettings {
      bUseBrushTransparency=false
      roundingType=HalfHeightRadius
      width=0
      color {colorUseRule=UseColor_Specified, specifiedColor=<rgba 0 0 0 0>}
      cornerRadii=<0 0 0 0>
     }
     tintColor {colorUseRule=UseColor_Specified, specifiedColor=<rgba 0.7529 0.2705 0.0369 1>}
     uVRegion {bIsValid=false, max=<0 0>, min=<0 0>}
    }}
   slot {class=OverlaySlot, properties {horizontalAlignment=HAlign_Center, verticalAlignment=VAlign_Center}}
  }
  Ico_Target {
   widgetClassPath=&Image
   bIsVariable=true
   parent=TargetIcon
   widget=Ico_Target
   properties {brush {
     bIsDynamicallyLoaded=false
     drawAs=Image
     imageType=NoImage
     mirroring=NoMirror
     resourceName=None
     resourceObject=&FTUE_TargetCounter_Sprite
     tiling=NoTile
     imageSize=<86 86>
     margin=<ltrb 0 0 0 0>
     outlineSettings {
      bUseBrushTransparency=false
      roundingType=HalfHeightRadius
      width=0
      color {colorUseRule=UseColor_Specified, specifiedColor=<rgba 0 0 0 0>}
      cornerRadii=<0 0 0 0>
     }
     tintColor {colorUseRule=UseColor_Specified, specifiedColor=<rgba 1 1 1 1>}
     uVRegion {bIsValid=false, max=<0 0>, min=<0 0>}
    }}
   slot {class=OverlaySlot, properties {horizontalAlignment=HAlign_Fill, verticalAlignment=VAlign_Fill}}
  }
  Tutorial_NextButton_Input {
   widgetClassPath=&/Tutorial_NextButton_Input
   bInherited=true
   bIsVariable=true
   parent=Overlay_Container
   widget=Tutorial_NextButton_Input
   properties {visibility=Visible, renderTransform {angle=0, scale=<1 1>, shear=<0 0>, translation=<64 64>}}
   slot {class=OverlaySlot, properties {horizontalAlignment=HAlign_Right, verticalAlignment=VAlign_Bottom}}
  }
 ]
}
