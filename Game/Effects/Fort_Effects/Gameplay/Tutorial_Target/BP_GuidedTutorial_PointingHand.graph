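# UMG widget that shows an orange-tinted pointing hand, named as a guided tutorial indicator.
# Nested SizeBoxes separate rotation from offset: the pivot is rotated, the distance box pads the hand from center, and the image counter-rotates.
# Anim_PointingHand nudges the image along its X translation and back over one second.
# The event graph holds only unconnected stub events, so the widget does not start the animation itself.
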WidgetBlueprint /Game/Effects/Fort_Effects/Gameplay/Tutorial_Target/BP_GuidedTutorial_PointingHand
animations [Anim_PointingHand {
  binding_count=1
  length_seconds=1
  object=&self:Anim_PointingHand
  sequence {bindings [PointingHand {
     id=#1
     possessed_class=Image
     tracks [{
       type=MovieScene2DTransformTrack
       sections [{
         type=MovieScene2DTransformSection
         end_frame=20
         start_frame=0
         channels [
          Translation.X {type=MovieSceneScriptingFloatChannel, keys [[0,0],[10,10],[20,0]]}
          Translation.Y {type=MovieSceneScriptingFloatChannel, const [0,1]}
          Angle {type=MovieSceneScriptingFloatChannel}
          Scale.X {type=MovieSceneScriptingFloatChannel}
          Scale.Y {type=MovieSceneScriptingFloatChannel}
          Shear.X {type=MovieSceneScriptingFloatChannel}
          Shear.Y {type=MovieSceneScriptingFloatChannel}
         ]
        }]
      }]
    }]}
 }]
graphs [EventGraph {nodes [
   Event PreConstruct()
   Event Construct()
   Event Tick()
  ]}]
variables []
widgets {
 info {inheritedWidgetCount=0, parentClass=&UserWidget, rootWidgetClass=&Overlay, widgetCount=4}
 widgets [
  Overlay_80 {widgetClassPath=&Overlay, widget=Overlay_80}
  PointingHandPivot {
   widgetClassPath=&SizeBox
   parent=Overlay_80
   widget=PointingHandPivot
   properties {renderTransform {angle=45, scale=<1 1>, shear=<0 0>, translation=<0 0>}}
   slot {class=OverlaySlot, properties {horizontalAlignment=HAlign_Center, verticalAlignment=VAlign_Center}}
  }
  PointingHandDistance {
   widgetClassPath=&SizeBox
   parent=PointingHandPivot
   widget=PointingHandDistance
   slot {class=SizeBoxSlot, properties {horizontalAlignment=HAlign_Center, verticalAlignment=VAlign_Center, padding=<ltrb 0 0 100 50>}}
  }
  PointingHand {
   widgetClassPath=&Image
   bIsVariable=true
   parent=PointingHandDistance
   widget=PointingHand
   properties {
    brush {
     bIsDynamicallyLoaded=false
     drawAs=Image
     imageType=NoImage
     mirroring=NoMirror
     resourceName=None
     resourceObject=&/T_GuidedTutorial_PointingHand
     tiling=NoTile
     imageSize=<100 100>
     margin=<ltrb 0 0 0 0>
     outlineSettings {
      bUseBrushTransparency=false
      roundingType=HalfHeightRadius
      width=0
      color {colorUseRule=UseColor_Specified, specifiedColor=<rgba 0 0 0 0>}
      cornerRadii=<0 0 0 0>
     }
     tintColor {colorUseRule=UseColor_Specified, specifiedColor=<rgba 1 1 1 1>}
     uVRegion {bIsValid=false, max=<0 0>, min=<0 0>}
    }
    colorAndOpacity=<rgba 1 0.6757 0 1>
    renderTransform {angle=-45, scale=<1 1>, shear=<0 0>, translation=<0 0>}
   }
   slot {class=SizeBoxSlot, properties {horizontalAlignment=HAlign_Center, verticalAlignment=VAlign_Center}}
  }
 ]
}
