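# Tutorial HUD widget for Athena, built on FortAthenaTutorialHUD, with a progress bar, target counter, text prompt, options button and completion banner.
# Parent-class events show a looping swipe-finger and gradient hint for drag-to-turn, loop a pointing-hand animation, or reveal the completion banner and, after a delay, an animated Continue button.
# Continue fires a LeaveZone client event and analytic, then leaves through the party context.
# Options opens a collapsed skip popup and broadcasts it.
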WidgetBlueprint /Game/Athena/HUD/Tutorial/AthenaTutorialHUD
animations [
 Finger_SwipeLeft_Idle {
  binding_count=1
  length_seconds=1
  object=&self:Finger_SwipeLeft_Idle
  sequence {bindings [Overlay_Finger {
     id=#1
     possessed_class=Overlay
     tracks [
      {
       type=MovieScene2DTransformTrack
       sections [{
         type=MovieScene2DTransformSection
         end_frame=20
         start_frame=0
         channels [
          Translation.X {type=MovieSceneScriptingFloatChannel, keys [[4,0],[16,-330]]}
          Translation.Y {type=MovieSceneScriptingFloatChannel, const [280,2]}
          Scale.X {type=MovieSceneScriptingFloatChannel}
          Scale.Y {type=MovieSceneScriptingFloatChannel}
          Shear.X {type=MovieSceneScriptingFloatChannel}
          Shear.Y {type=MovieSceneScriptingFloatChannel}
         ]
        }]
      }
      {
       type=MovieSceneFloatTrack
       display_name=Render Opacity
       sections [{
         type=MovieSceneFloatSection
         end_frame=20
         start_frame=0
         channels [None {type=MovieSceneScriptingFloatChannel, keys [[0,0],[4,1],[16,1],[20.0007,0]]}]
        }]
      }
     ]
    }]}
 }
 ContinueButton {
  binding_count=1
  length_seconds=0.25
  object=&self:ContinueButton
  sequence {bindings [Tutorial_ContinueButton_Regular {
     id=#2
     possessed_class=IconTextButton_C
     tracks [{
       type=MovieScene2DTransformTrack
       sections [{
         type=MovieScene2DTransformSection
         end_frame=5
         start_frame=0
         channels [
          Translation.X {type=MovieSceneScriptingFloatChannel}
          Translation.Y {type=MovieSceneScriptingFloatChannel}
          Angle {type=MovieSceneScriptingFloatChannel}
          Scale.X {type=MovieSceneScriptingFloatChannel, keys [[0,0.125],[1,0.3001],[2,1],[3,1],[4,0.9134],[5,1]]}
          Scale.Y {type=MovieSceneScriptingFloatChannel, keys [[0,1.75],[1,1],[2,0.6676],[3,1],[4,1.0888],[5,1]]}
          Shear.X {type=MovieSceneScriptingFloatChannel}
          Shear.Y {type=MovieSceneScriptingFloatChannel}
         ]
        }]
      }]
    }]}
 }
]
graphs [EventGraph {nodes [
   "On Button Clicked (Tutorial_ContinueButton_Regular)"!() -> (OutputDelegate, then, Button:Common Button Legacy Object Reference) {is=title K2Node_ComponentBoundEvent}
   > Execute Client Event!(
    execute:Exec
    EventType:Fort Client Event Name Structure=(CategoryName="UserInterface",EventName="LeaveZone")
    EventSource:Object Reference=Self-Reference()
    EventFocus:Object Reference=Self-Reference()
   ) -> (then:Exec) {is=title K2Node_ExecuteClientEvent}
   > SendLeaveZoneAnalytic(self:Fort Global UIContext Object Reference=Get FortGlobalUIContext() -> (ReturnValue:Fort Global UIContext Object Reference) {is=title K2Node_GetSubsystem})
   > LeaveAction!(
    execute
    self:Fort Party Context Object Reference=Get FortPartyContext!() -> (ReturnValue:Fort Party Context Object Reference) {is=title K2Node_GetBlueprintContextEx}
    bInStayWithParty:bool=false
   ) -> (then)
   "On Button Clicked (Athena_Tutorial_Skip_Mobile)"!() -> (OutputDelegate, then, Button:Common Button Legacy Object Reference) {is=title K2Node_ComponentBoundEvent}
   n1 = Get Athena_Tutorial_ModalPopup() -> (Athena_Tutorial_ModalPopup:Fort Athena Tutorial Screen Object Reference)
   > Branch(Condition="Equal (Enum)"(
     A:ESlateVisibility Enum=GetVisibility!(self:Widget Object Reference=@n1) -> (ReturnValue:ESlateVisibility Enum=Visible)
     B:ESlateVisibility Enum=Collapsed
    ))
   n2 = Get Athena_Tutorial_ModalPopup()
   > SetVisibility(self=@n2, InVisibility=Visible)
   > n3 = Cast To Athena_Tutorial_ModalPopup!(execute, Object=@n2) -> (then, CastFailed, AsAthena Tutorial Modal Popup:Athena Tutorial Modal Popup Object Reference) {is=title K2Node_DynamicCast}
   > PlayAnimation(self=@n3, InAnimation=Get In!(self:Athena Tutorial Modal Popup Object Reference=@n3) -> (In:Widget Animation Object Reference))
   > BroadcastSkipMenuOpened(self:Fort Athena Tutorial HUD Object Reference)
   Event CompletedEnabled() {is=title K2Node_Event}
   > SetVisibility(self=Get Overlay_Completed() -> (Overlay_Completed:Overlay Object Reference), InVisibility=SelfHitTestInvisible)
   > OnCompleted(self:Tutorial Complete V2 Object Reference=Get Tutorial_Complete_V2() -> (Tutorial_Complete_V2:Tutorial Complete V2 Object Reference))
   > Delay(Duration=3)
   > SetVisibility(
    self=Get Tutorial_ContinueButton_Regular() -> (Tutorial_ContinueButton_Regular:Icon Text Button Object Reference)
    InVisibility=SelfHitTestInvisible
   )
   > PlayAnimation(InAnimation=Get ContinueButton() -> (ContinueButton:Widget Animation Object Reference))
   Event DragToTurnEnabled() {is=title K2Node_Event}
   > SetVisibility(
    self=[Get Overlay_Finger() -> (Overlay_Finger:Overlay Object Reference), Get Overlay_Gradient() -> (Overlay_Gradient:Overlay Object Reference)]
    InVisibility=SelfHitTestInvisible
   )
   > PlayAnimation(InAnimation=Get Finger_SwipeLeft_Idle() -> (Finger_SwipeLeft_Idle:Widget Animation Object Reference), NumLoopsToPlay=0)
   Event PlayFingerPointingAnimation() {is=title K2Node_Event}
   > n4 = Cast To BP_GuidedTutorial_PointingHand!(execute, Object=Get Widget_EasyInteractFinger() -> (Widget_EasyInteractFinger:User Widget Object Reference)) -> (then, CastFailed, AsBP Guided Tutorial Pointing Hand:BP Guided Tutorial Pointing Hand Object Reference) {is=title K2Node_DynamicCast}
   > PlayAnimation(
    self=@n4
    InAnimation=Get Anim_PointingHand!(self:BP Guided Tutorial Pointing Hand Object Reference=@n4) -> (Anim_PointingHand:Widget Animation Object Reference)
    NumLoopsToPlay=0
   )
  ]}]
variables []
widgets {
 info {inheritedWidgetCount=9, parentClass=/Script/FortniteUI.FortAthenaTutorialHUD, rootWidgetClass=&CanvasPanel, widgetCount=15}
 widgets [
  CanvasPanel_0 {widgetClassPath=&CanvasPanel, widget=CanvasPanel_0}
  Overlay_TutorialScreen {
   widgetClassPath=&Overlay
   bInherited=true
   bIsVariable=true
   parent=CanvasPanel_0
   widget=Overlay_TutorialScreen
   slot {class=CanvasPanelSlot, properties {layoutData {alignment=<0 0>, anchors {maximum=<1 1>, minimum=<0 0>}, offsets=<ltrb 0 0 0 0>}}}
  }
  Athena_Tutorial_ProgressBar {
   widgetClassPath=&/Athena_Tutorial_ProgressBar
   bInherited=true
   parent=Overlay_TutorialScreen
   widget=Athena_Tutorial_ProgressBar
   properties {visibility=HitTestInvisible}
   slot {class=OverlaySlot, properties {padding=<ltrb 40 180 0 0>}}
  }
  Athena_Tutorial_TargetCounter {
   widgetClassPath=&/Athena_Tutorial_TargetCounterScreen
   bInherited=true
   parent=CanvasPanel_0
   widget=Athena_Tutorial_TargetCounter
   properties {visibility=Collapsed}
   slot {
    class=CanvasPanelSlot
    properties {layoutData {alignment=<0 0>, anchors {maximum=<0.5 0>, minimum=<0.5 0>}, offsets=<ltrb -245 200 490 100>}}
   }
  }
  Athena_Tutorial_ModalPopup {
   widgetClassPath=&/Athena_Tutorial_ModalPopup
   bInherited=true
   parent=CanvasPanel_0
   widget=Athena_Tutorial_ModalPopup
   properties {visibility=Collapsed}
   slot {
    class=CanvasPanelSlot
    properties {zOrder=10, layoutData {alignment=<0 0>, anchors {maximum=<0.5 0.5>, minimum=<0.5 0.5>}, offsets=<ltrb -634.0266 -333.2082 1292.5559 786.71>}}
   }
  }
  TextPrompt_Tutorial {
   widgetClassPath=&/Tutorial_TextPrompt_Regular_Press
   bInherited=true
   bIsVariable=true
   parent=CanvasPanel_0
   widget=TextPrompt_Tutorial
   properties {visibility=SelfHitTestInvisible}
   slot {class=CanvasPanelSlot, properties {layoutData {alignment=<0.5 0>, anchors {maximum=<0.5 0>, minimum=<0.5 0>}, offsets=<ltrb 0 0 500 250>}}}
  }
  Overlay_Finger {
   widgetClassPath=&Overlay
   bInherited=true
   parent=CanvasPanel_0
   widget=Overlay_Finger
   properties {visibility=Collapsed}
   slot {
    class=CanvasPanelSlot
    properties {bAutoSize=true, layoutData {alignment=<1 0.5>, anchors {maximum=<1 1>, minimum=<1 1>}, offsets=<ltrb -9.9219 -568.0811 210 213>}}
   }
  }
  Tutorial_Finger {
   widgetClassPath=&/Tutorial_Finger
   bIsVariable=true
   parent=Overlay_Finger
   widget=Tutorial_Finger
   slot {class=OverlaySlot, properties {horizontalAlignment=HAlign_Fill, verticalAlignment=VAlign_Fill}}
  }
  Overlay_Gradient {
   widgetClassPath=&Overlay
   bInherited=true
   parent=CanvasPanel_0
   widget=Overlay_Gradient
   properties {visibility=Collapsed}
   slot {
    class=CanvasPanelSlot
    properties {layoutData {alignment=<1 1>, anchors {maximum=<1 1>, minimum=<1 1>}, offsets=<ltrb 154.0781 178.9189 1100 1100>}}
   }
  }
  Tutorial_RadialGradient {
   widgetClassPath=&/Tutorial_RadialGradient
   bIsVariable=true
   parent=Overlay_Gradient
   widget=Tutorial_RadialGradient
   slot {class=OverlaySlot, properties {horizontalAlignment=HAlign_Fill, verticalAlignment=VAlign_Fill}}
  }
  Tutorial_ContinueButton_Regular {
   widgetClassPath=&IconTextButton
   bIsVariable=true
   parent=CanvasPanel_0
   widget=Tutorial_ContinueButton_Regular
   properties {
    bIsFocusable=false
    buttonText=Continue
    contentAlignment=HAlign_Center
    controllerInputStyle=/FortUILibrary/Foundation/Buttons/ButtonStyles/ButtonStyle_Feature_L_Yellow.ButtonStyle_Feature_L_Yellow_C
    inputModeOverride=Game
    inputPriority=10000
    onButtonBaseClicked=(&self:WidgetTree.Tutorial_ContinueButton_Regular.HandleOnButtonClicked)
    onButtonBaseDoubleClicked=(&self:WidgetTree.Tutorial_ContinueButton_Regular.HandleOnButtonDoubleClicked)
    onButtonBaseHovered=(&self:WidgetTree.Tutorial_ContinueButton_Regular.HandleOnButtonHovered)
    onButtonBaseUnhovered=(&self:WidgetTree.Tutorial_ContinueButton_Regular.HandleOnButtonUnhovered)
    onSelectedChangedBase=(&self:WidgetTree.Tutorial_ContinueButton_Regular.HandleOnSelectedChanged)
    priority=10000
    style=&ButtonStyle_Feature_L_Yellow_C
    visibility=Collapsed
    triggeringInputAction {dataTable=&/Tutorial_InputActionDataTable, rowName=Tutorial_Continue}
   }
   slot {
    class=CanvasPanelSlot
    properties {layoutData {alignment=<0.5 0>, anchors {maximum=<0.5 1>, minimum=<0.5 1>}, offsets=<ltrb 0 -300 400 100>}}
   }
  }
  Overlay_Completed {
   widgetClassPath=&Overlay
   bIsVariable=true
   parent=CanvasPanel_0
   widget=Overlay_Completed
   properties {visibility=Collapsed}
   slot {
    class=CanvasPanelSlot
    properties {bAutoSize=true, layoutData {alignment=<0.5 0>, anchors {maximum=<0.5 0>, minimum=<0.5 0>}, offsets=<ltrb 0 0 410 400>}}
   }
  }
  Tutorial_Complete_V2 {
   widgetClassPath=&/Tutorial_Complete_V2
   bIsVariable=true
   parent=Overlay_Completed
   widget=Tutorial_Complete_V2
   slot {class=OverlaySlot, properties {horizontalAlignment=HAlign_Fill, verticalAlignment=VAlign_Fill}}
  }
  Athena_Tutorial_Skip_Mobile {
   widgetClassPath=&IconTextButton
   bInherited=true
   bIsVariable=true
   parent=CanvasPanel_0
   widget=Athena_Tutorial_Skip_Mobile
   properties {
    buttonText=Options
    inputModeOverride=Game
    inputPriority=10
    onButtonBaseClicked=(&self:WidgetTree.Athena_Tutorial_Skip_Mobile.HandleOnButtonClicked)
    onButtonBaseDoubleClicked=(&self:WidgetTree.Athena_Tutorial_Skip_Mobile.HandleOnButtonDoubleClicked)
    onButtonBaseHovered=(&self:WidgetTree.Athena_Tutorial_Skip_Mobile.HandleOnButtonHovered)
    onButtonBaseUnhovered=(&self:WidgetTree.Athena_Tutorial_Skip_Mobile.HandleOnButtonUnhovered)
    onSelectedChangedBase=(&self:WidgetTree.Athena_Tutorial_Skip_Mobile.HandleOnSelectedChanged)
    priority=10
    style=/FortUILibrary/Foundation/Buttons/ButtonStyles/TopBar/ButtonStyle-Outline-S.ButtonStyle-Outline-S_C
    colorAndOpacity=<rgba 1 1 1 0.5>
    triggeringInputAction {dataTable=&Tutorial_InputActionDataTable, rowName=Tutorial_Skip}
   }
   slot {
    class=CanvasPanelSlot
    properties {layoutData {alignment=<0 0>, anchors {maximum=<1 0>, minimum=<1 0>}, offsets=<ltrb -228.1901 20 210.3643 42.2027>}}
   }
  }
  Widget_EasyInteractFinger {
   widgetClassPath=/Game/Effects/Fort_Effects/Gameplay/Tutorial_Target/BP_GuidedTutorial_PointingHand
   bInherited=true
   bIsVariable=true
   parent=CanvasPanel_0
   widget=Widget_EasyInteractFinger
   properties {renderTransformPivot=<0.5 1>}
   slot {class=CanvasPanelSlot, properties {layoutData {alignment=<0.5 0.5>, anchors {maximum=<0 0>, minimum=<0 0>}, offsets=<ltrb 0 0 200 300>}}}
  }
 ]
}
