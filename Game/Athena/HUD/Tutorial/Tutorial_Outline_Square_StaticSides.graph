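# Tutorial HUD widget that draws a gold square highlight: a bordered outline image over a soft, blurry center glow.
# On Construct it plays the Intro animation once, which holds both images invisible briefly, then fades them in.
# During the fade the center scales down to normal size, and a padding track bound to the outline's slot animates from negative to zero for a closing-in effect.
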WidgetBlueprint /Game/Athena/HUD/Tutorial/Tutorial_Outline_Square_StaticSides
animations [Intro {
  binding_count=3
  length_seconds=1.1333
  object=&self:Intro
  sequence {
   bindings [
    OverlaySlot_5 {
     id=#1
     parent_id=#2
     possessed_class=OverlaySlot
     tracks [{
       type=MovieSceneMarginTrack
       display_name=Padding
       sections [{
         type=MovieSceneMarginSection
         end_frame=67
         start_frame=0
         channels [
          Left {type=MovieSceneScriptingFloatChannel, keys [[0,-400],[36,-400],[66,0]]}
          Top {type=MovieSceneScriptingFloatChannel, keys [[0,-400],[36,-400],[66,0]]}
          Right {type=MovieSceneScriptingFloatChannel, keys [[0,-400],[36,-400],[66,0]]}
          Bottom {type=MovieSceneScriptingFloatChannel, keys [[0,-400],[36,-400],[66,0]]}
         ]
        }]
      }]
    }
    IMG_Center {
     id=#3
     possessed_class=Image
     tracks [
      {
       type=MovieScene2DTransformTrack
       sections [{
         type=MovieScene2DTransformSection
         end_frame=67
         start_frame=0
         channels [
          Translation.X {type=MovieSceneScriptingFloatChannel}
          Translation.Y {type=MovieSceneScriptingFloatChannel}
          Angle {type=MovieSceneScriptingFloatChannel}
          Scale.X {type=MovieSceneScriptingFloatChannel, keys [[0,4],[36,4],[66,1]]}
          Scale.Y {type=MovieSceneScriptingFloatChannel, keys [[0,4],[36,4],[66,1]]}
          Shear.X {type=MovieSceneScriptingFloatChannel}
          Shear.Y {type=MovieSceneScriptingFloatChannel}
         ]
        }]
      }
      {
       type=MovieSceneFloatTrack
       display_name=Render Opacity
       sections [{
         type=MovieSceneFloatSection
         end_frame=67
         start_frame=0
         channels [None {type=MovieSceneScriptingFloatChannel, keys [[0,0],[36,0],[66,0.25]]}]
        }]
      }
     ]
    }
    IMG_Outline_Mid {
     id=#2
     possessed_class=Image
     tracks [{
       type=MovieSceneFloatTrack
       display_name=Render Opacity
       sections [{type=MovieSceneFloatSection, end_frame=67, start_frame=0, channels [None {type=MovieSceneScriptingFloatChannel, keys [[0,0],[36,0],[66,1]]}]}]
      }]
    }
   ]
   display_rate {numerator=60}
  }
 }]
graphs [EventGraph {nodes [
   Event Construct()
   > PlayAnimation(InAnimation=Get Intro())
  ]}]
variables []
widgets {
 info {inheritedWidgetCount=0, parentClass=&UserWidget, rootWidgetClass=&Overlay, widgetCount=4}
 widgets [
  Overlay_55 {widgetClassPath=&Overlay, widget=Overlay_55}
  IMG_Center {
   widgetClassPath=&Image
   bIsVariable=true
   parent=Overlay_55
   widget=IMG_Center
   properties {
    renderOpacity=0.25
    visibility=SelfHitTestInvisible
    brush {
     bIsDynamicallyLoaded=false
     drawAs=Image
     imageType=NoImage
     mirroring=NoMirror
     resourceName=None
     resourceObject=/Game/Athena/HUD/Art/Materials/MI_HUD_Thumbstick_Blurry
     tiling=NoTile
     imageSize=<92 92>
     margin=<ltrb 0 0 0 0>
     outlineSettings {
      bUseBrushTransparency=false
      roundingType=HalfHeightRadius
      width=0
      color {colorUseRule=UseColor_Specified, specifiedColor=<rgba 0 0 0 0>}
      cornerRadii=<0 0 0 0>
     }
     tintColor {colorUseRule=UseColor_Specified, specifiedColor=<rgba 0.9131 0.6308 0.0865 1>}
     uVRegion {bIsValid=false, max=<0 0>, min=<0 0>}
    }
   }
   slot {class=OverlaySlot, properties {horizontalAlignment=HAlign_Fill, verticalAlignment=VAlign_Fill}}
  }
  CommonVisibilitySwitcher_128 {
   widgetClassPath=&CommonVisibilitySwitcher
   parent=Overlay_55
   widget=CommonVisibilitySwitcher_128
   slot {class=OverlaySlot, properties {horizontalAlignment=HAlign_Fill, verticalAlignment=VAlign_Fill}}
  }
  IMG_Outline_Mid {
   widgetClassPath=&Image
   bIsVariable=true
   parent=CommonVisibilitySwitcher_128
   widget=IMG_Outline_Mid
   properties {
    visibility=SelfHitTestInvisible
    brush {
     bIsDynamicallyLoaded=false
     drawAs=Border
     imageType=NoImage
     mirroring=NoMirror
     resourceName=None
     resourceObject=/Game/Athena/HUD/Art/Tutorial/MI_GuidedTutorial_SquareHighlight_Mid
     tiling=NoTile
     imageSize=<92 92>
     margin=<ltrb 0.11 0 0.11 0>
     outlineSettings {
      bUseBrushTransparency=false
      roundingType=HalfHeightRadius
      width=0
      color {colorUseRule=UseColor_Specified, specifiedColor=<rgba 0 0 0 0>}
      cornerRadii=<0 0 0 0>
     }
     tintColor {colorUseRule=UseColor_Specified, specifiedColor=<rgba 0.4508 0.3095 0.0423 1>}
     uVRegion {bIsValid=false, max=<0 0>, min=<0 0>}
    }
   }
   slot {class=CommonVisibilitySwitcherSlot}
  }
 ]
}
