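# Tutorial HUD widget that shows an animated, gold-tinted triangle arrow, drawn by a material-instance image inside a root overlay.
# Construct plays Intro, which holds the arrow hidden briefly, then drops it into place while fading it in and scaling it down.
# When Intro finishes, an animation-finished event starts Idle on an endless loop, bobbing the overlay vertically with a slight opacity pulse.
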WidgetBlueprint /Game/Athena/HUD/Tutorial/Tutorial_Arrow_Triangle
animations [
 Idle {
  binding_count=1
  length_seconds=1
  object=&self:Idle
  sequence {
   bindings [Overlay_31 {
     id=#1
     possessed_class=Overlay
     tracks [
      {
       type=MovieScene2DTransformTrack
       sections [{
         type=MovieScene2DTransformSection
         end_frame=60
         start_frame=0
         channels [
          Translation.X {type=MovieSceneScriptingFloatChannel}
          Translation.Y {type=MovieSceneScriptingFloatChannel, keys [[0,-25],[30,-75],[60,-25]]}
          Angle {type=MovieSceneScriptingFloatChannel}
          Scale.X {type=MovieSceneScriptingFloatChannel}
          Scale.Y {type=MovieSceneScriptingFloatChannel}
          Shear.X {type=MovieSceneScriptingFloatChannel}
          Shear.Y {type=MovieSceneScriptingFloatChannel}
         ]
        }]
      }
      {
       type=MovieSceneFloatTrack
       display_name=Render Opacity
       sections [{
         type=MovieSceneFloatSection
         end_frame=60
         start_frame=0
         channels [None {type=MovieSceneScriptingFloatChannel, keys [[0,1],[30,0.75],[60,1]]}]
        }]
      }
     ]
    }]
   display_rate {numerator=60}
  }
 }
 Intro {
  binding_count=1
  length_seconds=1.1333
  object=&self:Intro
  sequence {
   bindings [Overlay_31 {
     id=#1
     possessed_class=Overlay
     tracks [
      {
       type=MovieScene2DTransformTrack
       sections [{
         type=MovieScene2DTransformSection
         end_frame=68
         start_frame=0
         channels [
          Translation.X {type=MovieSceneScriptingFloatChannel}
          Translation.Y {type=MovieSceneScriptingFloatChannel, keys [[0,-150],[36,-150],[66,-25]]}
          Angle {type=MovieSceneScriptingFloatChannel}
          Scale.X {type=MovieSceneScriptingFloatChannel, keys [[0,1.2],[36,1.2],[66,1]]}
          Scale.Y {type=MovieSceneScriptingFloatChannel, keys [[0,1.2],[36,1.2],[66,1]]}
          Shear.X {type=MovieSceneScriptingFloatChannel}
          Shear.Y {type=MovieSceneScriptingFloatChannel}
         ]
        }]
      }
      {
       type=MovieSceneFloatTrack
       display_name=Render Opacity
       sections [{type=MovieSceneFloatSection, end_frame=68, start_frame=0, channels [None {type=MovieSceneScriptingFloatChannel, keys [[0,0],[36,0],[66,1]]}]}]
      }
     ]
    }]
   display_rate {numerator=60}
  }
 }
]
graphs [EventGraph {nodes [
   Event Construct()
   > Parent: Construct()
   > PlayAnimation(InAnimation=Get Intro())
   "Animation Finished (Intro)"!() -> (OutputDelegate:Delegate, then:Exec) {is=title K2Node_WidgetAnimationEvent}
   > PlayAnimation(InAnimation=Get Idle() -> (Idle:Widget Animation Object Reference), NumLoopsToPlay=0)
   Event PreConstruct()
   Event Tick()
  ]}]
variables []
widgets {
 info {inheritedWidgetCount=0, parentClass=&UserWidget, rootWidgetClass=&Overlay, widgetCount=2}
 widgets [
  Overlay_31 {widgetClassPath=&Overlay, widget=Overlay_31}
  IMG_Arrow_Triangle {
   widgetClassPath=&Image
   bIsVariable=true
   parent=Overlay_31
   widget=IMG_Arrow_Triangle
   properties {brush {
     bIsDynamicallyLoaded=false
     drawAs=Image
     imageType=NoImage
     mirroring=NoMirror
     resourceName=None
     resourceObject=/Game/Athena/HUD/Art/Materials/MI_Tutorial_Arrow_Triangle
     tiling=NoTile
     imageSize=<74 64>
     margin=<ltrb 0 0 0 0>
     outlineSettings {
      bUseBrushTransparency=false
      roundingType=HalfHeightRadius
      width=0
      color {colorUseRule=UseColor_Specified, specifiedColor=<rgba 0 0 0 0>}
      cornerRadii=<0 0 0 0>
     }
     tintColor {colorUseRule=UseColor_Specified, specifiedColor=<rgba 0.9131 0.6308 0.0865 1>}
     uVRegion {bIsValid=false, max=<0 0>, min=<0 0>}
    }}
   slot {class=OverlaySlot, properties {horizontalAlignment=HAlign_Fill, verticalAlignment=VAlign_Fill}}
  }
 ]
}
