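# Tutorial highlight circle HUD widget that overlays a gold thumbstick outline, a soft blurred center and a Tutorial_Arrow_Triangle child.
# Intro plays on Construct and from a visibility custom event for visible states, holding the images hidden briefly, then shrinking and fading them in.
# Intro completion starts Idle, which pulses outline scale and both opacities.
# Two Intro-finished handlers both play Idle, one looping endlessly, the other with default loop count.
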WidgetBlueprint /Game/Athena/HUD/Tutorial/Tutorial_Highlight_Circle
animations [
 Idle {
  binding_count=2
  length_seconds=1
  object=&self:Idle
  sequence {
   bindings [
    IMG_Outline {
     id=#1
     possessed_class=Image
     tracks [
      {
       type=MovieScene2DTransformTrack
       sections [{
         type=MovieScene2DTransformSection
         end_frame=60
         start_frame=0
         channels [
          Translation.X {type=MovieSceneScriptingFloatChannel}
          Translation.Y {type=MovieSceneScriptingFloatChannel}
          Angle {type=MovieSceneScriptingFloatChannel}
          Scale.X {type=MovieSceneScriptingFloatChannel, keys [[0,1],[30,1.4],[60,1]]}
          Scale.Y {type=MovieSceneScriptingFloatChannel, keys [[0,1],[30,1.4],[60,1]]}
          Shear.X {type=MovieSceneScriptingFloatChannel}
          Shear.Y {type=MovieSceneScriptingFloatChannel}
         ]
        }]
      }
      {
       type=MovieSceneFloatTrack
       display_name=Render Opacity
       sections [{type=MovieSceneFloatSection, end_frame=60, start_frame=0, channels [None {type=MovieSceneScriptingFloatChannel, keys [[0,1],[30,0],[60,1]]}]}]
      }
     ]
    }
    IMG_Center {
     id=#2
     possessed_class=Image
     tracks [{
       type=MovieSceneFloatTrack
       display_name=Render Opacity
       sections [{
         type=MovieSceneFloatSection
         end_frame=60
         start_frame=0
         channels [None {type=MovieSceneScriptingFloatChannel, keys [[0,0.25],[30,0.1],[60,0.25]]}]
        }]
      }]
    }
   ]
   display_rate {numerator=60}
  }
 }
 Intro {
  binding_count=2
  length_seconds=1.1333
  object=&self:Intro
  sequence {
   bindings [
    IMG_Outline {
     id=#1
     possessed_class=Image
     tracks [
      {
       type=MovieScene2DTransformTrack
       sections [{
         type=MovieScene2DTransformSection
         end_frame=67
         start_frame=0
         channels [
          Translation.X {type=MovieSceneScriptingFloatChannel}
          Translation.Y {type=MovieSceneScriptingFloatChannel}
          Angle {type=MovieSceneScriptingFloatChannel}
          Scale.X {type=MovieSceneScriptingFloatChannel, keys [[0,4],[36,4],[66,1]]}
          Scale.Y {type=MovieSceneScriptingFloatChannel, keys [[0,4],[36,4],[66,1]]}
          Shear.X {type=MovieSceneScriptingFloatChannel}
          Shear.Y {type=MovieSceneScriptingFloatChannel}
         ]
        }]
      }
      {
       type=MovieSceneFloatTrack
       display_name=Render Opacity
       sections [{type=MovieSceneFloatSection, end_frame=67, start_frame=0, channels [None {type=MovieSceneScriptingFloatChannel, keys [[0,0],[36,0],[66,1]]}]}]
      }
     ]
    }
    IMG_Center {
     id=#2
     possessed_class=Image
     tracks [
      {
       type=MovieSceneFloatTrack
       display_name=Render Opacity
       sections [{
         type=MovieSceneFloatSection
         end_frame=67
         start_frame=0
         channels [None {type=MovieSceneScriptingFloatChannel, keys [[0,0],[36,0],[66,0.25]]}]
        }]
      }
      {
       type=MovieScene2DTransformTrack
       sections [{
         type=MovieScene2DTransformSection
         end_frame=67
         start_frame=0
         channels [
          Translation.X {type=MovieSceneScriptingFloatChannel}
          Translation.Y {type=MovieSceneScriptingFloatChannel}
          Angle {type=MovieSceneScriptingFloatChannel}
          Scale.X {type=MovieSceneScriptingFloatChannel, keys [[0,5],[36,5],[66,1]]}
          Scale.Y {type=MovieSceneScriptingFloatChannel, keys [[0,5],[36,5],[66,1]]}
          Shear.X {type=MovieSceneScriptingFloatChannel}
          Shear.Y {type=MovieSceneScriptingFloatChannel}
         ]
        }]
      }
     ]
    }
   ]
   display_rate {numerator=60}
  }
 }
]
graphs [EventGraph {nodes [
   Event Construct()
   > Parent: Construct()
   > PlayAnimation(InAnimation=Get Intro())
   K2Node_Event_3 = OnVisibilitySetEvent_1!() -> (OutputDelegate, then, InVisibility:ESlateVisibility Enum=Visible) {is=title K2Node_CustomEvent}
   > n1 = Switch on ESlateVisibility!(execute, Selection:ESlateVisibility Enum=@K2Node_Event_3.InVisibility) -> (Visible:Exec, Collapsed:Exec, Hidden:Exec, HitTestInvisible:Exec, SelfHitTestInvisible:Exec) {is=title K2Node_SwitchEnum}
   PlayAnimation(execute=[@n1.HitTestInvisible, @n1.SelfHitTestInvisible, @n1.Visible], InAnimation=Get Intro())
   "Animation Finished (Intro)"!() -> (OutputDelegate:Delegate, then:Exec) {is=title K2Node_WidgetAnimationEvent}
   > PlayAnimation(InAnimation=Get Idle() -> (Idle:Widget Animation Object Reference), NumLoopsToPlay=0)
   "Animation Finished (Intro)"()
   > PlayAnimation(InAnimation=Get Idle())
   Event PreConstruct()
   Event Tick()
  ]}]
variables []
widgets {
 info {inheritedWidgetCount=0, parentClass=&UserWidget, rootWidgetClass=&Overlay, widgetCount=4}
 widgets [
  Overlay_23 {widgetClassPath=&Overlay, widget=Overlay_23}
  IMG_Outline {
   widgetClassPath=&Image
   bIsVariable=true
   parent=Overlay_23
   widget=IMG_Outline
   properties {
    visibility=SelfHitTestInvisible
    brush {
     bIsDynamicallyLoaded=false
     drawAs=Image
     imageType=NoImage
     mirroring=NoMirror
     resourceName=None
     resourceObject=/Game/Athena/HUD/Art/Materials/MI_Hud_Thumbstick_Tutorial
     tiling=NoTile
     imageSize=<256 256>
     margin=<ltrb 0 0 0 0>
     outlineSettings {
      bUseBrushTransparency=false
      roundingType=HalfHeightRadius
      width=0
      color {colorUseRule=UseColor_Specified, specifiedColor=<rgba 0 0 0 0>}
      cornerRadii=<0 0 0 0>
     }
     tintColor {colorUseRule=UseColor_Specified, specifiedColor=<rgba 0.9131 0.6308 0.0865 1>}
     uVRegion {bIsValid=false, max=<0 0>, min=<0 0>}
    }
   }
   slot {class=OverlaySlot, properties {horizontalAlignment=HAlign_Fill, verticalAlignment=VAlign_Fill}}
  }
  IMG_Center {
   widgetClassPath=&Image
   bIsVariable=true
   parent=Overlay_23
   widget=IMG_Center
   properties {
    renderOpacity=0.25
    visibility=SelfHitTestInvisible
    brush {
     bIsDynamicallyLoaded=false
     drawAs=Image
     imageType=NoImage
     mirroring=NoMirror
     resourceName=None
     resourceObject=/Game/Athena/HUD/Art/Materials/MI_HUD_Thumbstick_Blurry
     tiling=NoTile
     imageSize=<256 256>
     margin=<ltrb 0 0 0 0>
     outlineSettings {
      bUseBrushTransparency=false
      roundingType=HalfHeightRadius
      width=0
      color {colorUseRule=UseColor_Specified, specifiedColor=<rgba 0 0 0 0>}
      cornerRadii=<0 0 0 0>
     }
     tintColor {colorUseRule=UseColor_Specified, specifiedColor=<rgba 0.9131 0.6308 0.0865 1>}
     uVRegion {bIsValid=false, max=<0 0>, min=<0 0>}
    }
   }
   slot {class=OverlaySlot, properties {horizontalAlignment=HAlign_Fill, verticalAlignment=VAlign_Fill}}
  }
  Tutorial_Arrow_Triangle {
   widgetClassPath=&/Tutorial_Arrow_Triangle
   bIsVariable=true
   parent=Overlay_23
   widget=Tutorial_Arrow_Triangle
   slot {class=OverlaySlot, properties {horizontalAlignment=HAlign_Center}}
  }
 ]
}
