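# HUD tutorial highlight widget showing a move-forward cue: a yellow thumbstick circle over a gradient arrow.
# It starts collapsed on Construct; switching to a visible state plays Intro, which fades in the arrow and shrinks the circle into place.
# When Intro finishes, Idle loops endlessly, sliding the circle upward and fading it out.
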WidgetBlueprint /Game/Athena/HUD/Tutorial/Tutorial_Highlight_MoveFoward
animations [
 Idle {
  binding_count=1
  length_seconds=1.5
  object=&self:Idle
  sequence {
   bindings [Overlay_MoveCircle {
     id=#1
     possessed_class=Overlay
     tracks [
      {
       type=MovieScene2DTransformTrack
       sections [{
         type=MovieScene2DTransformSection
         end_frame=45
         start_frame=0
         channels [
          Translation.X {type=MovieSceneScriptingFloatChannel, const [0,1]}
          Translation.Y {type=MovieSceneScriptingFloatChannel, keys [[0,128],[30,-51],[45,-51]]}
          Angle {type=MovieSceneScriptingFloatChannel, const [0,1]}
          Scale.X {type=MovieSceneScriptingFloatChannel, const [1,1]}
          Scale.Y {type=MovieSceneScriptingFloatChannel, const [1,1]}
          Shear.X {type=MovieSceneScriptingFloatChannel, const [0,1]}
          Shear.Y {type=MovieSceneScriptingFloatChannel, const [0,1]}
         ]
        }]
      }
      {
       type=MovieSceneFloatTrack
       display_name=Render Opacity
       sections [{
         type=MovieSceneFloatSection
         end_frame=90
         start_frame=0
         channels [None {type=MovieSceneScriptingFloatChannel, keys [[0,0],[6,1],[45,1],[75,0]]}]
        }]
      }
     ]
    }]
   display_rate {numerator=60}
  }
 }
 Intro {
  binding_count=2
  length_seconds=0.5834
  object=&self:Intro
  sequence {
   bindings [
    Overlay_MoveCircle {
     id=#2
     possessed_class=Overlay
     tracks [
      {
       type=MovieSceneFloatTrack
       display_name=Render Opacity
       sections [{type=MovieSceneFloatSection, end_frame=35, start_frame=0, channels [None {type=MovieSceneScriptingFloatChannel, keys [[0,0],[17,1],[29,1]]}]}]
      }
      {
       type=MovieScene2DTransformTrack
       sections [{
         type=MovieScene2DTransformSection
         end_frame=35
         start_frame=0
         channels [
          Translation.X {type=MovieSceneScriptingFloatChannel, const [0,1]}
          Translation.Y {type=MovieSceneScriptingFloatChannel, const [128,1]}
          Angle {type=MovieSceneScriptingFloatChannel, const [0,1]}
          Scale.X {type=MovieSceneScriptingFloatChannel, keys [[0,3],[23,1]]}
          Scale.Y {type=MovieSceneScriptingFloatChannel, keys [[0,3],[23,1]]}
          Shear.X {type=MovieSceneScriptingFloatChannel, const [0,1]}
          Shear.Y {type=MovieSceneScriptingFloatChannel, const [0,1]}
         ]
        }]
      }
     ]
    }
    IMG_Gradient_Arrow {
     id=#3
     possessed_class=Image
     tracks [{
       type=MovieSceneFloatTrack
       display_name=Render Opacity
       sections [{type=MovieSceneFloatSection, end_frame=35, start_frame=0, channels [None {type=MovieSceneScriptingFloatChannel, keys [[0,0],[23,1]]}]}]
      }]
    }
   ]
   display_rate {numerator=60}
  }
 }
]
graphs [EventGraph {nodes [
   Event Construct()
   > Parent: Construct()
   > SetVisibility(InVisibility=Collapsed)
   Event OnVisibilitySetEvent!() -> (OutputDelegate, then, InVisibility:ESlateVisibility Enum=Visible) {is=title K2Node_Event}
   > n1 = Switch on ESlateVisibility!(execute, Selection:ESlateVisibility Enum=Visible) -> (Visible:Exec, Collapsed:Exec, Hidden:Exec, HitTestInvisible:Exec, SelfHitTestInvisible:Exec) {is=title K2Node_SwitchEnum}
   PlayAnimation(execute=[@n1.HitTestInvisible, @n1.SelfHitTestInvisible, @n1.Visible], InAnimation=Get Intro())
   "Animation Finished (Intro)"!() -> (OutputDelegate:Delegate, then:Exec) {is=title K2Node_WidgetAnimationEvent}
   > PlayAnimation(InAnimation=Get Idle() -> (Idle:Widget Animation Object Reference), NumLoopsToPlay=0)
  ]}]
variables []
widgets {
 info {inheritedWidgetCount=1, parentClass=&FortHUDElementWidget, rootWidgetClass=&Overlay, widgetCount=6}
 widgets [
  Overlay_Global {widgetClassPath=&Overlay, widget=Overlay_Global}
  IMG_Gradient_Arrow {
   widgetClassPath=&Image
   bIsVariable=true
   parent=Overlay_Global
   widget=IMG_Gradient_Arrow
   properties {
    visibility=SelfHitTestInvisible
    brush {
     bIsDynamicallyLoaded=false
     drawAs=Image
     imageType=NoImage
     mirroring=NoMirror
     resourceName=None
     resourceObject=/Game/Athena/HUD/Art/Tutorial/FTUE_Arrow_Sprite
     tiling=NoTile
     imageSize=<325 464>
     margin=<ltrb 0 0 0 0>
     outlineSettings {
      bUseBrushTransparency=false
      roundingType=HalfHeightRadius
      width=0
      color {colorUseRule=UseColor_Specified, specifiedColor=<rgba 0 0 0 0>}
      cornerRadii=<0 0 0 0>
     }
     tintColor {colorUseRule=UseColor_Specified, specifiedColor=<rgba 0.9137 0.6314 0.0863 1>}
     uVRegion {bIsValid=false, max=<0 0>, min=<0 0>}
    }
    renderTransform {angle=0, scale=<0.4 0.6>, shear=<0 0>, translation=<0 0.5104>}
    renderTransformPivot=<0.5 0.8>
   }
   slot {class=OverlaySlot, properties {horizontalAlignment=HAlign_Center, verticalAlignment=VAlign_Bottom}}
  }
  Overlay_MoveCircle {
   widgetClassPath=&Overlay
   parent=Overlay_Global
   widget=Overlay_MoveCircle
   properties {renderTransform {angle=0, scale=<1 1>, shear=<0 0>, translation=<0 128>}}
   slot {class=OverlaySlot, properties {horizontalAlignment=HAlign_Center, verticalAlignment=VAlign_Center}}
  }
  IMG_Outline_Circle {
   widgetClassPath=&Image
   bIsVariable=true
   parent=Overlay_MoveCircle
   widget=IMG_Outline_Circle
   properties {
    visibility=SelfHitTestInvisible
    brush {
     bIsDynamicallyLoaded=false
     drawAs=Image
     imageType=NoImage
     mirroring=NoMirror
     resourceName=None
     resourceObject=/Game/Athena/HUD/Art/Materials/MI_Hud_Thumbstick_yELLOW.MI_Hud_Thumbstick_Yellow
     tiling=NoTile
     imageSize=<150 150>
     margin=<ltrb 0 0 0 0>
     outlineSettings {
      bUseBrushTransparency=false
      roundingType=HalfHeightRadius
      width=0
      color {colorUseRule=UseColor_Specified, specifiedColor=<rgba 0 0 0 0>}
      cornerRadii=<0 0 0 0>
     }
     tintColor {colorUseRule=UseColor_Specified, specifiedColor=<rgba 0.9131 0.6308 0.0865 1>}
     uVRegion {bIsValid=false, max=<0 0>, min=<0 0>}
    }
   }
   slot {class=OverlaySlot, properties {horizontalAlignment=HAlign_Fill, verticalAlignment=VAlign_Fill}}
  }
  IMG_Gradient_Circle {
   widgetClassPath=&Image
   bIsVariable=true
   parent=Overlay_MoveCircle
   widget=IMG_Gradient_Circle
   properties {
    renderOpacity=0.25
    visibility=SelfHitTestInvisible
    brush {
     bIsDynamicallyLoaded=false
     drawAs=Image
     imageType=NoImage
     mirroring=NoMirror
     resourceName=None
     resourceObject=/Game/Athena/HUD/Art/Materials/MI_HUD_Thumbstick_Blurry
     tiling=NoTile
     imageSize=<150 150>
     margin=<ltrb 0 0 0 0>
     outlineSettings {
      bUseBrushTransparency=false
      roundingType=HalfHeightRadius
      width=0
      color {colorUseRule=UseColor_Specified, specifiedColor=<rgba 0 0 0 0>}
      cornerRadii=<0 0 0 0>
     }
     tintColor {colorUseRule=UseColor_Specified, specifiedColor=<rgba 0.9131 0.6308 0.0865 1>}
     uVRegion {bIsValid=false, max=<0 0>, min=<0 0>}
    }
   }
   slot {class=OverlaySlot, properties {horizontalAlignment=HAlign_Fill, verticalAlignment=VAlign_Fill}}
  }
  TutorialHighlight {widgetClassPath=&TutorialHighlight}
 ]
}
